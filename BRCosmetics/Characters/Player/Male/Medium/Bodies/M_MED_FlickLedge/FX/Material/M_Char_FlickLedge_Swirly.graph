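# Translucent lit, two-sided Niagara mesh-particle material for the FlickLedge character's FX, turning strip meshes into animated puffy, outlined swirls.
# Custom HLSL uses particle age and Niagara dynamic parameters to roll vertices into spirals or corkscrews, adding sine distortion and tapered SplineThicken width through world position offset.
# Per-particle random values vary the shapes.
# A channel-packed texture supplies fill, shadow, outline; opacity combines noise erosion, tail ramp, camera and depth fades.
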Material /BRCosmetics/Characters/Player/Male/Medium/Bodies/M_MED_FlickLedge/FX/Material/M_Char_FlickLedge_Swirly
graph {
 properties {
  bAutomaticallySetUsageInEditor=false
  bUsedWithNiagaraMeshParticles=true
  blendMode=BLEND_Translucent
  translucencyLightingMode=TLM_VolumetricPerVertexDirectional
  twoSided=true
  usedShadingModels=Default Lit
 }
 nodes [
  MaterialExpressionTextureCoordinate_4 = TextureCoordinate()
  MaterialExpressionScalarParameter_30 = ScalarParameter() {
   defaultValue=0.35
   desc="If the spline has center verts, this is the amount to displace them out to get more dimention and shading out of the spline geometry. "
   parameterName=DisplaceOutCenterVerts
  }
  MaterialExpressionNamedRerouteUsage_23 = NamedRerouteUsage()
  n1 = Custom!(
   v=@MaterialExpressionTextureCoordinate_4
   Intensity=@MaterialExpressionScalarParameter_30
   Splinethicken=@MaterialExpressionNamedRerouteUsage_23
  ) {
   code="float v1 = min(v.y*2, 1);\r\nfloat v2 = min((1-v.y)*2, 1);\r\nfloat result = v1*v2*Intensity*Splinethicken;\r\n\r\nreturn result;"
   outputType=CMOT_Float1
   showCode=true
   inputs [
    {
     inputName=v
     input {expression=&self:MaterialExpressionTextureCoordinate_4}
    }
    {
     inputName=Intensity
     input {expression=&self:MaterialExpressionScalarParameter_30}
    }
    {
     inputName=Splinethicken
     input {expression=&self:MaterialExpressionNamedRerouteUsage_23}
    }
   ]
  }
  n2 = ScalarParameter() {defaultValue=0.33, desc="The amount to create corkscrew shapes instead of spiral shapre. ", group=Spiral_Misc, parameterName=CorkScrewAmt}
  n3 = ScalarParameter() {
   defaultValue=-0.3
   desc="How much to squash the spiral at age 0. Negative numbers OK. "
   group=Spiral_Misc
   parameterName=SquashAmtStart
   sortPriority=99
  }
  n4 = ScalarParameter() {
   defaultValue=1.3
   desc=How much to squash the spiral at normaized age 1. Negative numbers OK.
   group=Spiral_Misc
   parameterName=SquashAmtEnd
   sortPriority=100
  }
  MaterialExpressionMaterialFunctionCall_4 = MaterialFunctionCall(Input=ParticleRelativeTime(), Target Low=@n3, Target High=@n4) {materialFunction=&RemapValueRange}
  MaterialExpressionNamedRerouteUsage_24 = NamedRerouteUsage()
  MaterialExpressionNamedRerouteUsage_34 = NamedRerouteUsage()
  MaterialExpressionNamedRerouteUsage_36 = NamedRerouteUsage()
  MaterialExpressionNamedRerouteUsage_44 = NamedRerouteUsage()
  MaterialExpressionNamedRerouteUsage_52 = NamedRerouteUsage()
  MaterialExpressionNamedRerouteUsage_40 = NamedRerouteUsage()
  n5 = Custom!(
   p0=@MaterialExpressionNamedRerouteUsage_36
   p1=@MaterialExpressionNamedRerouteUsage_44
   p2=@MaterialExpressionNamedRerouteUsage_52
   p3=@MaterialExpressionNamedRerouteUsage_40
  ) {
   code="offsets = p0 + p1 + p2;\r\n\r\nreturn p0 + p1 + p2 + p3;"
   description=AddPos
   showCode=true
   additionalOutputs [{outputName=offsets, outputType=CMOT_Float3}]
   inputs [
    {
     inputName=p0
     input {expression=&self:MaterialExpressionNamedRerouteUsage_36}
    }
    {
     inputName=p1
     input {expression=&self:MaterialExpressionNamedRerouteUsage_44}
    }
    {
     inputName=p2
     input {expression=&self:MaterialExpressionNamedRerouteUsage_52}
    }
    {
     inputName=p3
     input {expression=&self:MaterialExpressionNamedRerouteUsage_40}
    }
   ]
  }
  MaterialExpressionMaterialFunctionCall_7 = MaterialFunctionCall!(
   RotationAngle=NamedRerouteUsage()
   NormalizedRotationAxis=Multiply(A=NamedRerouteUsage(), B=NamedRerouteUsage())
   PivotPoint=NamedRerouteUsage()
   Position=@n5.return
   Normal=NamedRerouteUsage()
   TangentU
   TangentV
  ) {materialFunction=/BRCosmetics/Characters/Player/Male/Medium/Bodies/M_MED_Inferno/FX/Mat/RotateVertexAboutAxis}
  n6 = Custom!(
   amt=@MaterialExpressionMaterialFunctionCall_4.Result
   center=@MaterialExpressionNamedRerouteUsage_24
   up=@MaterialExpressionNamedRerouteUsage_34
   rot_wpo=@MaterialExpressionMaterialFunctionCall_7.WPO
  ) {
   code="//Normalize vectors\r\nfloat3 loc = rot_wpo-center;\r\nfloat3 n_loc = normalize(loc);\r\nfloat dist_loc = length(loc);\r\n\r\n//Streatch from swirl center\r\nfloat3 n_up = normalize(up);\r\nfloat ratio = dot(n_up, n_loc);\r\nfloat3 streatch = n_up * ratio * dist_loc * amt;\r\n\r\n//Squash back towards axis (not swirl center)\r\nfloat3 n_perp = normalize(center);\r\nfloat3 perp_vec = n_perp * dot(n_perp, rot_wpo);\r\nfloat3 squash = -perp_vec * min(amt, .25);\r\n\r\n\r\nreturn (loc + streatch + center) + squash;\r\n"
   desc="Can't figure out why the squash scales outer edge to nothing???\r\nThe bug actually looks good so leaving it in.\r\nAdd a min() fix to get by for now.\r\nEven an overall scale coming into this node creates the thinning."
   description=StreatchSwirlWpo
   showCode=true
   inputs [
    {
     inputName=amt
     input {expression=&self:MaterialExpressionMaterialFunctionCall_4}
    }
    {
     inputName=center
     input {expression=&self:MaterialExpressionNamedRerouteUsage_24}
    }
    {
     inputName=up
     input {expression=&self:MaterialExpressionNamedRerouteUsage_34}
    }
    {
     inputName=rot_wpo
     input {expression=&self:MaterialExpressionMaterialFunctionCall_7}
    }
   ]
  }
  MaterialExpressionReroute_5 = Reroute(None=@n6)
  n7 = Add(A=@MaterialExpressionMaterialFunctionCall_7.WPO, B=Multiply(A=NamedRerouteUsage()) {constB=-0.4})
  n8 = If(
   A=MaterialFunctionCall!(Input=ParticleRandom(), min, max, seed=Constant() {r=7}) {materialFunction=/InterchangeAssets/Functions/MX_RandomFloat}.Result
   B=@n2
   "A > B"=@MaterialExpressionReroute_5
   "A == B"=@n7
   "A < B"=@n7
  ) {constB=0.3}
  MaterialExpressionReroute_1 = Reroute(None=Add(A=NamedRerouteDeclaration(None=@n8) {name=SwirlOffset_Rotate, nodeColor=<rgba 1 0 0.7881 1>}, B=@n5.offsets))
  n9 = NamedRerouteUsage()
  n10 = ParticleRandom()
  n11 = Subtract(A=Multiply(A=Round(None=@n10)) {constB=2})
  n12 = ScalarParameter() {defaultValue=1, desc="Turn on sine wave random flippping to get more variation. ", group=Sinewave_WPO, parameterName=DualSineWaveSwitch}
  n13 = ParticleRandom()
  n14 = Add(A=ScalarParameter() {defaultValue=0.5, desc=Vary frequency of sine wave, group=Sinewave_WPO, parameterName=Sine_FreqVariation})
  n15 = NamedRerouteUsage()
  n16 = ScalarParameter() {defaultValue=1.2, desc=Frequency of sine wave, group=Sinewave_WPO, parameterName=SineFreq}
  MaterialExpressionPanner_1 = Panner(
   Coordinate=Multiply(
    A=MaterialFunctionCall(Input=@n13, Target Low=@n14) {materialFunction=&RemapValueRange}.Result
    B=Multiply(A=@n15, B=@n16) {constB=0.002, desc=freq}
   )
   Time=Time()
   Speed=Multiply(A=ScalarParameter() {defaultValue=0.2, desc=Speed sine wave moves so is not locked to geom., group=Sinewave_WPO, parameterName=SineWaveSpeed}) {constB=-1}
  ) {speedX=-1}
  n17 = Convert(None=@MaterialExpressionPanner_1) {
   convertInputs [{
     type=Vector2
     expressionInput {
      expression=&self:MaterialExpressionPanner_1
      inputName=None
      mask=0
      maskA=0
      maskB=0
      maskG=0
      maskR=0
      outputIndex=0
     }
    }]
   convertOutputs [{type=Scalar}]
  }
  n18 = Sine(None=@n17)
  n19 = Constant() {desc=ramp mag ndist, r=0.25}
  n20 = Clamp(None=MaterialFunctionCall(
    Input=@n15
    Input Low=Multiply(A=@n19) {constB=-1}
    Input High=@n19
    Target Low=Constant() {desc=mag before swirl, r=1}
    Target High=@MaterialExpressionReroute_1.Result
   ) {materialFunction=&RemapValueRange}.Result)
  n21 = Multiply(
   A=Multiply(A=Normalize(VectorInput=@n9) {desc=sine Wave in direction of swirl}, B=LinearInterpolate(B=@n11, Alpha=@n12) {constA=1})
   B=Multiply(A=Multiply(A=@n18, B=@n20), B=ScalarParameter() {defaultValue=8, desc=SineWaveAmp, group=Sinewave_WPO, parameterName=SineAmp}) {constB=100, desc=sin amt}
  ) {constA=1}
  Add(
   A=Multiply(A=NamedRerouteUsage(), B=@n1)
   B=Add(A=@MaterialExpressionReroute_1, B=NamedRerouteDeclaration(None=@n21) {name=SwirlOffset_GlobalDistortion, nodeColor=<rgba 1 0 0.8008 1>})
  )
  MaterialExpressionTextureCoordinate_0 = TextureCoordinate()
  Composite!(value=@MaterialExpressionTextureCoordinate_0) {subgraphName=randomize}
  n22 = ScalarParameter() {defaultValue=14, desc=how much radius expands over time. Negative numbers OK., group=Spiral, parameterName=Radius_ExpandDrift}
  Composite(value=@n22) {subgraphName=randomize}
  n23 = ScalarParameter() {
   defaultValue=-0.1
   desc="Spiral will start at this normaized age. Negative numbers will cause the spiral to already be started when partile is emitted. "
   group=Spiral
   parameterName=Rotation_StartAge
  }
  Composite(value=@n23) {subgraphName=randomize}
  Composite(value=ScalarParameter() {defaultValue=2, desc="How fast the spiral rotation is. ", group=Spiral, parameterName=Rotation_Speed}) {subgraphName=randomize}
  n24 = ScalarParameter() {
   defaultValue=0.5
   desc="Amplitude of sine wave on the spiral portion of the geometry (not the tail). "
   group=Sinewave_WPO
   parameterName=Sine_SwirlAmp
  }
  Composite(value=@n24) {subgraphName=randomize}
  n25 = ScalarParameter() {defaultValue=0.07, desc=Emit amount with a bigger radius. 0 to 1 range, group=Spiral, parameterName=Radius_InitialBigAmt, sortPriority=3}
  n26 = ScalarParameter() {
   defaultValue=15
   desc="For big ones, multiply up the initial radius of the spiral"
   group=Spiral
   parameterName=Radius_InitialBigOnes
   sortPriority=2
  }
  n27 = ScalarParameter() {defaultValue=1, desc="Initial radius when spiral begins. ", group=Spiral, parameterName=Radius_Initial, sortPriority=1}
  Composite!(B=@n25, B2=@n26, InputPin=@n27)
  n28 = VectorParameter() {
   desc=Main color. From texture blue channel.
   group=Color
   parameterName=ColorFill_Green
   sortPriority=3
   defaultValue=<rgba 0.0296 0.0332 0.04 1>
  }
  n29 = VectorParameter() {
   desc=Uses Blue channel from texture to create shadows
   group=Color
   parameterName=ColorShadow_Blue
   sortPriority=4
   defaultValue=<rgba 0.0229 0.0263 0.03 1>
  }
  n30 = TextureSampleParameter2D(UVs=Multiply(A=TextureCoordinate(), B=NamedRerouteUsage())) {
   desc="Red oultines, Green fill, Blue shadow, Alpha"
   group=Color
   parameterName=ColorTexture
   sortPriority=6
   texture=/BRCosmetics/Characters/Player/Male/Medium/Bodies/M_MED_FlickLedge/FX/Texture/T_Char_FlickLedge_Swirl
  }
  n31 = LinearInterpolate(A=@n28.RGB, B=@n29.RGB, Alpha=@n30.B)
  n32 = VectorParameter() {
   desc=Uses Red chanel from texture to create outlines
   group=Color
   parameterName=ColorOutline_Red
   sortPriority=5
   defaultValue=<rgba 0.1 0.1 0.1 1>
  }
  n33 = Multiply(A=ParticleColor().RGB, B=Multiply(A=LinearInterpolate(A=@n31, B=@n32.RGB, Alpha=@n30.R), B=@n30.A))
  MaterialFunctionCall(Emissive=Multiply(A=@n33, B=ScalarParameter() {defaultValue=0.7, group=Color, parameterName=EmissiveColorIntensity, sortPriority=1}) {constB=1.5}) {materialFunction=&MF_FX_EmissiveAdjustments}
  Multiply(A=@n33, B=ScalarParameter() {defaultValue=0.15, group=Color, parameterName=BaseColorIntensity, sortPriority=0}) {constB=0.15}
  n34 = Constant() {r=2}
  MaterialExpressionReroute_3 = Reroute(None=@MaterialExpressionTextureCoordinate_0)
  n35 = MaterialFunctionCall(
   Input=ParticleRandom()
   min=Subtract(A=@MaterialExpressionReroute_3.value, B=@n34)
   max=Add(A=@n34, B=@MaterialExpressionReroute_3.value)
   seed=Constant() {r=11}
  ) {desc=Drift, materialFunction=&MX_RandomFloat}
  MaterialExpressionReroute_9 = Reroute(None=@n35.Result)
  n36 = ScalarParameter() {
   defaultValue=2
   desc="Frequency of sine wave to thicken the spline geometry. "
   group=SplineThicken
   parameterName=SineThickness_Freq
   sortPriority=1
  }
  n37 = Multiply(A=@MaterialExpressionReroute_9.Result, B=@n36) {constB=2}
  n38 = Time()
  n39 = ScalarParameter() {
   defaultValue=1.5
   desc="Additive thickness ofspline geometry. Keep over 1 or sine wave will collapse in on itself. "
   group=SplineThicken
   parameterName=SineThickness_Add
   sortPriority=2
  }
  n40 = Add(A=Sine(None=Panner(Coordinate=@n37, Time=@n38) {speedX=-1}), B=@n39) {constB=0.5}
  n41 = ScalarParameter() {
   defaultValue=3
   desc="Amplitude of sine wave (multiply up sine wave). "
   group=SplineThicken
   parameterName=SineThickness_Mult
   sortPriority=3
  }
  n42 = Custom!(tex_coord=@MaterialExpressionTextureCoordinate_0) {
   code="//Taper back and front ends of the mesh\r\nfloat taper_u= .08;\r\nfloat taper_out = clamp(tex_coord.x/taper_u, .2, 1);\r\nfloat taper_in = clamp((1-tex_coord.x)/taper_u, .2, 1);\r\n\r\nreturn taper_out * taper_in;"
   desc="taper base, tip"
   description=taper
   outputType=CMOT_Float1
   showCode=true
   inputs [{
     inputName=tex_coord
     input {expression=&self:MaterialExpressionTextureCoordinate_0}
    }]
  }
  n43 = ScalarParameter() {
   defaultValue=0.05
   desc="Distance from swirl to add an extra thicken. Helps make mor of a puffy cloud feel. "
   group=SplineThicken
   parameterName=PuffDistance
   sortPriority=4
  }
  n44 = Constant()
  n45 = ScalarParameter() {
   desc="Amount of puff thickness to add to the swirl (not the tail). Helps make a puffy cloud look. "
   group=SplineThicken
   parameterName=PuffAmount
   sortPriority=5
  }
  n46 = Clamp(
   None=MaterialFunctionCall(Input=NamedRerouteUsage(), Input Low=Multiply(A=@n43) {constB=-1}, Input High=@n43, Target Low=@n44, Target High=@n45) {materialFunction=&RemapValueRange}.Result
   Min=@n44
   Max=@n45
  )
  n47 = Length(None=NamedRerouteUsage()) {is=class MaterialExpressionLength}
  n48 = Add(A=Multiply(A=Multiply(A=@n40, B=@n41) {constB=15}, B=@n42), B=Multiply(A=@n42, B=Multiply(A=@n46, B=@n47))) {constB=0}
  n49 = ScalarParameter() {defaultValue=1.1, desc="Multiply up overall spline thicken effect. ", group=SplineThicken, parameterName=MainThickness, sortPriority=0}
  MaterialExpressionReroute_4 = Reroute(None=NamedRerouteDeclaration(None=Multiply(A=@n48, B=@n49)) {name=SwirlSplineThicken, nodeColor=<rgba 1 0.8502 0.0038 1>})
  n50 = MaterialFunctionCall(
   WidthBase=@MaterialExpressionReroute_4
   WidthTip=@MaterialExpressionReroute_4
   Expand U or V UV Channel=StaticBool()
   AdditionalNormal=MaterialFunctionCall!(
    Height Map=TextureObject() {samplerType=SAMPLERTYPE_LinearColor, texture=/Game/Global/Textures/Default/Masks_and_Gradients/T_VertGradient}
    Normal Map Intensity=Constant() {r=100}
    Height Map UV Offset
    Coordinates
    Height Map Channel Selector
   ) {materialFunction=/Engine/Functions/Engine_MaterialFunctions03/Procedurals/NormalFromHeightmap}.Result
   FlattenPixelNormal=StaticBool() {value=true}
  ) {materialFunction=&SplineThicken}
  NamedRerouteDeclaration(None=@n50.Normal) {name=ThickenNormal, nodeColor=<rgba 0.2891 1 0 1>}
  MaterialExpressionSubtract_13 = Subtract(A=WorldPosition().XYZ, B=ParticlePositionWS())
  MaterialExpressionNamedRerouteUsage_33 = NamedRerouteUsage()
  n51 = ParticleRandom()
  MaterialExpressionReroute_31 = Reroute(None=@n25)
  MaterialExpressionReroute_33 = Reroute(None=@n26)
  MaterialExpressionReroute_34 = Reroute(None=@n27)
  n52 = Constant() {r=0.33}
  n53 = Subtract(B=@n52)
  n54 = Add(A=@n52)
  n55 = Constant() {r=1}
  n56 = If(
   A=MaterialFunctionCall(Input=@n51, seed=Constant() {r=21}) {desc=Some big one, materialFunction=&MX_RandomFloat}.Result
   B=OneMinus(None=@MaterialExpressionReroute_31.B)
   "A > B"=Multiply(A=@MaterialExpressionReroute_34.InputPin, B=@MaterialExpressionReroute_33.B2) {constB=15}
   "A < B"=Multiply(
    A=MaterialFunctionCall(Input=@n51, min=@n53, max=@n54, seed=@n55) {desc=Drift, materialFunction=&MX_RandomFloat}.Result
    B=@MaterialExpressionReroute_34.InputPin
   )
  ) {constB=0.93}
  MaterialExpressionReroute_32 = Reroute(None=@n56)
  n57 = NamedRerouteUsage()
  MaterialExpressionAdd_7 = Add(A=Multiply(A=Subtract(A=@n57)) {constB=2.5}, B=@MaterialExpressionReroute_5.Result)
  MaterialExpressionScalarParameter_37 = ScalarParameter() {
   defaultValue=1
   desc="If the spiral should face the camera (1), or derive a normale based on particle direction (0). "
   group=Spiral
   parameterName=Spiral_FaceCamera
  }
  n58 = Subtract(A=ParticlePositionWS(), B=CameraPositionWS())
  MaterialExpressionIf_0 = If(
   A=Constant() {r=1}
   "A > B"=@n58
   "A == B"=@n58
   "A < B"=MaterialFunctionCall() {materialFunction=&CameraDirectionVector}.Result
  ) {constB=0.5, desc=Face Vec To Particle or Cam Vec}
  MaterialExpressionParticleRelativeTime_0 = ParticleRelativeTime()
  MaterialExpressionNamedRerouteUsage_26 = NamedRerouteUsage()
  MaterialExpressionScalarParameter_38 = ScalarParameter() {
   defaultValue=100
   desc="This is usually the length of the original static mesh geometry on the x axis for spline thicken geometry. This will be used to calculate a lot of normalizing of distance values. "
   parameterName=StaticMeshSize
  }
  MaterialExpressionAdd_0 = Add(A=@MaterialExpressionReroute_1.Result)
  MaterialExpressionConstant_75 = Constant() {desc=center accelerate rotation, r=0.5}
  MaterialExpressionNamedRerouteUsage_3 = NamedRerouteUsage() {desc=PreventFlippingWhenCameraMoves}
  n59 = Custom!(
   vtx_at_origin=@MaterialExpressionSubtract_13
   part_offset=@MaterialExpressionNamedRerouteUsage_33
   radius_init=@MaterialExpressionReroute_32.Output
   radius_drift=@MaterialExpressionAdd_7
   FaceCamera=@MaterialExpressionScalarParameter_37
   CameraVec=@MaterialExpressionIf_0
   n_age=@MaterialExpressionParticleRelativeTime_0
   part_radius=@MaterialExpressionNamedRerouteUsage_26
   mesh_radius=@MaterialExpressionScalarParameter_38
   start_age=@MaterialExpressionReroute_1.Result
   end_age=@MaterialExpressionAdd_0
   rotation=@MaterialExpressionReroute_5.Result
   center_accel=@MaterialExpressionConstant_75
   static_cam_vec=@MaterialExpressionNamedRerouteUsage_3
  ) {
   code="n_age = 1-pow(1-n_age, 2); // over all have start fast and then slow\r\n\r\n//**Move to origin, and verts along paral direction \r\nfloat n_swirl_age = (n_age - start_age)/(end_age - start_age);\r\nfloat3 pdir = normalize(part_offset);\r\nfloat size = part_radius * mesh_radius;\r\nfloat age_offset_dist = size * n_swirl_age;\r\nVtxOriginDelta_out = pdir * age_offset_dist;\r\n\r\n//**Flatten Tip and calc rotation amount\r\nfloat vtx_dot = dot(pdir, vtx_at_origin + VtxOriginDelta_out); \r\nN_Vtx_SlidDist_out = vtx_dot/size;\r\nfloat c_accel = N_Vtx_SlidDist_out * center_accel + 1;\r\nif(vtx_dot > 0){\r\n    VtxOriginDelta_out -= pdir *vtx_dot;\r\n    float n_dist = vtx_dot/(size*n_swirl_age);\r\n    Rotate_out = (n_dist*c_accel) * n_age * rotation; \r\n}\r\n\r\n//**Initail vectors to rotate about (camera or direction based)\r\nfloat3 up_vec = 0, perp_axis = 0, radius_vec=0;\r\nif(FaceCamera==1) {\r\n    up_vec = normalize(CameraVec); \r\n    radius_vec = -normalize(cross(part_offset, static_cam_vec));\r\n    if(radius_vec.z < 0){\r\n        radius_vec *= -1;\r\n        Rotate_out *= -1;\r\n    }\r\n    perp_axis = up_vec;\r\n}\r\nelse{\r\n    up_vec = normalize(float3(0, 0, 1));\r\n    perp_axis = -normalize(cross(part_offset, up_vec));\r\n    radius_vec = normalize(cross(perp_axis, part_offset));\r\n}\r\n\r\n//**Position of center of the swirl\r\nfloat3 drift_vec = radius_vec * n_age * radius_drift; //swirl drift perp\r\nfloat3 rad_offset_vec = (radius_vec * radius_init) + drift_vec;  //final perp offset\r\n\r\n//**Swirl Vector\r\nSwirlAxis_out = perp_axis;\r\nSwirlRadVec_out = rad_offset_vec;\r\n\r\nreturn 1;"
   description=Swirl Calcs
   showCode=true
   additionalOutputs [
    {outputName=Rotate_out, outputType=CMOT_Float1}
    {outputName=SwirlAxis_out, outputType=CMOT_Float3}
    {outputName=SwirlRadVec_out, outputType=CMOT_Float3}
    {outputName=VtxOriginDelta_out, outputType=CMOT_Float3}
    {outputName=N_Vtx_SlidDist_out, outputType=CMOT_Float1}
   ]
   inputs [
    {
     inputName=vtx_at_origin
     input {expression=&self:MaterialExpressionSubtract_13}
    }
    {
     inputName=part_offset
     input {expression=&self:MaterialExpressionNamedRerouteUsage_33}
    }
    {
     inputName=radius_init
     input {expression=&self:MaterialExpressionReroute_32}
    }
    {
     inputName=radius_drift
     input {expression=&self:MaterialExpressionAdd_7}
    }
    {
     inputName=FaceCamera
     input {expression=&self:MaterialExpressionScalarParameter_37}
    }
    {
     inputName=CameraVec
     input {expression=&self:MaterialExpressionIf_0}
    }
    {
     inputName=n_age
     input {expression=&self:MaterialExpressionParticleRelativeTime_0}
    }
    {
     inputName=part_radius
     input {expression=&self:MaterialExpressionNamedRerouteUsage_26}
    }
    {
     inputName=mesh_radius
     input {expression=&self:MaterialExpressionScalarParameter_38}
    }
    {
     inputName=start_age
     input {expression=&self:MaterialGraph_1.MaterialGraphNode_Composite_6.randomize.MaterialExpressionReroute_1}
    }
    {
     inputName=end_age
     input {expression=&self:MaterialExpressionAdd_0}
    }
    {
     inputName=rotation
     input {expression=&self:MaterialGraph_0.MaterialGraphNode_Composite_6.randomize.MaterialExpressionReroute_5}
    }
    {
     inputName=center_accel
     input {expression=&self:MaterialExpressionConstant_75}
    }
    {
     inputName=static_cam_vec
     input {expression=&self:MaterialExpressionNamedRerouteUsage_3}
    }
   ]
  }
  NamedRerouteDeclaration(None=Multiply(A=NamedRerouteUsage(), B=@n59.SwirlRadVec_out)) {name=SwirlCenter, nodeColor=<rgba 1 0.8095 0 1>}
  NamedRerouteDeclaration(None=@n59.SwirlRadVec_out) {name=SwirlRadVec, nodeColor=<rgba 1 0.8095 0 1>}
  NamedRerouteDeclaration(None=Subtract(A=NamedRerouteUsage(), B=NamedRerouteUsage())) {name=SwirlNSTraveledPos, nodeColor=<rgba 1 0.8705 0 1>}
  n60 = DynamicParameter() {defaultValue=<rgba 1 0.9428 0.0132 1>, paramNames [posx,poxy,posz,None]}
  NamedRerouteDeclaration(None=@n60.A) {name=SwirlNSRadius, nodeColor=<rgba 1 0.8672 0 1>}
  n61 = ScalarParameter() {desc="Randomly flip spirals to either side to get more variation. ", group=Spiral, parameterName=Spiral_DoRandomFlip, sortPriority=100}
  n62 = Round(None=MaterialFunctionCall(Input=ParticleRandom(), seed=Constant() {r=15}) {materialFunction=&MX_RandomFloat}.Result)
  n63 = Constant() {r=1}
  n64 = If(A=@n61, "A > B"=Subtract(A=Multiply(A=@n62) {constB=2}), "A == B"=@n63, "A < B"=@n63) {constB=0.5, desc=Do Rand Flip}
  NamedRerouteDeclaration(None=If(
    A=ScalarParameter() {desc=Flip all spirals to other side., group=Spiral, parameterName=Spiral_Flip, sortPriority=99}
    "A > B"=Multiply(A=@n64) {constB=-1}
    "A == B"=@n64
    "A < B"=@n64
   ) {constB=0.5, desc=Reverse Flip}) {name=SwirlRadVecDoFlip, nodeColor=<rgba 1 0.9547 0 1>}
  NamedRerouteDeclaration(None=@n60.RGB) {name=ParticleNSPos, nodeColor=<rgba 1 0.7781 0 1>}
  NamedRerouteDeclaration(None=@MaterialExpressionMaterialFunctionCall_7.Normal) {name=SwirlNormalRotate, nodeColor=<rgba 0.5703 0 1 1>}
  NamedRerouteDeclaration(None=@n59.Rotate_out) {name=SwirlRotate, nodeColor=<rgba 1 0.8095 0 1>}
  NamedRerouteDeclaration(None=Subtract(A=@n59.VtxOriginDelta_out) {constB=0}) {name=SwirlVtxOriginDelta, nodeColor=<rgba 1 0.7771 0 1>}
  NamedRerouteDeclaration(None=@n59.N_Vtx_SlidDist_out) {name=SwirlNSlidDist, nodeColor=<rgba 1 0.9853 0 1>}
  MaterialExpressionNamedRerouteUsage_46 = NamedRerouteUsage()
  MaterialExpressionNamedRerouteUsage_42 = NamedRerouteUsage()
  MaterialExpressionScalarParameter_33 = ScalarParameter() {defaultValue=1, desc="How quickly spiral gets to center. ", group=Spiral, parameterName=SwirlInAmt}
  MaterialExpressionNamedRerouteUsage_27 = NamedRerouteUsage()
  MaterialExpressionScalarParameter_34 = ScalarParameter() {
   defaultValue=0.2
   desc="Poke out center of spiral to give dimention. Mostly usefull if not camera facing to give dimention. "
   group=Spiral_Misc
   parameterName=PokeOutAmount
  }
  MaterialExpressionTextureCoordinate_2 = TextureCoordinate()
  MaterialExpressionConvert_3 = Convert(None=@MaterialExpressionTextureCoordinate_2) {
   convertInputs [{
     type=Vector2
     expressionInput {
      expression=&self:MaterialExpressionTextureCoordinate_2
      inputName=None
      mask=0
      maskA=0
      maskB=0
      maskG=0
      maskR=0
      outputIndex=0
     }
    }]
   convertOutputs [{type=Scalar}]
  }
  MaterialExpressionScalarParameter_0 = ScalarParameter() {
   defaultValue=1
   desc="After spiral has hit center, causes a rippling out effect that nooks neat. Numbers past 1 will oversize the swirl. "
   group=Spiral_Misc
   parameterName=RippleOut
  }
  n65 = Custom!(
   rot_amt=@MaterialExpressionNamedRerouteUsage_46
   rad_vec=@MaterialExpressionNamedRerouteUsage_42
   swirl_in_amt=@MaterialExpressionScalarParameter_33
   axis=@MaterialExpressionNamedRerouteUsage_27
   poke_out_amt=@MaterialExpressionScalarParameter_34
   u=@MaterialExpressionConvert_3
   expand=@MaterialExpressionScalarParameter_0
  ) {
   code="float3 roll_vec = rad_vec * rot_amt * swirl_in_amt;\r\n//**As we get more towards the tail, our offset reduces, causing spiral expand\r\n//**to not go in as far, or visual push away from center\r\nfloat exp = ((1-u)*expand+(1-expand)) * length(rad_vec);\r\n//**will use len of roll_vec until mrs becomes smaller, pushing out.\r\nroll_vec = normalize(rad_vec) * min(length(roll_vec), exp);\r\n\r\nfloat3 extrude = axis * length(roll_vec) * poke_out_amt * (1/swirl_in_amt);\r\n\r\nreturn roll_vec + extrude;\r\n"
   description=swirl_roll
   showCode=true
   inputs [
    {
     inputName=rot_amt
     input {expression=&self:MaterialExpressionNamedRerouteUsage_46}
    }
    {
     inputName=rad_vec
     input {expression=&self:MaterialExpressionNamedRerouteUsage_42}
    }
    {
     inputName=swirl_in_amt
     input {expression=&self:MaterialExpressionScalarParameter_33}
    }
    {
     inputName=axis
     input {expression=&self:MaterialExpressionNamedRerouteUsage_27}
    }
    {
     inputName=poke_out_amt
     input {expression=&self:MaterialExpressionScalarParameter_34}
    }
    {
     inputName=u
     input {expression=&self:MaterialExpressionConvert_3}
    }
    {
     inputName=expand
     input {expression=&self:MaterialExpressionScalarParameter_0}
    }
   ]
  }
  NamedRerouteDeclaration(None=@n65) {name=SwirlRoll, nodeColor=<rgba 1 0.8996 0 1>}
  NamedRerouteDeclaration(None=@n59.SwirlAxis_out) {name=SwirlAxis, nodeColor=<rgba 1 0.8095 0 1>}
  NamedRerouteDeclaration(None=@MaterialExpressionSubtract_13) {name=SwirlVtxOriginPos, nodeColor=<rgba 1 0.8057 0 1>}
  n66 = DynamicParameter() {parameterIndex=1, defaultValue=<rgba 1 0.9428 0.0132 1>, paramNames [origposx,origpoxy,origposz,None]}
  NamedRerouteDeclaration(None=@n66.RGB) {name=ParticleNSInitPos, nodeColor=<rgba 1 0.8777 0 1>}
  n67 = DynamicParameter() {parameterIndex=2, defaultValue=<rgba 1 0.9428 0.0132 1>, paramNames [origposx,origpoxy,origposz,None]}
  NamedRerouteDeclaration(None=@n67.RGB) {name=SwirlStaticCamVec, nodeColor=<rgba 1 0.8564 0 1>}
  NamedRerouteDeclaration(None=@n50.WorldPositionOffset) {name=SwirlOffset_SplineThicken, nodeColor=<rgba 1 0.775 0 1>}
  NamedRerouteUsage()
  n68 = ScalarParameter() {
   desc="0 is Camera facing normals, 1 is from geom original orientation. If Spiral_FaceCamera is 1, WORLD SPACE MAY BE BROKEN.  "
   parameterName=NormalSelect_CameraOrWorld
  }
  Normalize(VectorInput=LinearInterpolate(
    A=MaterialFunctionCall() {materialFunction=&CameraDirectionVector}.Result
    B=NamedRerouteUsage()
    Alpha=@n68
   ))
  PinBase() {reroutePins [value {expression=&self:MaterialExpressionReroute_3}]}
  PinBase!(Result=@n35.Result) {reroutePins [Result {expression=&self:MaterialExpressionReroute_9}]}
  PinBase!() {reroutePins [value {expression=&self:MaterialGraph_2.MaterialGraphNode_Composite_6.randomize.MaterialExpressionReroute_0}]}
  n69 = ParticleRandom()
  n70 = Constant() {r=0.33}
  n71 = Subtract(B=@n70)
  n72 = Add(A=@n70)
  n73 = Constant() {r=9}
  n74 = ScalarParameter() {defaultValue=50, desc=Distance to fade in opacity of geom., group=Opacity, parameterName=CameraFadeLength, sortPriority=3}
  n75 = ScalarParameter() {defaultValue=100, desc=Dist from cam to start fading in geom, group=Opacity, parameterName=CameraFadeOffset, sortPriority=3}
  n76 = Multiply(A=TextureCoordinate() {uTiling=0.2, vTiling=0.15}) {constB=0.8}
  n77 = ParticleRandom()
  MaterialExpressionTextureSampleParameter2D_1 = TextureSampleParameter2D(UVs=Panner(Coordinate=Add(A=@n76, B=@n77), Time=Time()) {speedX=0.05}) {
   desc="Texture used to erode in and out the geometry when the particle is born and dies. "
   group=Opacity
   parameterName=Erode_Tex
   samplerType=SAMPLERTYPE_LinearColor
   sortPriority=13
   texture=/Game/Athena/Apollo/Water/Textures/02_Rounded/03_WaterFoam_Veroni_base_H
  }
  MaterialExpressionParticleRelativeTime_1 = ParticleRelativeTime()
  MaterialExpressionScalarParameter_26 = ScalarParameter() {defaultValue=0.33, desc="Normalized age from age 0 to erode in geom. ", group=Opacity, parameterName=Erode_RampInNAge, sortPriority=10}
  MaterialExpressionScalarParameter_27 = ScalarParameter() {defaultValue=0.7, desc=Normalized age to begin eroding out geom, group=Opacity, parameterName=Erode_RampOutNAge, sortPriority=11}
  MaterialExpressionScalarParameter_28 = ScalarParameter() {
   defaultValue=0.01
   desc="How smooth erode edges are. \r\nSmaller numbers will have sharper edges as erodes. "
   group=Opacity
   parameterName=Erode_TexRemapRange
   sortPriority=12
  }
  n78 = Custom!(
   tx=@MaterialExpressionTextureSampleParameter2D_1.R
   ntime=@MaterialExpressionParticleRelativeTime_1
   RampIn=@MaterialExpressionScalarParameter_26
   RampOut=@MaterialExpressionScalarParameter_27
   RemapRange=@MaterialExpressionScalarParameter_28
  ) {
   code="float ramp_in = RampIn;\r\nfloat ramp_out = RampOut;\r\nfloat range = RemapRange;\r\nfloat power = 2;\r\n\r\n//2 0-1 remaps, 1)time,  2) texture using remap time and range \r\nfloat nt_out = (ntime-ramp_out)/(1-ramp_out);\r\nnt_out = 1-pow(max(1-nt_out, 0), power);\r\nfloat remap_out = (tx-nt_out)*(1/range);\r\nremap_out = clamp(remap_out, 0, 1);\r\n\r\nfloat nt_in = (ntime/ramp_in); //0-1\r\nnt_in = pow(max(1-nt_in, 0), power);\r\nfloat remap_in = (tx-nt_in)*(1/range);\r\nremap_in = clamp(remap_in, 0, 1);\r\n\r\nreturn remap_in * remap_out;\r\n\r\n\r\n"
   showCode=true
   inputs [
    {
     inputName=tx
     input {
      expression=&self:MaterialExpressionTextureSampleParameter2D_1
      mask=1
      maskR=1
      outputIndex=1
     }
    }
    {
     inputName=ntime
     input {expression=&self:MaterialExpressionParticleRelativeTime_1}
    }
    {
     inputName=RampIn
     input {expression=&self:MaterialExpressionScalarParameter_26}
    }
    {
     inputName=RampOut
     input {expression=&self:MaterialExpressionScalarParameter_27}
    }
    {
     inputName=RemapRange
     input {expression=&self:MaterialExpressionScalarParameter_28}
    }
   ]
  }
  n79 = Multiply(
   A=@n30.A
   B=Multiply(
    A=MaterialFunctionCall(Fade Length=@n74, Fade Offset=@n75) {materialFunction=&CameraDepthFade}.Result
    B=@n78
   )
  )
  MaterialExpressionTextureCoordinate_5 = TextureCoordinate()
  n80 = Convert(None=@MaterialExpressionTextureCoordinate_5) {
   nodeName=Break Float2
   convertInputs [{
     type=Vector2
     expressionInput {
      expression=&self:MaterialExpressionTextureCoordinate_5
      inputName=None
      mask=0
      maskA=0
      maskB=0
      maskG=0
      maskR=0
      outputIndex=0
     }
    }]
   convertOutputs [{type=Scalar}, {type=Scalar}]
  }
  n81 = OneMinus(None=@n80)
  n82 = ScalarParameter() {
   defaultValue=1
   desc="How fast opacity ramp slides off end of the swirl tail. "
   group=Opacity
   parameterName=TailOpac_SlideSpeed
   sortPriority=17
  }
  n83 = Multiply(A=ParticleRelativeTime(), B=@n82)
  n84 = ScalarParameter() {
   defaultValue=0.3
   desc="When particle age is 0, U value where the opacity begins to ramp up from TailOac_OpacStart"
   group=Opacity
   parameterName=TailOpac_RampStart
   sortPriority=14
  }
  n85 = ScalarParameter() {defaultValue=0.6, desc="When age is 0, U value when ramp is fully opaque.", group=Opacity, parameterName=TailOpac_RampEnd, sortPriority=15}
  n86 = ScalarParameter() {defaultValue=0.2, desc=Opacity at TailOpac_RampStart, group=Opacity, parameterName=TailOpac_OpacityStart, sortPriority=16}
  n87 = MaterialFunctionCall(Input=Clamp(None=Add(A=@n81, B=@n83)), Input Low=@n84, Input High=@n85, Target Low=@n86) {materialFunction=&RemapValueRange}
  n88 = ScalarParameter() {defaultValue=10, desc=Distance tranparency ramps out to the ground, group=Opacity, parameterName=GroundFadeDist, sortPriority=2}
  n89 = Multiply(A=ParticleColor().A, B=Multiply(A=Clamp(None=@n87.Result), B=DepthFade(FadeDistance=@n88) {fadeDistanceDefault=10}))
  MaterialExpressionReroute_0 = Reroute(None=Multiply(A=Multiply(A=@n79, B=@n89), B=ScalarParameter() {defaultValue=1, group=Opacity, parameterName=Opacity, sortPriority=1}))
  PinBase!(Result=Multiply(
    A=MaterialFunctionCall(Input=@n69, min=@n71, max=@n72, seed=@n73) {desc=Drift, materialFunction=&MX_RandomFloat}.Result
    B=@MaterialExpressionReroute_0.value
   )) {reroutePins [Result {expression=&self:MaterialGraph_2.MaterialGraphNode_Composite_6.randomize.MaterialExpressionReroute_1}]}
  PinBase!() {reroutePins [
    B {expression=&self:MaterialExpressionReroute_31}
    B2 {expression=&self:MaterialExpressionReroute_33}
    InputPin {expression=&self:MaterialExpressionReroute_34}
   ]}
  PinBase!(Output=@n56) {reroutePins [Output {expression=&self:MaterialExpressionReroute_32}]}
  PinBase!() {reroutePins [value {expression=&self:MaterialGraph_0.MaterialGraphNode_Composite_3.randomize.MaterialExpressionReroute_4}]}
  n90 = ParticleRandom()
  n91 = Constant() {r=0.33}
  n92 = Subtract(B=@n91)
  n93 = Add(A=@n91)
  n94 = Constant() {r=100}
  PinBase!(Result=Multiply(
    A=MaterialFunctionCall(Input=@n90, min=@n92, max=@n93, seed=@n94) {desc=Drift, materialFunction=&MX_RandomFloat}.Result
    B=@MaterialExpressionReroute_4.value
   )) {reroutePins [Result {expression=&self:MaterialGraph_0.MaterialGraphNode_Composite_3.randomize.MaterialExpressionReroute_5}]}
  PinBase!() {reroutePins [value {expression=&self:MaterialGraph_1.MaterialGraphNode_Composite_6.randomize.MaterialExpressionReroute_0}]}
  n95 = Constant() {r=0.25}
  PinBase!(Result=MaterialFunctionCall(
    Input=ParticleRandom()
    min=Subtract(A=@MaterialExpressionReroute_0.value, B=@n95)
    max=Add(A=@n95, B=@MaterialExpressionReroute_0.value)
    seed=Constant() {r=3}
   ) {desc=Drift, materialFunction=&MX_RandomFloat}.Result) {reroutePins [Result {expression=&self:MaterialGraph_1.MaterialGraphNode_Composite_6.randomize.MaterialExpressionReroute_1}]}
  PinBase!() {reroutePins [value {expression=&self:MaterialGraph_0.MaterialGraphNode_Composite_6.randomize.MaterialExpressionReroute_4}]}
  n96 = ParticleRandom()
  n97 = Constant() {r=0.3}
  n98 = Subtract(B=@n97)
  n99 = Add(A=@n97)
  n100 = Constant() {r=7}
  PinBase!(Result=Multiply(
    A=MaterialFunctionCall(Input=@n96, min=@n98, max=@n99, seed=@n100) {desc=Drift, materialFunction=&MX_RandomFloat}.Result
    B=@MaterialExpressionReroute_4.value
   )) {reroutePins [Result {expression=&self:MaterialGraph_0.MaterialGraphNode_Composite_6.randomize.MaterialExpressionReroute_5}]}
 ]
}
