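# Niagara system for the FlickLedge character body, emitting smoke and swirl effects sampled from its skeletal mesh.
# Smoke spawns when nearly still, SwirlCloak scales with movement, and SmokeSky and SwirlSky replace them during fast falls.
# Local modules handle low placement, wake suction and above-waist culling, with visibility tags set by vehicle state.
# Time-of-day curves tint particles, while swirl meshes orient along motion and pass positions and camera vector to materials.
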NiagaraSystem /BRCosmetics/Characters/Player/Male/Medium/Bodies/M_MED_FlickLedge/FX/NS_Char_FlickLedge_Swirl
dependencies {
 usedDataInterfaces [
  &NiagaraDataInterfaceSpriteRendererInfo
  &NiagaraDataInterfaceCurve
  &NiagaraDataInterfaceColorCurve
  &NiagaraDataInterfaceCamera
 ]
 usedDynamicInputs []
 usedModules [
  &SystemState
  &BlendSpaceHelper
  &self:SystemSpawnScript.SystemScriptSource.SystemScriptGraph.NiagaraNodeAssignment_0.SetVariables_#1
  &self:SystemSpawnScript.SystemScriptSource.SystemScriptGraph.NiagaraNodeAssignment_1.SetVariables_#2
  /Game/Effects/Niagara/Modules/Contrails/FN_Contrail_Setup
  &self:MyBoneData
 ]
}
emitters {
 Smoke {
  emitterUpdateScript [
   &EmitterState()
   &SpawnRate(
    SpawnRate=&RemapRange(
     Input Value To Remap=&VectorLength(Vector=$Engine.Owner.Velocity)
     Input Min=5
     Input Max=10
     Output Min=20
     Output Max=0
    )
    Use Spawn Probability=false
   )
  ]
  particleSpawnScript [
   &InitializeParticle(
    Sprite Rotation Mode=::NewEnumerator2
    Sprite Size Mode=::NewEnumerator3
    Sprite UV Mode=::NewEnumerator3
    Uniform Sprite Size=&RemapRange(
     Input Value To Remap=&VectorLength(Vector=$Engine.Owner.Velocity)
     Input Min=50
     Input Max=200
     Output Min=100
     Output Max=150
    )
    UsePositionOffset=false
    Write Material Random=true
   )
   &SkeletalMeshLocation(
    MaskSpawnGroup=false
    Mesh Sampling Type=::NewEnumerator6
    Orientation Sampling=::NewEnumerator1
    OverrideRandomSeed=false
    Sampled Position Offset=<5 0 0>
    Skeletal Mesh=$System.Skeletal Mesh Reader
   )
   &AddVelocity(
    Rotation Coordinate Space=::Simulation
    Velocity=&RotateVectorByQuaternion(Vector To Rotate=<1 0 0>, Quaternion=$Particles.SkeletalMeshLocation.SampledRotation)
    Velocity Speed Scale=&RandomRangeFloat(Minimum=10, Maximum=15)
   )
   &OffsetPosition(Position Offset=&Multiply_VectorByFloat(
     Vector=/Niagara/DynamicInputs/Vectors/CrossProduct(Vector A:Vector3f=$Particles.Velocity, Vector B:Vector3f=<0 0 1>, Normalize Incoming Vectors:NiagaraBool=true)
     Float=&RandomRangeFloat(Minimum=-60, Maximum=60)
    )) {module=OffsetPosition001}
   &self:FastAndLow(Always Low=true, BoneHeightOffset=-75, SpeedMaxRamp=0, SpeedMinRemap=0) {module=FastAndLow}
   /SetVariables(Particles.Position=&MakeVector(
     X=&MakeFloatFromVector(Vector=Linked() {linkedVariable {name=Particles.Position, type=NiagaraPosition}}, Channel=::NewEnumerator0)
     Y=&MakeFloatFromVector(Vector=Linked() {linkedVariable {name=Particles.Position, type=NiagaraPosition}}, Channel=::NewEnumerator1)
     Z=&Max_Float(
      A:NiagaraFloat=&Add_Float(
       A=0
       B=&MakeFloatFromVector(Vector=Linked() {linkedVariable {name=Engine.Owner.Position, type=NiagaraPosition}}, Channel=::NewEnumerator2)
      )
      B:NiagaraFloat=&MakeFloatFromVector(Vector=Linked() {linkedVariable {name=Particles.Position, type=NiagaraPosition}}, Channel=::NewEnumerator2)
     )
    )) {isSetParameters=true}
   &self:SuckIntoWake() {module=SuckIntoWake}
   /SetVariables(Particles.VisibilityTag=&MakeCustomIntFromBool(Bool=$User.bisInVehicle, True INT=1, False INT=0)) {isSetParameters=true}
  ]
  particleUpdateScript [
   &ParticleState()
   &SubUVAnimation(
    End Frame Range Override=3
    Evaluation Type=::NewEnumerator0
    "Sprite Renderer "=DataInterface() {propertyValues {SpriteRenderer=&self:Smoke_0.NiagaraSpriteRendererProperties_1}}
    SubUV Animation Mode:ENiagara_SubUVLookupModeV2=::NewEnumerator2
    UseEndFrame=true
    UseStartFrame=true
   )
   &ScaleSpriteSize(Uniform Curve Sprite Scale=DataInterface() {propertyValues {Curve {keys [
        {arriveTangent=0.3413, interpMode=RCIM_Cubic, leaveTangent=0.3413, tangentMode=RCTM_User, time=0, value=0.3}
        {arriveTangent=0.0236, interpMode=RCIM_Cubic, leaveTangent=0.0236, tangentMode=RCTM_User, time=1, value=1}
       ]}}})
   /SetVariables(Particles.Lifetime=&Multiply_Float(
     A=$Particles.Lifetime
     B=&RemapRange(
      Input Value To Remap=&DistanceBetweenPositions(A=$Particles.Position, B=$Engine.Emitter.SimulationPosition)
      Input Min=50
      Input Max=90
      Output Min=1
      Output Max=&Subtract_Float(A=1, B=&Multiply_Float(A=$Engine.DeltaTime, B=1))
     )
    )) {isSetParameters=true}
   &self:KillAboveWaist(BoneZOffsetLife=-40, LifeMultDown=7) {module=KillAboveWaist}
   &SpriteRotationRate(Rotation Rate=&Multiply_FloatByInt(
     Float=&RandomRangeFloat(Minimum=55, Maximum=110)
     Integer=/Niagara/DynamicInputs/Subtract/Subtract_Integer(
      A:NiagaraInt32=&Multiply_Int(A=2, B=&RandomRangeInt(Minimum=0, Maximum=1))
      B:NiagaraInt32=1
     )
    ))
   &CurlNoiseForce(Bias Noise=false, Noise Frequency=30, Noise Strength=60, Pan Noise=true, Pan Noise Field=<0 0 -3>)
   &Drag(Drag=2, Use Linear Drag=true, Use Rotational Drag=false)
   &Color(
    Color=&ColorFromCurve(ColorCurve=DataInterface() {propertyValues {
       AlphaCurve {keys [<at 0 0 RCIM_Cubic>,<at 0.4196 0.8>,<at 1 0>]}
       BlueCurve {keys [<at 0 0.027 RCIM_Cubic>]}
       GreenCurve {keys [<at 0 0.022 RCIM_Cubic>]}
       RedCurve {keys [<at 0 0.0184 RCIM_Cubic>]}
      }})
    Scale Color=&MakeVectorFromLinearColorRGB(Color=&ColorFromCurve(
      ColorCurve=DataInterface() {propertyValues {
        AlphaCurve {keys [<at 0 1 RCIM_Cubic>]}
        BlueCurve {keys [<at 0 1>,<at 0.0836 1>,<at 0.1161 0.5223>,<at 0.1358 0.9857>,<at 0.25 1.1>,<at 0.5 1>,<at 0.75 0.9323>,<at 0.8278 0.5466>,<at 1 1 RCIM_Cubic>]}
        GreenCurve {keys [<at 0 0.6418>,<at 0.0836 0.7026>,<at 0.1161 0.6832>,<at 0.1358 0.9061>,<at 0.25 1.1>,<at 0.5 1>,<at 0.75 0.867>,<at 0.8278 0.7704>,<at 1 0.6418 RCIM_Cubic>]}
        RedCurve {keys [<at 0 0.3862>,<at 0.0836 0.4903>,<at 0.1161 1.1>,<at 0.1358 1.0203>,<at 0.25 1.1>,<at 0.5 1>,<at 0.75 0.8128>,<at 0.8278 1.1>,<at 1 0.3862 RCIM_Cubic>]}
       }}
      CurveIndex=Linked() {linkedVariable {name=NPC.FortniteNPC.TimeOfDay_0-1, type=NiagaraFloat}}
     ))
   )
   /SetVariables(
    Particles.GlowColor=&MultiplyLinearColor(Linear Color=$Particles.Color, Scale Factor=<1 1 1 0.2>)
    Particles.GlowScale:Vector2f=&Multiply_Vector2D_ByFloat(Float=2, Vector2D=$Particles.SpriteSize)
   ) {isSetParameters=true}
   &CameraOffset(Camera Offset Amount=15)
   &SolveForcesAndVelocity(Clamp Velocity=false, Limit Acceleration=false)
  ]
  rendererClasses [&NiagaraSpriteRendererProperties]
  renderers [
   {rendererClass=&NiagaraSpriteRendererProperties}
   {rendererClass=&NiagaraSpriteRendererProperties, rendererIndex=1}
  ]
 }
 SmokeSky {
  emitterUpdateScript [
   &EmitterState()
   &SpawnRate(
    SpawnRate=&RemapRange(
     Input Value To Remap=&MakeFloatFromVector(Vector=$Engine.Owner.Velocity, Channel=::NewEnumerator2)
     Input Min=-450
     Input Max=-500
     Output Min=0
     Output Max=80
    )
    Use Spawn Probability=false
   )
  ]
  particleSpawnScript [
   &InitializeParticle(
    Lifetime Max=0.2
    Lifetime Min=0.2
    Lifetime Mode=::NewEnumerator1
    Sprite Rotation Mode=::NewEnumerator2
    Sprite Size Mode=::NewEnumerator3
    Uniform Sprite Size=100
    UsePositionOffset=false
    Write Material Random=true
   ) {module=InitializeParticle001}
   &SkeletalMeshLocation(
    MaskSpawnGroup=false
    Mesh Sampling Type=::NewEnumerator6
    Orientation Sampling=::NewEnumerator1
    OverrideRandomSeed=false
    Skeletal Mesh=$System.Skeletal Mesh Reader
   ) {module=SkeletalMeshLocation001}
   &InheritVelocity(
    Inherited Velocity Amount Scale=&Multiply_VectorByFloat(
     Vector=<1 1 1>
     Float=&RemapRange(
      Input Value To Remap=&VectorLength(Vector=$Engine.Owner.Velocity)
      Input Min=1500
      Input Max=3000
      Output Min=0.95
      Output Max=1
     )
    )
    Inherited Velocity Speed Limit=150
    Speed Limit=false
    UseSpeedThreshold=false
   )
   &OffsetPosition(Position Offset=&Multiply_VectorByFloat(Vector=$Particles.Velocity, Float=0.005)) {module=OffsetPosition002}
   /SetVariables(Particles.VisibilityTag=&MakeCustomIntFromBool(Bool=$User.bisInVehicle, True INT=1, False INT=0)) {isSetParameters=true}
  ]
  particleUpdateScript [
   &ParticleState()
   &self:SuckIntoWake(VelSuckInAmt=80) {module=SuckIntoWake}
   &SubUVAnimation(
    End Frame Range Override=3
    Evaluation Type=::NewEnumerator0
    "Sprite Renderer "=DataInterface() {propertyValues {SpriteRenderer=&self:SmokeSky_0.NiagaraSpriteRendererProperties_1}}
    SubUV Animation Mode:ENiagara_SubUVLookupModeV2=::NewEnumerator2
    UseEndFrame=true
    UseStartFrame=true
   )
   &Color(Color=&MultiplyLinearColor(
     Linear Color=&MultiplyLinearColor(
      Linear Color=&ColorFromCurve(ColorCurve=DataInterface() {propertyValues {
         AlphaCurve {keys [<at 0 0 RCIM_Cubic>,<at 0.2725 0.3>,<at 1 0>]}
         BlueCurve {keys [<at 0 0.03 RCIM_Cubic>]}
         GreenCurve {keys [<at 0 0.0238 RCIM_Cubic>]}
         RedCurve {keys [<at 0 0.0226 RCIM_Cubic>]}
        }})
      Scale Factor=&MakeVector4_FromLinearColor(LinearColor=&FNTimeOfDayColor(
        Night Color=<rgba 0.3601 0.6202 0.9472 1>
        Morning Color=<rgba 1.2734 1.0921 1.0952 1>
        Day Color=<rgba 1 1 1 1>
        Evening Color=<rgba 1 0.8213 0.785 1>
       ))
     )
     Scale Factor=&MakeVector4_FromLinearColor(LinearColor=$Particles.Initial.Color)
    ))
   &ScaleSpriteSize(Uniform Curve Sprite Scale=DataInterface() {propertyValues {Curve {keys [
        {arriveTangent=0.3413, interpMode=RCIM_Cubic, leaveTangent=0.3413, tangentMode=RCTM_User, time=0, value=0.3}
        {arriveTangent=0.0236, interpMode=RCIM_Cubic, leaveTangent=0.0236, tangentMode=RCTM_User, time=1, value=1}
       ]}}})
   /SetVariables(
    Particles.GlowColor=&MultiplyLinearColor(Linear Color=$Particles.Color, Scale Factor=<1 1 1 0.7>)
    Particles.GlowScale:Vector2f=&Multiply_Vector2D_ByFloat(Float=2, Vector2D=$Particles.SpriteSize)
   ) {isSetParameters=true}
   &SpriteRotationRate(Rotation Rate=&Multiply_FloatByInt(
     Float=&RandomRangeFloat(Minimum=100, Maximum=200)
     Integer=&Subtract_Integer(
      A:NiagaraInt32=&Multiply_Int(A=2, B=&RandomRangeInt(Minimum=0, Maximum=1))
      B:NiagaraInt32=1
     )
    ))
   &Drag(Drag=0.4, Use Linear Drag=true, Use Rotational Drag=true)
   &SolveForcesAndVelocity(Clamp Velocity=false, Limit Acceleration=false)
  ]
  rendererClasses [&NiagaraSpriteRendererProperties]
  renderers [
   {rendererClass=&NiagaraSpriteRendererProperties}
   {rendererClass=&NiagaraSpriteRendererProperties, rendererIndex=1}
  ]
 }
 SwirlCloak {
  emitterUpdateScript [
   &EmitterState()
   &SpawnRate(
    SpawnRate=&Multiply_Float(
     A=&RemapRange(
      Input Value To Remap=&Abs_Float(Float:NiagaraFloat=&MakeFloatFromVector(Vector=$Engine.Owner.Velocity, Channel=::NewEnumerator2))
      Input Min=450
      Input Max=460
      Output Min=1
      Output Max=0
     )
     B=&RemapRange(
      Input Value To Remap=&VectorLength(Vector=$System.TranslationDeltaWorld)
      Input Min=5
      Input Max=15
      Output Min=12
      Output Max=65
     )
    )
    Use Spawn Probability=false
   )
  ]
  particleSpawnScript [
   &InitializeParticle(
    Lifetime=1.7
    Mesh Scale Mode=::NewEnumerator3
    Mesh Uniform Scale=0.8
    Sprite Size Mode=::NewEnumerator3
    Uniform Sprite Size=3
    UsePositionOffset=false
    Write Material Random=true
   ) {module=InitializeParticle001}
   &SkeletalMeshLocation(
    MaskSpawnGroup=false
    Mesh Sampling Type=::NewEnumerator6
    Orientation Sampling=::NewEnumerator1
    OverrideRandomSeed=false
    Sampled Position Offset=<5 0 0>
    Skeletal Mesh=$System.Skeletal Mesh Reader
   ) {module=SkeletalMeshLocation001}
   &AddVelocity(
    Rotation Coordinate Space=::Simulation
    Velocity=&RotateVectorByQuaternion(Vector To Rotate=<1 0 0>, Quaternion=$Particles.SkeletalMeshLocation001.SampledRotation)
    Velocity Speed Scale=10
   )
   &OffsetPosition(Position Offset=&Multiply_VectorByFloat(
     Vector=&CrossProduct(Vector A:Vector3f=$Particles.Velocity, Vector B:Vector3f=<0 0 1>, Normalize Incoming Vectors:NiagaraBool=true)
     Float=&RandomRangeFloat(Minimum=-60, Maximum=50)
    )) {module=OffsetPosition001}
   &self:FastAndLow(BoneHeightOffset=-80, SpeedMaxRamp=300, SpeedMinRemap=150) {module=FastAndLow}
   &InheritVelocity(
    Inherited Velocity Amount Scale=&Multiply_VectorByFloat(Vector=<1 1 1>, Float=1)
    Inherited Velocity Speed Limit=150
    Speed Limit=false
    UseSpeedThreshold=false
   )
   /SetVariables(Particles.OrigPosition:NiagaraPosition=&SubtractVectorFromPosition(
     Position=$Particles.Position
     Vector=&Multiply_VectorByFloat(Vector=$Particles.Velocity, Float=$Engine.DeltaTime)
    )) {isSetParameters=true, module=SetVariables_#3}
   /SetVariables(Particles.StaticCamVec:Vector3f=&ReturnCameraVector(Camera Vectors=::NewEnumerator1, Coordinate Space=::Simulation)) {isSetParameters=true, module=SetVariables_#4}
   /SetVariables(Particles.Age:NiagaraFloat=&RemapRange(
     Input Value To Remap=&VectorLength(Vector=$Engine.Owner.Velocity)
     Input Min=10
     Input Max=15
     Output Min=0
     Output Max=0.9
    )) {isSetParameters=true}
   &self:SuckIntoWake(VelSuckInAmt=50) {module=SuckIntoWake}
   /SetVariables(Particles.VisibilityTag=&MakeCustomIntFromBool(Bool=$User.bisInVehicle, True INT=1, False INT=0)) {isSetParameters=true}
  ]
  particleUpdateScript [
   &ParticleState()
   &UpdateMeshOrientation(
    Facing Direction=&Subtract_Position(A=$Particles.OrigPosition, B=$Particles.Position)
    Facing Interpolation=false
    Orientation Method=::NewEnumerator1
    Override Quat=false
    Side Direction:Vector3f=&NormalizeVector(
     Vector To Normalize=&CrossProduct(
      Vector A:Vector3f=&ReturnCameraVector(Camera Vectors=::NewEnumerator3, Coordinate Space=::Simulation)
      Vector B:Vector3f=<0 0 -1>
      Normalize Incoming Vectors:NiagaraBool=true
     )
     Safe Normalize=true
     Safe Fallback Threshold=0
     Safe Fallback Vector=<0 0 0.001>
    )
    Stabilization Mode:ENiagara_OrientVectorCount=::NewEnumerator2
    TransformFacing=false
    TransformSide=false
   )
   /SetVariables(Particles.Scale=/NiagaraFluids/DynamicInput/MinVectorByLength(
     A:Vector3f=<2 2 2>
     B:Vector3f=&Add_Vector(
      A=$Particles.Scale
      B=&Multiply_VectorByFloat(
       Vector=<1 1 1>
       Float=&RemapRange(
        Input Value To Remap=&VectorLength(Vector=$Particles.Velocity)
        Input Min=50
        Input Max=150
        Output Min=0
        Output Max=0.15
       )
      )
     )
    )) {isSetParameters=true}
   /SetVariables(
    Particles.Color=&MultiplyLinearColor(
     Linear Color=$Particles.Color
     Scale Factor=&MakeVector4FromVectorAndFloat(
      "Vector (RGB)"=<1 1 1>
      "Float (Alpha)"=&MakeCustomFloatFromBool(
       Bool=&SetBoolByFloatComparison(A=65, B=$System.MyBoneToOwnerDist)
       True Float=1
       False Float=&RemapRange(
        Input Value To Remap=&DistanceBetweenPositions(A=$Particles.Position, B=$Engine.Emitter.SimulationPosition)
        Input Min=60
        Input Max=65
        Output Min=1
        Output Max=&Subtract_Float(A=1, B=&Multiply_Float(A=$Engine.DeltaTime, B=5))
       )
      )
     )
    )
    Particles.Lifetime=&Multiply_Float(
     A=$Particles.Lifetime
     B=&MakeCustomFloatFromBool(
      Bool=&SetBoolByFloatComparison(A=65, B=$System.MyBoneToOwnerDist)
      True Float=1
      False Float=&RemapRange(
       Input Value To Remap=&DistanceBetweenPositions(A=$Particles.Position, B=$Engine.Emitter.SimulationPosition)
       Input Min=60
       Input Max=80
       Output Min=1
       Output Max=&Subtract_Float(A=1, B=&Multiply_Float(A=$Engine.DeltaTime, B=5))
      )
     )
    )
   ) {isSetParameters=true}
   &self:KillAboveWaist(BoneZOffsetLife=-10, BoneZOffsetOpacity=-50, LifeMultDown=20) {module=KillAboveWaist}
   &ScaleVelocity(
    Coordinate Space:ENiagaraCoordinateSpace=::Simulation
    Velocity Scale=&Multiply_VectorByFloat(
     Vector=<1 1 1>
     Float=&RemapRange(
      Input Value To Remap=&MakeFloatFromVector(Vector=&Subtract_Position(A=$Particles.Position, B=$Engine.Owner.Position), Channel=::NewEnumerator2)
      Input Min=20
      Input Max=0
      Output Min=1
      Output Max=0.15
     )
    )
   )
   &Color(
    Color=&ColorFromCurve(
     ColorCurve=DataInterface() {propertyValues {
       AlphaCurve {keys [<at 0 1 RCIM_Cubic>]}
       BlueCurve {keys [<at 0 0.8847>,<at 0.086 1.1>,<at 0.1178 0.5223>,<at 0.1358 0.9857>,<at 0.25 1.1>,<at 0.5 1>,<at 0.75 0.9323>,<at 0.8278 0.5466>,<at 1 0.8847 RCIM_Cubic>]}
       GreenCurve {keys [<at 0 0.5718>,<at 0.086 0.6894>,<at 0.1178 0.6832>,<at 0.1358 0.9061>,<at 0.25 1.1>,<at 0.5 1>,<at 0.75 0.867>,<at 0.8278 0.7704>,<at 1 0.5718 RCIM_Cubic>]}
       RedCurve {keys [<at 0 0.3996>,<at 0.086 0.4673>,<at 0.1178 1.1>,<at 0.1358 1.0203>,<at 0.25 1.1>,<at 0.5 1>,<at 0.75 0.8128>,<at 0.8278 1.1>,<at 1 0.3996 RCIM_Cubic>]}
      }}
     CurveIndex=Linked() {linkedVariable {name=NPC.FortniteNPC.TimeOfDay_0-1, type=NiagaraFloat}}
    )
    Scale Alpha=&MakeFloatFromLinearColor(LinearColor=$Particles.Color, Channel:ENiagaraFloatFromLinearColorOptions=::NewEnumerator3)
    Scale Color=&Multiply_VectorByFloat(
     Vector=<1 1 1>
     Float=&MakeCustomFloatFromBool(Bool=$User.bIsFrontEndPreview, True Float=0.7, False Float=1)
    )
   )
   &DynamicMaterialParameters(
    Index 0 Param 4=&MakeFloatFromVector(Vector=$Particles.Scale, Channel=::NewEnumerator0)
    Index 0 Vector 1:NiagaraPosition=$Particles.Position
    Index 1 Vector 1:NiagaraPosition=$Particles.OrigPosition
    Index 2 Vector 1:Vector3f=$Particles.StaticCamVec
    Module.Index0Float3Write=true
    Module.Index0VectorWrite=true
    Module.Index1Float3Write=false
    Module.Index1VectorWrite=true
    Module.Index2Float3Write=false
    Module.Index2VectorWrite=true
    Write Parameter Index 0=::NewEnumerator3
    Write Parameter Index 1=::NewEnumerator3
    Write Parameter Index 2=::NewEnumerator3
   )
   &CameraOffset(Camera Offset Amount=15)
   &Drag(Drag=2, Use Linear Drag=true, Use Rotational Drag=true)
   &SolveForcesAndVelocity(Clamp Velocity=false, Limit Acceleration=false)
  ]
  rendererClasses [&NiagaraMeshRendererProperties]
  renderers [{rendererClass=&NiagaraMeshRendererProperties, rendererIndex=0}]
 }
 SwirlSky {
  emitterUpdateScript [
   &EmitterState()
   &SpawnRate(
    SpawnRate=&RemapRange(
     Input Value To Remap=&MakeFloatFromVector(Vector=$Engine.Owner.Velocity, Channel=::NewEnumerator2)
     Input Min=-450
     Input Max=-500
     Output Min=0
     Output Max=20
    )
    Use Spawn Probability=false
   )
  ]
  particleSpawnScript [
   &InitializeParticle(
    Lifetime Max=0.3
    Lifetime Min=0.2
    Lifetime Mode=::NewEnumerator1
    Mesh Scale Mode=::NewEnumerator3
    Mesh Uniform Scale=1.5
    Sprite Size Mode=::NewEnumerator3
    Uniform Sprite Size=3
    UsePositionOffset=false
    Write Material Random=true
   ) {module=InitializeParticle001}
   &SkeletalMeshLocation(
    MaskSpawnGroup=false
    Mesh Sampling Type=::NewEnumerator6
    Orientation Sampling=::NewEnumerator1
    OverrideRandomSeed=false
    Skeletal Mesh=$System.Skeletal Mesh Reader
   ) {module=SkeletalMeshLocation001}
   &InheritVelocity(
    Inherited Velocity Amount Scale=&Multiply_VectorByFloat(
     Vector=<1 1 1>
     Float=&RemapRange(
      Input Value To Remap=&VectorLength(Vector=$Engine.Owner.Velocity)
      Input Min=1500
      Input Max=3000
      Output Min=0.9
      Output Max=0.96
     )
    )
    Inherited Velocity Speed Limit=150
    Speed Limit=false
    UseSpeedThreshold=false
   )
   &OffsetPosition(Position Offset=&Multiply_VectorByFloat(Vector=$Particles.Velocity, Float=0.008)) {module=OffsetPosition002}
   &OffsetPosition(Position Offset=&RandomVector(Vector Scale=5)) {module=OffsetPosition001}
   /SetVariables(Particles.OrigPosition:NiagaraPosition=&SubtractVectorFromPosition(
     Position=$Particles.Position
     Vector=&Multiply_VectorByFloat(Vector=$Particles.Velocity, Float=$Engine.DeltaTime)
    )) {isSetParameters=true, module=SetVariables_#3}
   /SetVariables(Particles.StaticCamVec:Vector3f=&ReturnCameraVector(Camera Vectors=::NewEnumerator1, Coordinate Space=::Simulation)) {isSetParameters=true, module=SetVariables_#4}
   /SetVariables(Particles.VisibilityTag=&MakeCustomIntFromBool(Bool=$User.bisInVehicle, True INT=1, False INT=0)) {isSetParameters=true}
  ]
  particleUpdateScript [
   &ParticleState()
   &UpdateMeshOrientation(
    Facing Direction=&Subtract_Position(A=$Particles.OrigPosition, B=$Particles.Position)
    Facing Interpolation=false
    Orientation Method=::NewEnumerator1
    Override Quat=false
    Side Direction:Vector3f=&NormalizeVector(
     Vector To Normalize=&CrossProduct(
      Vector A:Vector3f=&ReturnCameraVector(Camera Vectors=::NewEnumerator3, Coordinate Space=::Simulation)
      Vector B:Vector3f=<0 0 -1>
      Normalize Incoming Vectors:NiagaraBool=true
     )
     Safe Normalize=true
     Safe Fallback Threshold=0
     Safe Fallback Vector=<0 0 0.001>
    )
    Stabilization Mode:ENiagara_OrientVectorCount=::NewEnumerator2
    TransformFacing=false
    TransformSide=false
   )
   &self:SuckIntoWake(VelSuckInAmt=5) {module=SuckIntoWake}
   &DynamicMaterialParameters(
    Index 0 Param 4=&MakeFloatFromVector(Vector=$Particles.Scale, Channel=::NewEnumerator0)
    Index 0 Vector 1:NiagaraPosition=$Particles.Position
    Index 1 Vector 1:NiagaraPosition=$Particles.OrigPosition
    Index 2 Vector 1:Vector3f=$Particles.StaticCamVec
    Module.Index0Float3Write=true
    Module.Index0VectorWrite=true
    Module.Index1Float3Write=false
    Module.Index1VectorWrite=true
    Module.Index2Float3Write=false
    Module.Index2VectorWrite=true
    Write Parameter Index 0=::NewEnumerator3
    Write Parameter Index 1=::NewEnumerator3
    Write Parameter Index 2=::NewEnumerator3
   )
   &Drag(Drag=0.4, Use Linear Drag=true, Use Rotational Drag=true)
   &SolveForcesAndVelocity(Clamp Velocity=false, Limit Acceleration=false)
   &ScaleColor(
    Scale Alpha=&FloatFromCurve(FloatCurve=DataInterface() {propertyValues {Curve {keys [
         {arriveTangent=-0.5485, interpMode=RCIM_Cubic, leaveTangent=-0.5485, tangentMode=RCTM_User, time=0, value=1}
         {interpMode=RCIM_Cubic, tangentMode=RCTM_User, time=1, value=0}
        ]}}})
    Scale RGB=&MakeVectorFromLinearColorRGB(Color=&ColorFromCurve(
      ColorCurve=DataInterface() {propertyValues {
        AlphaCurve {keys [<at 0 1 RCIM_Cubic>]}
        BlueCurve {keys [<at 0 0.8847>,<at 0.0795 1.0073>,<at 0.1161 0.5223>,<at 0.1358 0.9857>,<at 0.25 1.1>,<at 0.5 1>,<at 0.75 0.9323>,<at 0.8278 0.5466>,<at 1 0.8847 RCIM_Cubic>]}
        GreenCurve {keys [<at 0 0.5677>,<at 0.0795 0.561>,<at 0.1161 0.6832>,<at 0.1358 0.9061>,<at 0.25 1.1>,<at 0.5 1>,<at 0.75 0.867>,<at 0.8278 0.7704>,<at 1 0.5677 RCIM_Cubic>]}
        RedCurve {keys [<at 0 0.4496>,<at 0.0795 0.3741>,<at 0.1161 1.1>,<at 0.1358 1.0203>,<at 0.25 1.1>,<at 0.5 1>,<at 0.75 0.8128>,<at 0.8278 1.1>,<at 1 0.4496 RCIM_Cubic>]}
       }}
      CurveIndex=Linked() {linkedVariable {name=NPC.FortniteNPC.TimeOfDay_0-1, type=NiagaraFloat}}
     ))
    ScaleA=true
    ScaleRGB=true
   )
  ]
  rendererClasses [&NiagaraMeshRendererProperties]
  renderers [{rendererClass=&NiagaraMeshRendererProperties, rendererIndex=0}]
 }
}
summary {
 systemName=NS_Char_FlickLedge_Swirl
 emitters [
  {emitterName=Smoke, rendererClasses [&NiagaraSpriteRendererProperties]}
  {emitterName=SwirlSky, rendererClasses [&NiagaraMeshRendererProperties]}
  {emitterName=SmokeSky, rendererClasses [&NiagaraSpriteRendererProperties]}
  {emitterName=SwirlCloak, rendererClasses [&NiagaraMeshRendererProperties]}
 ]
 userVariables [
  User.bIsFrontEnd {type=NiagaraBool}
  User.bIsFrontEndPreview {type=NiagaraBool}
  User.bisInVehicle {type=NiagaraBool}
  User.ParaGlideLeanAlpha {type=NiagaraFloat, defaultValue=0, description=""}
  User.PlayerSpeed {type=NiagaraFloat, defaultValue=0, description=""}
  User.SkeletalMesh {
   type=NiagaraDataInterfaceSkeletalMesh
   defaultValue {
    struct=NiagaraExt_VariableValue_DataInterface
    value {dataInterface=&self:NiagaraDataInterfaceSkeletalMesh_0}
   }
  }
 ]
}
