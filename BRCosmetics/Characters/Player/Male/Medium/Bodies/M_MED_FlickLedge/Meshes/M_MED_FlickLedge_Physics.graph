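# Physics asset for the FlickLedge medium male player body, set up for a dynamic skirt.
# Six capsule chains (front, side and back per side) use tight swing/twist limits, with the front swinging widest.
# The front tips are cross-linked both ways with free angular motion.
# Leg and foot capsules provide body shapes, and a huge offset sphere on a dedicated bone approximates flat ground for skirt collision.
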PhysicsAsset /BRCosmetics/Characters/Player/Male/Medium/Bodies/M_MED_FlickLedge/Meshes/M_MED_FlickLedge_Physics
bodies {
 calf_l [{shapeType=Capsule, length=60.9689, radius=10.9824, center=<-17.5123 -0.5124 -0.0823>, rotation=<rot -0.1046 -95.2647 -89.5518>}]
 calf_r [{shapeType=Capsule, length=59.191, radius=10.8301, center=<16.8034 0.5766 0.7116>, rotation=<rot -0.1046 -95.2647 90.4482>}]
 dyn_skirt_back_l_01 [{shapeType=Capsule, length=62.8175, radius=6.8703, shapeName=dyn_rear_skirt_l_01_capsule, center=<0 -6 0>, rotation=<rot 0 0.0004 3.4582>}]
 dyn_skirt_back_l_02 [{
   shapeType=Capsule
   length=69.2824
   radius=9.3406
   shapeName=dyn_rear_skirt_l_02_capsule
   center=<9.7113 -6 0>
   rotation=<rot 0 0.0004 3.4582>
  }]
 dyn_skirt_back_l_03 [{shapeType=Capsule, length=73.518, radius=9.9656, shapeName=dyn_rear_skirt_l_03_capsule, center=<5.3608 -6 0>, rotation=<rot 0 0.0004 3.4582>}]
 dyn_skirt_back_l_04 [{
   shapeType=Capsule
   length=74.8886
   radius=9.5957
   shapeName=dyn_rear_skirt_l_04_capsule
   center=<7 -5.0025 0>
   rotation=<rot 0 0.0004 3.4582>
  }]
 dyn_skirt_back_l_05 [{
   shapeType=Capsule
   length=78.3279
   radius=9.9387
   shapeName=dyn_rear_skirt_l_05_capsule
   center=<7 -4.6618 0>
   rotation=<rot 0 0.0004 3.4582>
  }]
 dyn_skirt_back_l_06 [{
   shapeType=Capsule
   length=80.5016
   radius=10.3275
   shapeName=dyn_rear_skirt_l_06_capsule
   center=<5 -4.6474 0>
   rotation=<rot 0 0.0004 3.4582>
  }]
 dyn_skirt_back_r_01 [{shapeType=Capsule, length=72.4271, radius=7.6945, shapeName=dyn_rear_skirt_r_01_capsule, center=<0 6 0>}]
 dyn_skirt_back_r_02 [{shapeType=Capsule, length=71.3311, radius=8.7542, shapeName=dyn_rear_skirt_r_02_capsule, center=<-8.4972 6 0>}]
 dyn_skirt_back_r_03 [{shapeType=Capsule, length=74.4255, radius=9.058, shapeName=dyn_rear_skirt_r_03_capsule, center=<-3.4783 5.3301 0>}]
 dyn_skirt_back_r_04 [{shapeType=Capsule, length=75.2062, radius=9.2781, shapeName=dyn_rear_skirt_r_04_capsule, center=<-5.6433 4.8986 0>}]
 dyn_skirt_back_r_05 [{shapeType=Capsule, length=78.7048, radius=9.5618, shapeName=dyn_rear_skirt_r_05_capsule, center=<-6.3516 4.3293 0>}]
 dyn_skirt_back_r_06 [{shapeType=Capsule, length=81.051, radius=9.7781, shapeName=dyn_rear_skirt_r_06_capsule, center=<-5 3.9723 0>}]
 dyn_skirt_front_l_01 [{
   shapeType=Capsule
   length=59.5576
   radius=7.125
   shapeName=dyn_front_skirt_l_01_capsule
   center=<0 -4.4965 0>
   rotation=<rot 0.0001 -0.0004 13.0359>
  }]
 dyn_skirt_front_l_02 [{
   shapeType=Capsule
   length=64.2487
   radius=9.5694
   shapeName=dyn_front_skirt_l_02_capsule
   center=<7 -6.5964 0>
   rotation=<rot 0.0001 -0.0004 13.0359>
  }]
 dyn_skirt_front_l_03 [{
   shapeType=Capsule
   length=66.1703
   radius=9.9293
   shapeName=dyn_front_skirt_l_03_capsule
   center=<4 -7 0>
   rotation=<rot 0.0001 -0.0004 13.0359>
  }]
 dyn_skirt_front_l_04 [{
   shapeType=Capsule
   length=70.1249
   radius=8.9077
   shapeName=dyn_front_skirt_l_04_capsule
   center=<7 -6 0>
   rotation=<rot 0.0001 -0.0004 13.0359>
  }]
 dyn_skirt_front_l_05 [{
   shapeType=Capsule
   length=71.9144
   radius=8.9682
   shapeName=dyn_front_skirt_l_05_capsule
   center=<7 -6 0>
   rotation=<rot 0.0001 -0.0004 13.0359>
  }]
 dyn_skirt_front_l_06 [{
   shapeType=Capsule
   length=73.3133
   radius=10.1319
   shapeName=dyn_front_skirt_l_06_capsule
   center=<7 -6 0>
   rotation=<rot 0.0001 -0.0004 13.0359>
  }]
 dyn_skirt_front_r_01 [{
   shapeType=Capsule
   length=58.1344
   radius=8.5481
   shapeName=dyn_front_skirt_r_01_capsule
   center=<0 5.3522 0>
   rotation=<rot 0.0001 -0.0004 10.4798>
  }]
 dyn_skirt_front_r_02 [{
   shapeType=Capsule
   length=64.6584
   radius=9.1597
   shapeName=dyn_front_skirt_r_02_capsule
   center=<-7 5.498 0.3844>
   rotation=<rot 0.0001 -0.0004 10.4798>
  }]
 dyn_skirt_front_r_03 [{
   shapeType=Capsule
   length=66.8621
   radius=9.2374
   shapeName=dyn_front_skirt_r_03_capsule
   center=<-4 5.5173 0.3595>
   rotation=<rot 0.0001 -0.0004 10.4798>
  }]
 dyn_skirt_front_r_04 [{
   shapeType=Capsule
   length=70.0195
   radius=9.0131
   shapeName=dyn_front_skirt_r_04_capsule
   center=<-5.0159 5.0233 0.1807>
   rotation=<rot 0.0001 -0.0004 10.4798>
  }]
 dyn_skirt_front_r_05 [{
   shapeType=Capsule
   length=71.8287
   radius=9.054
   shapeName=dyn_front_skirt_r_05_capsule
   center=<-4.2351 4.8848 0.2063>
   rotation=<rot 0.0001 -0.0004 10.4798>
  }]
 dyn_skirt_front_r_06 [{
   shapeType=Capsule
   length=74.5891
   radius=8.8561
   shapeName=dyn_front_skirt_r_06_capsule
   center=<-2.657 4.6152 0.2562>
   rotation=<rot 0.0001 -0.0004 10.4798>
  }]
 dyn_skirt_ground_collision [{shapeType=Sphere, radius=1000, shapeName=dyn_skirt_ground_collision_sphere, center=<0 26.942 -1004>}]
 dyn_skirt_side_l_01 [{shapeType=Capsule, length=53.7234, radius=7.5427, shapeName=dyn_side_skirt_l_01_capsule, center=<0 -8.559 0>}]
 dyn_skirt_side_l_02 [{shapeType=Capsule, length=53.0459, radius=8.8452, shapeName=dyn_side_skirt_l_02_capsule, center=<8.0116 -9.167 0>}]
 dyn_skirt_side_l_03 [{shapeType=Capsule, length=52.3716, radius=9.0952, shapeName=dyn_side_skirt_l_03_capsule, center=<7 -8.2359 0>}]
 dyn_skirt_side_l_04 [{shapeType=Capsule, length=53.5599, radius=8.9077, shapeName=dyn_side_skirt_l_04_capsule, center=<7 -7.6878 0>}]
 dyn_skirt_side_l_05 [{shapeType=Capsule, length=57.4533, radius=8.7966, shapeName=dyn_side_skirt_l_05_capsule, center=<7 -7 0>}]
 dyn_skirt_side_l_06 [{shapeType=Capsule, length=59.9048, radius=8.9077, shapeName=dyn_side_skirt_l_06_capsule, center=<7 -8 0>}]
 dyn_skirt_side_r_01 [{shapeType=Capsule, length=55.2661, radius=7.5427, shapeName=dyn_side_skirt_r_01_capsule, center=<0 9.0274 0>, rotation=<rot 0 90 -180>}]
 dyn_skirt_side_r_02 [{shapeType=Capsule, length=57.0647, radius=8.4982, shapeName=dyn_side_skirt_r_02_capsule, center=<-9.2431 9.2553 0>, rotation=<rot 0 90 -180>}]
 dyn_skirt_side_r_03 [{shapeType=Capsule, length=52.2217, radius=9.0952, shapeName=dyn_side_skirt_r_03_capsule, center=<-4.7648 8.9196 0>, rotation=<rot 0 90 -180>}]
 dyn_skirt_side_r_04 [{shapeType=Capsule, length=53.5599, radius=8.9077, shapeName=dyn_side_skirt_r_04_capsule, center=<-7 8.2221 0>, rotation=<rot 0 90 -180>}]
 dyn_skirt_side_r_05 [{shapeType=Capsule, length=58.8423, radius=8.7966, shapeName=dyn_side_skirt_r_05_capsule, center=<-7 7.7292 0>, rotation=<rot 0 90 -180>}]
 dyn_skirt_side_r_06 [{shapeType=Capsule, length=59.9048, radius=8.9077, shapeName=dyn_side_skirt_r_06_capsule, center=<-7 7.0977 0>, rotation=<rot 0 90 -180>}]
 foot_l [{
   shapeType=Capsule
   length=16.9992
   radius=9.0663
   shapeName=foot_l_capsule
   center=<-4.4717 6.1157 -0.1676>
   rotation=<rot -1.6769 31.3492 -87.7001>
  }]
 foot_r [{
   shapeType=Capsule
   length=16.9992
   radius=9.0663
   shapeName=foot_r_capsule
   center=<6.8735 -8.7592 0.2887>
   rotation=<rot -1.1425 18.9288 92.6068>
  }]
 thigh_l [{shapeType=Capsule, length=28.3289, radius=10.9492, center=<-20.4288 -2.4134 2.1357>, rotation=<rot 0.4507 -91.6838 -88.7036>}]
 thigh_r [{shapeType=Capsule, length=27.9486, radius=11.3294, center=<20.5685 2.0248 -0.5192>, rotation=<rot 0.3313 -90.4104 95.3593>}]
}
constraints [
 {bone1Name=dyn_skirt_front_l_02, bone2Name=dyn_skirt_front_l_01, swing1LimitDegrees=24, swing2LimitDegrees=2, twistLimitDegrees=2}
 {bone1Name=dyn_skirt_front_l_03, bone2Name=dyn_skirt_front_l_02, swing1LimitDegrees=24, swing2LimitDegrees=2, twistLimitDegrees=2}
 {bone1Name=dyn_skirt_front_l_04, bone2Name=dyn_skirt_front_l_03, swing1LimitDegrees=24, swing2LimitDegrees=2, twistLimitDegrees=8}
 {bone1Name=dyn_skirt_front_l_05, bone2Name=dyn_skirt_front_l_04, swing1LimitDegrees=24, swing2LimitDegrees=8, twistLimitDegrees=8}
 {bone1Name=dyn_skirt_front_l_06, bone2Name=dyn_skirt_front_l_05, swing1LimitDegrees=24, swing2LimitDegrees=8, twistLimitDegrees=8}
 {bone1Name=dyn_skirt_front_r_02, bone2Name=dyn_skirt_front_r_01, swing1LimitDegrees=24, swing2LimitDegrees=2, twistLimitDegrees=2}
 {bone1Name=dyn_skirt_front_r_03, bone2Name=dyn_skirt_front_r_02, swing1LimitDegrees=24, swing2LimitDegrees=2, twistLimitDegrees=2}
 {bone1Name=dyn_skirt_front_r_04, bone2Name=dyn_skirt_front_r_03, swing1LimitDegrees=24, swing2LimitDegrees=8, twistLimitDegrees=8}
 {bone1Name=dyn_skirt_front_r_05, bone2Name=dyn_skirt_front_r_04, swing1LimitDegrees=24, swing2LimitDegrees=8, twistLimitDegrees=8}
 {bone1Name=dyn_skirt_front_r_06, bone2Name=dyn_skirt_front_r_05, swing1LimitDegrees=24, swing2LimitDegrees=8, twistLimitDegrees=8}
 {bone1Name=dyn_skirt_side_l_02, bone2Name=dyn_skirt_side_l_01, swing1LimitDegrees=12, swing2LimitDegrees=4, twistLimitDegrees=4}
 {bone1Name=dyn_skirt_side_l_03, bone2Name=dyn_skirt_side_l_02, swing1LimitDegrees=14, swing2LimitDegrees=4, twistLimitDegrees=4}
 {bone1Name=dyn_skirt_side_l_04, bone2Name=dyn_skirt_side_l_03, swing1LimitDegrees=14, swing2LimitDegrees=8, twistLimitDegrees=8}
 {bone1Name=dyn_skirt_side_l_05, bone2Name=dyn_skirt_side_l_04, swing1LimitDegrees=14, swing2LimitDegrees=8, twistLimitDegrees=8}
 {bone1Name=dyn_skirt_side_l_06, bone2Name=dyn_skirt_side_l_05, swing1LimitDegrees=14, swing2LimitDegrees=4, twistLimitDegrees=4}
 {bone1Name=dyn_skirt_side_r_02, bone2Name=dyn_skirt_side_r_01, swing1LimitDegrees=12, swing2LimitDegrees=4, twistLimitDegrees=4}
 {bone1Name=dyn_skirt_side_r_03, bone2Name=dyn_skirt_side_r_02, swing1LimitDegrees=14, swing2LimitDegrees=4, twistLimitDegrees=4}
 {bone1Name=dyn_skirt_side_r_04, bone2Name=dyn_skirt_side_r_03, swing1LimitDegrees=14, swing2LimitDegrees=8, twistLimitDegrees=8}
 {bone1Name=dyn_skirt_side_r_05, bone2Name=dyn_skirt_side_r_04, swing1LimitDegrees=14, swing2LimitDegrees=8, twistLimitDegrees=8}
 {bone1Name=dyn_skirt_side_r_06, bone2Name=dyn_skirt_side_r_05, swing1LimitDegrees=14, swing2LimitDegrees=4, twistLimitDegrees=4}
 {bone1Name=dyn_skirt_back_l_02, bone2Name=dyn_skirt_back_l_01, swing1LimitDegrees=12, swing2LimitDegrees=8, twistLimitDegrees=8}
 {bone1Name=dyn_skirt_back_l_03, bone2Name=dyn_skirt_back_l_02, swing1LimitDegrees=14, swing2LimitDegrees=8, twistLimitDegrees=8}
 {bone1Name=dyn_skirt_back_l_04, bone2Name=dyn_skirt_back_l_03, swing1LimitDegrees=14, swing2LimitDegrees=8, twistLimitDegrees=8}
 {bone1Name=dyn_skirt_back_l_05, bone2Name=dyn_skirt_back_l_04, swing1LimitDegrees=14, swing2LimitDegrees=8, twistLimitDegrees=8}
 {bone1Name=dyn_skirt_back_l_06, bone2Name=dyn_skirt_back_l_05, swing1LimitDegrees=14, swing2LimitDegrees=8, twistLimitDegrees=8}
 {bone1Name=dyn_skirt_back_r_02, bone2Name=dyn_skirt_back_r_01, swing1LimitDegrees=12, swing2LimitDegrees=8, twistLimitDegrees=8}
 {bone1Name=dyn_skirt_back_r_03, bone2Name=dyn_skirt_back_r_02, swing1LimitDegrees=14, swing2LimitDegrees=8, twistLimitDegrees=8}
 {bone1Name=dyn_skirt_back_r_04, bone2Name=dyn_skirt_back_r_03, swing1LimitDegrees=14, swing2LimitDegrees=8, twistLimitDegrees=8}
 {bone1Name=dyn_skirt_back_r_05, bone2Name=dyn_skirt_back_r_04, swing1LimitDegrees=14, swing2LimitDegrees=8, twistLimitDegrees=8}
 {bone1Name=dyn_skirt_back_r_06, bone2Name=dyn_skirt_back_r_05, swing1LimitDegrees=14, swing2LimitDegrees=8, twistLimitDegrees=8}
 {
  bone1Name=dyn_skirt_front_l_06
  bone2Name=dyn_skirt_front_r_06
  swing1LimitDegrees=45
  swing1Motion=Free
  swing2LimitDegrees=45
  swing2Motion=Free
  twistLimitDegrees=45
  twistMotion=Free
 }
 {
  bone1Name=dyn_skirt_front_r_06
  bone2Name=dyn_skirt_front_l_06
  swing1LimitDegrees=45
  swing1Motion=Free
  swing2LimitDegrees=45
  swing2Motion=Free
  twistLimitDegrees=45
  twistMotion=Free
 }
]
