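# Post-simulation Control Rig for the M_MED_FlickLedge player body that corrects the skirt bone chains after simulation.
# Skirt deformation offsets bones using pose curves (knee bend, foot near pelvis) and Kalman-smoothed frame-to-frame thigh rotation speed.
# A helper derives right-side offsets from left-side data by negating translation, with separate weights per side.
# Parent constraints blend middle chains between left and right chains, with front weighting biased by thigh-raise curves.
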ControlRigBlueprint /BRCosmetics/Characters/Player/Male/Medium/Bodies/M_MED_FlickLedge/Meshes/M_MED_FlickLedge_PostSimCtrlRig
graphs [
 ContainedGraph {nodes [
   n1 = RigVMFunctionEntryNode()
   n2 = /Script/RigVM.RigVMFunction_MathTransformToEulerTransform(Value=@n1.transform) {is=struct RigVMUnitNode, template="Cast::Execute(in Value,out Result)"}
   RigVMFunction_MathQuaternionFromRotatorV2 = /Script/RigVM.RigVMFunction_MathQuaternionFromRotatorV2(Value=@n2.Result.Rotation) {is=struct RigVMUnitNode, template="Cast::Execute(in Value,out Result)"}
   > RigVMFunctionReturnNode(
    mirrorTransform=(Rotation=(X=0,Y=0,Z=0,W=1),Translation=(X=0,Y=0,Z=0),Scale3D=(X=0,Y=0,Z=0))
    mirrorTransform.Translation=/Script/RigVM.RigVMFunction_MathVectorScale(Factor=-1, Value=@n1.transform.Translation) {is=struct RigVMUnitNode, template="Scale::Execute(in Value,in Factor,out Result)"}.Result
    mirrorTransform.Rotation=@RigVMFunction_MathQuaternionFromRotatorV2.Result
    mirrorTransform.Scale3D=@n2.Result.Scale
   )
  ]}
 ContainedGraph {nodes [
   n1 = RigVMFunctionEntryNode()
   n2 = &RigVMFunction_MathTransformToEulerTransform(Value=@n1.transform) {template="Cast::Execute(in Value,out Result)"}
   RigVMFunction_MathQuaternionFromRotatorV2 = &RigVMFunction_MathQuaternionFromRotatorV2(Value=@n2.Result.Rotation) {template="Cast::Execute(in Value,out Result)"}
   > RigVMFunctionReturnNode(
    mirrorTransform=(Rotation=(X=0,Y=0,Z=0,W=1),Translation=(X=0,Y=0,Z=0),Scale3D=(X=0,Y=0,Z=0))
    mirrorTransform.Translation=&RigVMFunction_MathVectorScale(Factor=-1, Value=@n1.transform.Translation) {template="Scale::Execute(in Value,in Factor,out Result)"}.Result
    mirrorTransform.Rotation=@RigVMFunction_MathQuaternionFromRotatorV2.Result
    mirrorTransform.Scale3D=@n2.Result.Scale
   )
  ]}
 ContainedGraph {nodes [
   n1 = RigVMFunctionEntryNode()
   n2 = &RigVMFunction_MathTransformToEulerTransform(Value=@n1.transform) {template="Cast::Execute(in Value,out Result)"}
   RigVMFunction_MathQuaternionFromRotatorV2 = &RigVMFunction_MathQuaternionFromRotatorV2(Value=@n2.Result.Rotation) {template="Cast::Execute(in Value,out Result)"}
   > RigVMFunctionReturnNode(
    mirrorTransform=(Rotation=(X=0,Y=0,Z=0,W=1),Translation=(X=0,Y=0,Z=0),Scale3D=(X=0,Y=0,Z=0))
    mirrorTransform.Translation=&RigVMFunction_MathVectorScale(Factor=-1, Value=@n1.transform.Translation) {template="Scale::Execute(in Value,in Factor,out Result)"}.Result
    mirrorTransform.Rotation=@RigVMFunction_MathQuaternionFromRotatorV2.Result
    mirrorTransform.Scale3D=@n2.Result.Scale
   )
  ]}
 New Function_ContainedGraph {nodes [
   n1 = RigVMFunctionEntryNode()
   > RigVMFunction_Sequence = RigVMAggregateNode()
   Set_Num_2 = RigVMDispatchNode(
    Num=RigVMDispatchNode(Array=@n1.Base Modify Data) {template="DISPATCH_RigVMDispatch_ArrayGetNum(in Array,out Num)"}.Num
    ExecuteContext=@RigVMFunction_Sequence.A
    Array=RigVMVariableNode() {variable=Mirrored}.Value
   ) {template="DISPATCH_RigVMDispatch_ArraySetNum(io Array,in Num)"}
   > For_Each_1 = RigVMDispatchNode(Array=@n1.Base Modify Data) {template="DISPATCH_RigVMDispatch_ArrayIterator(in Array,out Element,out Index,out Count,out Ratio)"}
   > Get Mirror Transform = RigVMFunctionReferenceNode(Transform=@For_Each_1.Element.Transform)
   n2 = RigVMDispatchNode(Index=@For_Each_1.Index, Array=@n1.Mirror Modify Data) {template="DISPATCH_RigVMDispatch_ArrayGetAtIndex(in Array,in Index,out Element)"}
   > Set_At_2 = RigVMDispatchNode(
    Element=(Item=(Type=Bone,Name="None"),Transform=(Rotation=(X=0,Y=0,Z=0,W=1),Translation=(X=0,Y=0,Z=0),Scale3D=(X=1,Y=1,Z=1)))
    Array=RigVMVariableNode() {variable=Mirrored}.Value
    Element.Transform.Rotation=/Script/RigVM.RigVMFunction_MathQuaternionMul(B=@n2.Element.Transform.Rotation, A=@"Get Mirror Transform".mirrorTransform.Rotation) {is=struct RigVMUnitNode, template="Multiply::Execute(in A,in B,out Result)"}.Result
    Element.Item=@n2.Element.Item
    Element.Transform.Scale3D=@"Get Mirror Transform".mirrorTransform.Scale3D
    Index=@For_Each_1.Index
    Element.Transform.Translation=/Script/RigVM.RigVMFunction_MathVectorAdd(B=@n2.Element.Transform.Translation, A=@"Get Mirror Transform".mirrorTransform.Translation) {is=struct RigVMUnitNode, template="Add::Execute(in A,in B,out Result)"}.Result
   ) {template="DISPATCH_RigVMDispatch_ArraySetAtIndex(io Array,in Index,in Element)"}
   &RigUnit_ModifyTransforms(WeightMinimum=0, Weight=@n1.Weight, ExecutePin=@RigVMFunction_Sequence.B, ItemToModify=@n1.Base Modify Data)
   > &RigUnit_ModifyTransforms(WeightMinimum=0, Weight=@n1."Mirror Weight ", ItemToModify=RigVMVariableNode() {variable=Mirrored}.Value)
   RigVMFunctionReturnNode(ExecuteContext=@RigVMFunction_Sequence.C)
  ]}
 New Function_ContainedGraph {nodes [
   n1 = RigVMFunctionEntryNode()
   > RigVMFunction_Sequence = RigVMAggregateNode()
   Set_Num_2 = RigVMDispatchNode(
    Num=RigVMDispatchNode(Array=@n1.Base Modify Data) {template="DISPATCH_RigVMDispatch_ArrayGetNum(in Array,out Num)"}.Num
    ExecuteContext=@RigVMFunction_Sequence.A
    Array=RigVMVariableNode() {variable=Mirrored}.Value
   ) {template="DISPATCH_RigVMDispatch_ArraySetNum(io Array,in Num)"}
   > For_Each_1 = RigVMDispatchNode(Array=@n1.Base Modify Data) {template="DISPATCH_RigVMDispatch_ArrayIterator(in Array,out Element,out Index,out Count,out Ratio)"}
   > Get Mirror Transform = RigVMFunctionReferenceNode(Transform=@For_Each_1.Element.Transform)
   n2 = RigVMDispatchNode(Index=@For_Each_1.Index, Array=@n1.Mirror Modify Data) {template="DISPATCH_RigVMDispatch_ArrayGetAtIndex(in Array,in Index,out Element)"}
   > Set_At_2 = RigVMDispatchNode(
    Element=(Item=(Type=Bone,Name="None"),Transform=(Rotation=(X=0,Y=0,Z=0,W=1),Translation=(X=0,Y=0,Z=0),Scale3D=(X=1,Y=1,Z=1)))
    Array=RigVMVariableNode() {variable=Mirrored}.Value
    Element.Transform.Rotation=&RigVMFunction_MathQuaternionMul(B=@n2.Element.Transform.Rotation, A=@"Get Mirror Transform".mirrorTransform.Rotation) {template="Multiply::Execute(in A,in B,out Result)"}.Result
    Element.Item=@n2.Element.Item
    Element.Transform.Scale3D=@"Get Mirror Transform".mirrorTransform.Scale3D
    Index=@For_Each_1.Index
    Element.Transform.Translation=&RigVMFunction_MathVectorAdd(B=@n2.Element.Transform.Translation, A=@"Get Mirror Transform".mirrorTransform.Translation) {template="Add::Execute(in A,in B,out Result)"}.Result
   ) {template="DISPATCH_RigVMDispatch_ArraySetAtIndex(io Array,in Index,in Element)"}
   &RigUnit_ModifyTransforms(WeightMinimum=0, Weight=@n1.Weight, ExecutePin=@RigVMFunction_Sequence.B, ItemToModify=@n1.Base Modify Data)
   > &RigUnit_ModifyTransforms(WeightMinimum=0, Weight=@n1."Mirror Weight ", ItemToModify=RigVMVariableNode() {variable=Mirrored}.Value)
   RigVMFunctionReturnNode(ExecuteContext=@RigVMFunction_Sequence.C)
  ]}
 New Function_ContainedGraph {nodes [
   n1 = RigVMFunctionEntryNode()
   > RigVMFunction_Sequence = RigVMAggregateNode()
   Set_Num_2 = RigVMDispatchNode(
    Num=RigVMDispatchNode(Array=@n1.Base Modify Data) {template="DISPATCH_RigVMDispatch_ArrayGetNum(in Array,out Num)"}.Num
    ExecuteContext=@RigVMFunction_Sequence.A
    Array=RigVMVariableNode() {variable=Mirrored}.Value
   ) {template="DISPATCH_RigVMDispatch_ArraySetNum(io Array,in Num)"}
   > For_Each_1 = RigVMDispatchNode(Array=@n1.Base Modify Data) {template="DISPATCH_RigVMDispatch_ArrayIterator(in Array,out Element,out Index,out Count,out Ratio)"}
   > Get Mirror Transform = RigVMFunctionReferenceNode(Transform=@For_Each_1.Element.Transform)
   n2 = RigVMDispatchNode(Index=@For_Each_1.Index, Array=@n1.Mirror Modify Data) {template="DISPATCH_RigVMDispatch_ArrayGetAtIndex(in Array,in Index,out Element)"}
   > Set_At_2 = RigVMDispatchNode(
    Element=(Item=(Type=Bone,Name="None"),Transform=(Rotation=(X=0,Y=0,Z=0,W=1),Translation=(X=0,Y=0,Z=0),Scale3D=(X=1,Y=1,Z=1)))
    Array=RigVMVariableNode() {variable=Mirrored}.Value
    Element.Transform.Rotation=&RigVMFunction_MathQuaternionMul(B=@n2.Element.Transform.Rotation, A=@"Get Mirror Transform".mirrorTransform.Rotation) {template="Multiply::Execute(in A,in B,out Result)"}.Result
    Element.Item=@n2.Element.Item
    Element.Transform.Scale3D=@"Get Mirror Transform".mirrorTransform.Scale3D
    Index=@For_Each_1.Index
    Element.Transform.Translation=&RigVMFunction_MathVectorAdd(B=@n2.Element.Transform.Translation, A=@"Get Mirror Transform".mirrorTransform.Translation) {template="Add::Execute(in A,in B,out Result)"}.Result
   ) {template="DISPATCH_RigVMDispatch_ArraySetAtIndex(io Array,in Index,in Element)"}
   &RigUnit_ModifyTransforms(WeightMinimum=0, Weight=@n1.Weight, ExecutePin=@RigVMFunction_Sequence.B, ItemToModify=@n1.Base Modify Data)
   > &RigUnit_ModifyTransforms(WeightMinimum=0, Weight=@n1."Mirror Weight ", ItemToModify=RigVMVariableNode() {variable=Mirrored}.Value)
   RigVMFunctionReturnNode(ExecuteContext=@RigVMFunction_Sequence.C)
  ]}
 New Function_ContainedGraph {nodes [
   RigVMFunctionEntryNode()
   > Modify Transforms with Mirror_0 = RigVMFunctionReferenceNode(
    Base Modify Data="((Item=(Type=Bone,Name=\"dyn_skirt_front_l_MSTR\"),Transform=(Rotation=(X=0,Y=-0.0175,Z=0,W=0.9998),Translation=(X=0,Y=-1.5,Z=0),Scale3D=(X=1,Y=1,Z=1))),(Item=(Type=Bone,Name=\"dyn_skirt_side_l_MSTR\"),Transform=(Rotation=(X=0,Y=0,Z=0,W=1),Translation=(X=0,Y=- ...(truncated)"
    Mirror Modify Data="((Item=(Type=Bone,Name=\"dyn_skirt_front_r_MSTR\"),Transform=(Rotation=(X=0,Y=0,Z=0,W=1),Translation=(X=0,Y=0,Z=0),Scale3D=(X=1,Y=1,Z=1))),(Item=(Type=Bone,Name=\"dyn_skirt_side_r_MSTR\"),Transform=(Rotation=(X=0,Y=0,Z=0,W=1),Translation=(X=0,Y=0,Z=0),Scale3D= ...(truncated)"
    "Mirror Weight "=1
    Weight=1
   )
   n1 = &RigUnit_GetTransform(Item=(Type=Bone,Name="thigh_r"), Space=LocalSpace, bInitial=False)
   n2 = /Script/RigVM.RigVMFunction_MathQuaternionToAxisAndAngle(Value=&RigVMFunction_MathQuaternionMul(
     B=/Script/RigVM.RigVMFunction_MathQuaternionInverse(Value=@n1.Transform.Rotation) {is=struct RigVMUnitNode, template="Inverse::Execute(in Value,out Result)"}.Result
     A=RigVMVariableNode() {variable=prev thigh r quaternion}.Value
    ) {template="Multiply::Execute(in A,in B,out Result)"}.Result) {is=struct RigVMUnitNode, template="ToAxisAndAngle::Execute(in Value,out Axis,out Angle)"}
   n3 = &RigVMFunction_KalmanFloat(
    BufferSize=4
    Value=&RigVMFunction_MathDoubleRemap(
     SourceMaximum=15
     SourceMinimum=0
     TargetMaximum=1
     TargetMinimum=0
     Value=/Script/RigVM.RigVMFunction_MathDoubleDeg(Value=@n2.Angle) {is=struct RigVMUnitNode, template="Degrees::Execute(in Value,out Result)"}.Result
    ).Result
   ) {template="KalmanFilter::Execute(in Value,in BufferSize,out Result)"}
   RigUnit_GetTransform = &RigUnit_GetTransform(Item=(Type=Bone,Name="thigh_l"), Space=LocalSpace, bInitial=False)
   Inverse = &RigVMFunction_MathQuaternionInverse(Value=@RigUnit_GetTransform.Transform.Rotation) {template="Inverse::Execute(in Value,out Result)"}
   RigVMFunction_MathQuaternionToAxisAndAngle = &RigVMFunction_MathQuaternionToAxisAndAngle(Value=&RigVMFunction_MathQuaternionMul(B=@Inverse.Result, A=RigVMVariableNode() {variable=prev thigh l quaternion}.Value) {template="Multiply::Execute(in A,in B,out Result)"}.Result) {template="ToAxisAndAngle::Execute(in Value,out Axis,out Angle)"}
   Degrees = &RigVMFunction_MathDoubleDeg(Value=@RigVMFunction_MathQuaternionToAxisAndAngle.Angle) {template="Degrees::Execute(in Value,out Result)"}
   KalmanFilter = &RigVMFunction_KalmanFloat(
    BufferSize=4
    Value=&RigVMFunction_MathDoubleRemap(SourceMaximum=15, SourceMinimum=0, TargetMaximum=1, TargetMinimum=0, Value=@Degrees.Result).Result
   ) {template="KalmanFilter::Execute(in Value,in BufferSize,out Result)"}
   > Modify Transforms with Mirror = RigVMFunctionReferenceNode(
    Base Modify Data="((Item=(Type=Bone,Name=\"dyn_skirt_front_l_MSTR\"),Transform=(Rotation=(X=0.0076,Y=0.0868,Z=-0.0868,W=0.9924),Translation=(X=0,Y=-5,Z=-5),Scale3D=(X=1,Y=1,Z=1))),(Item=(Type=Bone,Name=\"dyn_skirt_side_l_MSTR\"),Transform=(Rotation=(X=0.0114,Y=0.0864,Z=-0.13,W= ...(truncated)"
    Mirror Modify Data="((Item=(Type=Bone,Name=\"dyn_skirt_front_r_MSTR\"),Transform=(Rotation=(X=0,Y=0,Z=0,W=1),Translation=(X=0,Y=0,Z=0),Scale3D=(X=1,Y=1,Z=1))),(Item=(Type=Bone,Name=\"dyn_skirt_side_r_MSTR\"),Transform=(Rotation=(X=0,Y=0,Z=0,W=1),Translation=(X=0,Y=0,Z=0),Scale3D= ...(truncated)"
    "Mirror Weight "=@n3.Result
    Weight=@KalmanFilter.Result
   )
   > RigVMVariableNode(Value=@RigUnit_GetTransform.Transform.Rotation) {variable=prev thigh l quaternion}
   > RigVMVariableNode(Value=@n1.Transform.Rotation) {variable=prev thigh r quaternion}
   n4 = /Script/ControlRig.RigUnit_GetCurveValue(Curve=CR Rear Knee Bend L) {is=struct RigVMUnitNode}
   n5 = &RigUnit_GetCurveValue(Curve=CR Rear Knee Bend R)
   > Modify Transforms with Mirror_3 = RigVMFunctionReferenceNode(
    Base Modify Data="((Item=(Type=Bone,Name=\"dyn_skirt_back_l_01\"),Transform=(Rotation=(X=0,Y=0,Z=-0.2588,W=0.9659),Translation=(X=0,Y=0,Z=0),Scale3D=(X=1,Y=1,Z=1))),(Item=(Type=Bone,Name=\"dyn_skirt_back_mid_01\"),Transform=(Rotation=(X=0,Y=0,Z=0,W=1),Translation=(X=0,Y=0,Z=0), ...(truncated)"
    Mirror Modify Data="((Item=(Type=Bone,Name=\"dyn_skirt_back_r_01\"),Transform=(Rotation=(X=0,Y=0,Z=0,W=1),Translation=(X=0,Y=0,Z=0),Scale3D=(X=1,Y=1,Z=1))),(Item=(Type=Bone,Name=\"dyn_skirt_back_mid_01\"),Transform=(Rotation=(X=0,Y=0,Z=0,W=1),Translation=(X=0,Y=0,Z=0),Scale3D=(X= ...(truncated)"
    Weight=@n4.Value
    "Mirror Weight "=@n5.Value
   )
   Maximum = &RigVMFunction_MathDoubleMax(A=@n4.Value, B=@n5.Value)
   > &RigUnit_ModifyTransforms(
    ItemToModify="((Item=(Type=Bone,Name=\"dyn_skirt_back_mid_MSTR\"),Transform=(Rotation=(X=0,Y=0,Z=-0.1305,W=0.9914),Translation=(X=0,Y=0,Z=0),Scale3D=(X=1,Y=1,Z=1))),(Item=(Type=Bone,Name=\"dyn_skirt_back_mid_01\"),Transform=(Rotation=(X=0,Y=0,Z=-0.1305,W=0.9914),Translation ...(truncated)"
    WeightMinimum=0
    Weight=&RigVMFunction_KalmanFloat(BufferSize=4, Value=@Maximum.Result) {template="KalmanFilter::Execute(in Value,in BufferSize,out Result)"}.Result
   )
   > Modify Transforms with Mirror_4 = RigVMFunctionReferenceNode(
    Base Modify Data=((Item=(Type=Bone,Name="dyn_skirt_back_l_01"),Transform=(Rotation=(X=0,Y=-0.2164,Z=0,W=0.9763),Translation=(X=-5,Y=-3,Z=0),Scale3D=(X=1,Y=1,Z=1))))
    Mirror Modify Data=((Item=(Type=Bone,Name="dyn_skirt_back_r_01"),Transform=(Rotation=(X=0,Y=0,Z=0,W=1),Translation=(X=0,Y=0,Z=0),Scale3D=(X=1,Y=1,Z=1))))
    Weight=&RigUnit_GetCurveValue(Curve=CR Foot Near Pelvis L).Value
    "Mirror Weight "=&RigUnit_GetCurveValue(Curve=CR Foot Near Pelvis R).Value
   )
   > RigVMFunctionReturnNode()
   RigUnit_SetRelativeTransformForItem = /Script/ControlRig.RigUnit_SetRelativeTransformForItem(
    Child=(Type=Bone,Name="dyn_skirt_front_l_02")
    ExecutePin=()
    Parent=(Type=Bone,Name="dyn_front_skirt_l_02")
    Value=(Rotation=(X=0,Y=0,Z=-0.0698,W=0.9976),Translation=(X=0,Y=0,Z=0),Scale3D=(X=1,Y=1,Z=1))
    Weight=1
    bParentInitial=False
    bPropagateToChildren=false
   ) {
    is=struct RigVMUnitNode
    template="Set Relative Transform::Execute(in Child,in Parent,in bParentInitial,in Value,in Weight,in bPropagateToChildren)"
   }
   RerouteNode_5_Constant = RigVMDispatchNode(Value=1) {template="DISPATCH_RigVMDispatch_Constant(io Value)"}
   > &RigUnit_SetRelativeTransformForItem(
    Child=(Type=Bone,Name="dyn_skirt_front_l_03")
    Parent=(Type=Bone,Name="dyn_front_skirt_l_03")
    Value=(Rotation=(X=0,Y=0,Z=0,W=1),Translation=(X=0,Y=-1.2,Z=0),Scale3D=(X=1,Y=1,Z=1))
    bParentInitial=False
    bPropagateToChildren=false
    Weight=@RerouteNode_5_Constant.Value
   ) {template="Set Relative Transform::Execute(in Child,in Parent,in bParentInitial,in Value,in Weight,in bPropagateToChildren)"}
   > &RigUnit_SetRelativeTransformForItem(
    Child=(Type=Bone,Name="dyn_skirt_front_mid_02")
    Parent=(Type=Bone,Name="dyn_front_skirt_02")
    Value=(Rotation=(X=0,Y=0,Z=-0.0698,W=0.9976),Translation=(X=0,Y=0,Z=0),Scale3D=(X=1,Y=1,Z=1))
    bParentInitial=False
    bPropagateToChildren=false
    Weight=@RerouteNode_5_Constant.Value
   ) {template="Set Relative Transform::Execute(in Child,in Parent,in bParentInitial,in Value,in Weight,in bPropagateToChildren)"}
   > &RigUnit_SetRelativeTransformForItem(
    Child=(Type=Bone,Name="dyn_skirt_front_mid_03")
    Parent=(Type=Bone,Name="dyn_front_skirt_03")
    Value=(Rotation=(X=0,Y=0,Z=0,W=1),Translation=(X=0,Y=1.2,Z=0),Scale3D=(X=1,Y=1,Z=1))
    bParentInitial=False
    bPropagateToChildren=false
    Weight=@RerouteNode_5_Constant.Value
   ) {template="Set Relative Transform::Execute(in Child,in Parent,in bParentInitial,in Value,in Weight,in bPropagateToChildren)"}
   > &RigUnit_SetRelativeTransformForItem(
    Child=(Type=Bone,Name="dyn_skirt_side_l_01")
    Parent=(Type=Bone,Name="dyn_skirt_side_l_01")
    Value=(Rotation=(X=0,Y=0,Z=0.0175,W=0.9998),Translation=(X=0,Y=2,Z=0),Scale3D=(X=1,Y=1,Z=1))
    bParentInitial=False
    bPropagateToChildren=true
    Weight=@RerouteNode_5_Constant.Value
   ) {template="Set Relative Transform::Execute(in Child,in Parent,in bParentInitial,in Value,in Weight,in bPropagateToChildren)"}
   > &RigUnit_SetRelativeTransformForItem(
    Child=(Type=Bone,Name="dyn_skirt_side_r_02")
    Parent=(Type=Bone,Name="dyn_skirt_side_r_02")
    Value=(Rotation=(X=0,Y=0,Z=0.0175,W=0.9998),Translation=(X=0,Y=-2,Z=0),Scale3D=(X=1,Y=1,Z=1))
    bParentInitial=False
    bPropagateToChildren=True
    Weight=@RerouteNode_5_Constant.Value
   ) {template="Set Relative Transform::Execute(in Child,in Parent,in bParentInitial,in Value,in Weight,in bPropagateToChildren)"}
   > &RigUnit_SetRelativeTransformForItem(
    Child=(Type=Bone,Name="dyn_skirt_back_l_01")
    Parent=(Type=Bone,Name="dyn_skirt_back_l_01")
    Value=(Rotation=(X=0,Y=0,Z=0.061,W=0.9981),Translation=(X=0,Y=0,Z=0),Scale3D=(X=1,Y=1,Z=1))
    bParentInitial=False
    bPropagateToChildren=True
    Weight=@RerouteNode_5_Constant.Value
   ) {template="Set Relative Transform::Execute(in Child,in Parent,in bParentInitial,in Value,in Weight,in bPropagateToChildren)"}
   > &RigUnit_SetRelativeTransformForItem(
    Child=(Type=Bone,Name="dyn_skirt_back_r_01")
    Parent=(Type=Bone,Name="dyn_skirt_back_r_01")
    Value=(Rotation=(X=0,Y=0,Z=0.061,W=0.9981),Translation=(X=0,Y=0,Z=0),Scale3D=(X=1,Y=1,Z=1))
    bParentInitial=False
    bPropagateToChildren=True
    Weight=@RerouteNode_5_Constant.Value
   ) {template="Set Relative Transform::Execute(in Child,in Parent,in bParentInitial,in Value,in Weight,in bPropagateToChildren)"}
   Add = &RigVMFunction_MathDoubleAdd(B=@n3.Result, A=@KalmanFilter.Result)
   Clamp = &RigVMFunction_MathDoubleClamp(Value=&RigVMFunction_MathDoubleMul(B=0.5, A=@Add.Result).Result)
   &RigUnit_ModifyTransforms(
    ExecutePin=()
    ItemToModify=((Item=(Type=Bone,Name="dyn_skirt_front_mid_MSTR"),Transform=(Rotation=(X=0,Y=0,Z=0,W=1),Translation=(X=0,Y=0,Z=0),Scale3D=(X=1,Y=1,Z=1))))
    WeightMinimum=0
    Weight=@Clamp.Result
   )
  ]}
 New Function_ContainedGraph {nodes [
   RigVMFunctionEntryNode()
   > RigVMFunction_Sequence_1 = RigVMAggregateNode()
   For_Each_2 = RigVMDispatchNode(ExecuteContext=@RigVMFunction_Sequence_1.A, Array=RigVMVariableNode() {variable=back skirt mid}.Value) {template="DISPATCH_RigVMDispatch_ArrayIterator(in Array,out Element,out Index,out Count,out Ratio)"}
   > &RigUnit_ParentConstraint(
    AdvancedSettings=(InterpolationType=Average,RotationOrderForFilter=XZY)
    Parents=((Item=(Type=Bone,Name="None"),Weight=0.5),(Item=(Type=Bone,Name="None"),Weight=0.5))
    Weight=1
    Parents.0.Item=RigVMDispatchNode(Index=@For_Each_2.Index, Array=RigVMVariableNode() {variable=back skirt l}.Value) {template="DISPATCH_RigVMDispatch_ArrayGetAtIndex(in Array,in Index,out Element)"}.Element
    Parents.1.Item=RigVMDispatchNode(Index=@For_Each_2.Index, Array=RigVMVariableNode() {variable=back skirt r}.Value) {template="DISPATCH_RigVMDispatch_ArrayGetAtIndex(in Array,in Index,out Element)"}.Element
    Child=@For_Each_2.Element
   )
   RigVMFunctionReturnNode(ExecuteContext=@RigVMFunction_Sequence_1.D)
   Remap = &RigVMFunction_MathDoubleRemap(
    SourceMaximum=1
    SourceMinimum=0
    TargetMaximum=0.7
    TargetMinimum=0.25
    Value=&RigUnit_GetCurveValue(Curve=CR Thigh Raised Front L).Value
   )
   Subtract = &RigVMFunction_MathDoubleSub(A=1, B=@Remap.Result)
   &RigUnit_ParentConstraint(
    AdvancedSettings=(InterpolationType=Average,RotationOrderForFilter=XZY)
    Child=(Type=Bone,Name="dyn_skirt_front_mid_MSTR")
    Filter=(TranslationFilter=(bX=True,bY=True,bZ=True),RotationFilter=(bX=true,bY=true,bZ=true),ScaleFilter=(bX=True,bY=True,bZ=True))
    Parents=((Item=(Type=Bone,Name="dyn_skirt_front_l_MSTR"),Weight=0.5),(Item=(Type=Bone,Name="dyn_skirt_front_r_MSTR"),Weight=0.5))
    Weight=1
    ExecutePin=@RigVMFunction_Sequence_1.B
    Parents.0.Weight=@Remap.Result
    Parents.1.Weight=@Subtract.Result
   )
   > &RigUnit_ParentConstraint(
    AdvancedSettings=(InterpolationType=Average,RotationOrderForFilter=XZY)
    Child=(Type=Bone,Name="dyn_skirt_front_mid_01")
    Filter=(TranslationFilter=(bX=True,bY=True,bZ=True),RotationFilter=(bX=true,bY=true,bZ=true),ScaleFilter=(bX=True,bY=True,bZ=True))
    Parents=((Item=(Type=Bone,Name="dyn_skirt_front_l_01"),Weight=0.5),(Item=(Type=Bone,Name="dyn_skirt_front_r_01"),Weight=0.5))
    Weight=1
    Parents.0.Weight=@Remap.Result
    Parents.1.Weight=@Subtract.Result
   )
   > &RigUnit_ParentConstraint(
    AdvancedSettings=(InterpolationType=Average,RotationOrderForFilter=XZY)
    Child=(Type=Bone,Name="dyn_skirt_front_mid_02")
    Filter=(TranslationFilter=(bX=True,bY=True,bZ=True),RotationFilter=(bX=true,bY=true,bZ=true),ScaleFilter=(bX=True,bY=True,bZ=True))
    Parents=((Item=(Type=Bone,Name="dyn_skirt_front_l_02"),Weight=0.5),(Item=(Type=Bone,Name="dyn_skirt_front_r_02"),Weight=0.5))
    Weight=1
    Parents.0.Weight=@Remap.Result
    Parents.1.Weight=@Subtract.Result
   )
   Add = &RigVMFunction_MathDoubleAdd(
    A=&RigUnit_GetCurveValue(Curve=CR Thigh Raised Front L).Value
    B=&RigUnit_GetCurveValue(Curve=CR Thigh Raised Front R).Value
   )
   Multiply = &RigVMFunction_MathDoubleMul(B=0.5, A=@Add.Result)
   Clamp = &RigVMFunction_MathDoubleClamp(Value=@Multiply.Result)
   > &RigUnit_ModifyTransforms(
    ItemToModify=((Item=(Type=Bone,Name="dyn_skirt_front_mid_MSTR"),Transform=(Rotation=(X=0,Y=0,Z=-0.0436,W=0.999),Translation=(X=0,Y=0,Z=0),Scale3D=(X=1,Y=1,Z=1))))
    WeightMinimum=0
    Weight=@Clamp.Result
   )
   > &RigUnit_ModifyTransforms(
    ItemToModify=((Item=(Type=Bone,Name="dyn_skirt_front_mid_MSTR"),Transform=(Rotation=(X=-0.0436,Y=0.0436,Z=0.0019,W=0.9981),Translation=(X=0,Y=0,Z=0),Scale3D=(X=1,Y=1,Z=1))))
    WeightMinimum=0
    Weight=&RigUnit_GetCurveValue(Curve=CR Thigh Raised Front L).Value
   )
   > &RigUnit_ModifyTransforms(
    ItemToModify=((Item=(Type=Bone,Name="dyn_skirt_front_mid_MSTR"),Transform=(Rotation=(X=0.0436,Y=-0.0436,Z=0.0019,W=0.9981),Translation=(X=0,Y=0,Z=0),Scale3D=(X=1,Y=1,Z=1))))
    WeightMinimum=0
    Weight=&RigUnit_GetCurveValue(Curve=CR Thigh Raised Front R).Value
   )
   > &RigUnit_ParentConstraint(
    AdvancedSettings=(InterpolationType=Average,RotationOrderForFilter=XZY)
    Child=(Type=Bone,Name="dyn_skirt_front_mid_03")
    Filter=(TranslationFilter=(bX=True,bY=True,bZ=True),RotationFilter=(bX=true,bY=true,bZ=true),ScaleFilter=(bX=True,bY=True,bZ=True))
    Parents=((Item=(Type=Bone,Name="dyn_skirt_front_l_03"),Weight=0.5),(Item=(Type=Bone,Name="dyn_skirt_front_r_03"),Weight=0.5))
    Weight=1
    Parents.1.Weight=@Subtract.Result
    Parents.0.Weight=@Remap.Result
   )
   > &RigUnit_ParentConstraint(
    AdvancedSettings=(InterpolationType=Average,RotationOrderForFilter=XZY)
    Child=(Type=Bone,Name="dyn_skirt_front_mid_04")
    Filter=(TranslationFilter=(bX=True,bY=True,bZ=True),RotationFilter=(bX=true,bY=true,bZ=true),ScaleFilter=(bX=True,bY=True,bZ=True))
    Parents=((Item=(Type=Bone,Name="dyn_skirt_front_l_04"),Weight=0.5),(Item=(Type=Bone,Name="dyn_skirt_front_r_04"),Weight=0.5))
    Weight=1
    Parents.1.Weight=@Subtract.Result
    Parents.0.Weight=@Remap.Result
   )
   > &RigUnit_ParentConstraint(
    AdvancedSettings=(InterpolationType=Average,RotationOrderForFilter=XZY)
    Child=(Type=Bone,Name="dyn_skirt_front_mid_05")
    Filter=(TranslationFilter=(bX=True,bY=True,bZ=True),RotationFilter=(bX=true,bY=true,bZ=true),ScaleFilter=(bX=True,bY=True,bZ=True))
    Parents=((Item=(Type=Bone,Name="dyn_skirt_front_l_05"),Weight=0.5),(Item=(Type=Bone,Name="dyn_skirt_front_r_05"),Weight=0.5))
    Weight=1
    Parents.1.Weight=@Subtract.Result
    Parents.0.Weight=@Remap.Result
   )
   > &RigUnit_ParentConstraint(
    AdvancedSettings=(InterpolationType=Average,RotationOrderForFilter=XZY)
    Child=(Type=Bone,Name="dyn_skirt_front_mid_06")
    Filter=(TranslationFilter=(bX=True,bY=True,bZ=True),RotationFilter=(bX=true,bY=true,bZ=true),ScaleFilter=(bX=True,bY=True,bZ=True))
    Parents=((Item=(Type=Bone,Name="dyn_skirt_front_l_06"),Weight=0.5),(Item=(Type=Bone,Name="dyn_skirt_front_r_06"),Weight=0.5))
    Weight=1
    Parents.1.Weight=@Subtract.Result
    Parents.0.Weight=@Remap.Result
   )
   &RigUnit_ParentConstraint(
    AdvancedSettings=(InterpolationType=Average,RotationOrderForFilter=XZY)
    Child=(Type=Bone,Name="dyn_skirt_back_mid_MSTR")
    ExecutePin=()
    Parents=((Item=(Type=Bone,Name="None"),Weight=1))
    Weight=1
   )
   > &RigUnit_ParentConstraint(
    AdvancedSettings=(InterpolationType=Average,RotationOrderForFilter=XZY)
    Child=(Type=Bone,Name="dyn_skirt_back_mid_base")
    Parents=((Item=(Type=Bone,Name="None"),Weight=1))
    Weight=1
   )
  ]}
 RigVMFunctionLibrary {nodes [
   Constrain Mid Sim Chains = RigVMCollapseNode(ExecuteContext=())
   Deform Post Sim Skirt Curves = RigVMCollapseNode(ExecuteContext=())
   Get Mirror Transform = RigVMCollapseNode(ExecuteContext=(), transform=(Rotation=(X=0,Y=0,Z=0,W=1),Translation=(X=0,Y=0,Z=0),Scale3D=(X=1,Y=1,Z=1)))
   Get Mirror Transform_1 = RigVMCollapseNode(ExecuteContext=(), transform=(Rotation=(X=0,Y=0,Z=0,W=1),Translation=(X=0,Y=0,Z=0),Scale3D=(X=1,Y=1,Z=1)))
   Get Mirror Transform_2 = RigVMCollapseNode(ExecuteContext=(), transform=(Rotation=(X=0,Y=0,Z=0,W=1),Translation=(X=0,Y=0,Z=0),Scale3D=(X=1,Y=1,Z=1)))
   Modify Transforms with Mirror = RigVMCollapseNode(Base Modify Data=(), ExecuteContext=(), Mirror Modify Data=())
   Modify Transforms with Mirror_1 = RigVMCollapseNode(Base Modify Data=(), ExecuteContext=(), Mirror Modify Data=())
   Modify Transforms with Mirror_2 = RigVMCollapseNode(Base Modify Data=(), ExecuteContext=(), Mirror Modify Data=())
  ]}
 RigVMFunction_Sequence_1_ContainedGraph {nodes [
   n1 = RigVMFunctionEntryNode()
   RigVMFunction_Sequence = &RigVMFunction_Sequence(ExecuteContext=@n1.ExecuteContext)
   n2 = &RigVMFunction_Sequence(ExecuteContext=@RigVMFunction_Sequence.B)
   RigVMFunctionReturnNode(A=@RigVMFunction_Sequence.A, B=@n2.A, C=@n2.B)
  ]}
 RigVMFunction_Sequence_1_ContainedGraph {nodes [
   n1 = RigVMFunctionEntryNode()
   RigVMFunction_Sequence = &RigVMFunction_Sequence(ExecuteContext=@n1.ExecuteContext)
   n2 = &RigVMFunction_Sequence(ExecuteContext=@RigVMFunction_Sequence.B)
   RigVMFunctionReturnNode(A=@RigVMFunction_Sequence.A, B=@n2.A, C=@n2.B)
  ]}
 RigVMFunction_Sequence_1_ContainedGraph {nodes [
   n1 = RigVMFunctionEntryNode()
   RigVMFunction_Sequence = &RigVMFunction_Sequence(ExecuteContext=@n1.ExecuteContext)
   n2 = &RigVMFunction_Sequence(ExecuteContext=@RigVMFunction_Sequence.B)
   RigVMFunctionReturnNode(A=@RigVMFunction_Sequence.A, B=@n2.A, C=@n2.B)
  ]}
 RigVMFunction_Sequence_2_ContainedGraph {nodes [
   n1 = RigVMFunctionEntryNode()
   n2 = &RigVMFunction_Sequence(ExecuteContext=@n1.ExecuteContext)
   n3 = &RigVMFunction_Sequence(ExecuteContext=@n2.B)
   n4 = &RigVMFunction_Sequence(ExecuteContext=@n3.B)
   RigVMFunctionReturnNode(A=@n2.A, B=@n3.A, C=@n4.A, D=@n4.B)
  ]}
 RigVMModel {nodes [
   For_Each_2 = RigVMDispatchNode(ExecuteContext=(), Array=RigVMVariableNode() {variable=back skirt mid}.Value) {template="DISPATCH_RigVMDispatch_ArrayIterator(in Array,out Element,out Index,out Count,out Ratio)"}
   > &RigUnit_ParentConstraint(
    AdvancedSettings=(InterpolationType=Average,RotationOrderForFilter=XZY)
    Parents=((Item=(Type=Bone,Name="None"),Weight=0.5),(Item=(Type=Bone,Name="None"),Weight=0.5))
    Weight=1
    Parents.0.Item=RigVMDispatchNode(Index=@For_Each_2.Index, Array=RigVMVariableNode() {variable=back skirt l}.Value) {template="DISPATCH_RigVMDispatch_ArrayGetAtIndex(in Array,in Index,out Element)"}.Element
    Parents.1.Item=RigVMDispatchNode(Index=@For_Each_2.Index, Array=RigVMVariableNode() {variable=back skirt r}.Value) {template="DISPATCH_RigVMDispatch_ArrayGetAtIndex(in Array,in Index,out Element)"}.Element
    Child=@For_Each_2.Element
   )
   Add = &RigVMFunction_MathDoubleAdd(
    A=&RigUnit_GetCurveValue(Curve=CR Thigh Raised Front L).Value
    B=&RigUnit_GetCurveValue(Curve=CR Thigh Raised Front R).Value
   )
   Multiply = &RigVMFunction_MathDoubleMul(B=0.5, A=@Add.Result)
   Clamp = &RigVMFunction_MathDoubleClamp(Value=@Multiply.Result)
   &RigUnit_ModifyTransforms(
    ExecutePin=()
    ItemToModify=((Item=(Type=Bone,Name="dyn_skirt_front_mid_MSTR"),Transform=(Rotation=(X=0,Y=0,Z=-0.1736,W=0.9848),Translation=(X=0,Y=0,Z=0),Scale3D=(X=1,Y=1,Z=1))))
    WeightMinimum=0
    Weight=@Clamp.Result
   )
   > &RigUnit_ModifyTransforms(
    ItemToModify=((Item=(Type=Bone,Name="dyn_skirt_front_mid_MSTR"),Transform=(Rotation=(X=-0.0868,Y=0.0868,Z=0.0076,W=0.9924),Translation=(X=0,Y=0,Z=0),Scale3D=(X=1,Y=1,Z=1))))
    WeightMinimum=0
    Weight=&RigUnit_GetCurveValue(Curve=CR Thigh Raised Front L).Value
   )
   > &RigUnit_ModifyTransforms(
    ItemToModify=((Item=(Type=Bone,Name="dyn_skirt_front_mid_MSTR"),Transform=(Rotation=(X=0.0868,Y=-0.0868,Z=0.0076,W=0.9924),Translation=(X=0,Y=0,Z=0),Scale3D=(X=1,Y=1,Z=1))))
    WeightMinimum=0
    Weight=&RigUnit_GetCurveValue(Curve=CR Thigh Raised Front R).Value
   )
   n1 = &RigVMFunction_MathDoubleAdd(
    A=&RigUnit_GetCurveValue(Curve=CR Thigh Raised Front L).Value
    B=&RigUnit_GetCurveValue(Curve=CR Thigh Raised Front R).Value
   )
   &RigUnit_ModifyTransforms(
    ExecutePin=()
    ItemToModify=((Item=(Type=Bone,Name="dyn_skirt_front_mid_MSTR"),Transform=(Rotation=(X=0,Y=0,Z=-0.1736,W=0.9848),Translation=(X=0,Y=0,Z=0),Scale3D=(X=1,Y=1,Z=1))))
    WeightMinimum=0
    Weight=&RigVMFunction_MathDoubleClamp(Value=&RigVMFunction_MathDoubleMul(B=0.5, A=@n1.Result).Result).Result
   )
   > &RigUnit_ModifyTransforms(
    ItemToModify=((Item=(Type=Bone,Name="dyn_skirt_front_mid_MSTR"),Transform=(Rotation=(X=-0.0868,Y=0.0868,Z=0.0076,W=0.9924),Translation=(X=0,Y=0,Z=0),Scale3D=(X=1,Y=1,Z=1))))
    WeightMinimum=0
    Weight=&RigUnit_GetCurveValue(Curve=CR Thigh Raised Front L).Value
   )
   > &RigUnit_ModifyTransforms(
    ItemToModify=((Item=(Type=Bone,Name="dyn_skirt_front_mid_MSTR"),Transform=(Rotation=(X=0.0868,Y=-0.0868,Z=0.0076,W=0.9924),Translation=(X=0,Y=0,Z=0),Scale3D=(X=1,Y=1,Z=1))))
    WeightMinimum=0
    Weight=&RigUnit_GetCurveValue(Curve=CR Thigh Raised Front R).Value
   )
   n2 = RigVMRerouteNode(Value=())
   Remap = &RigVMFunction_MathDoubleRemap(
    SourceMaximum=1
    SourceMinimum=0
    TargetMaximum=0.7
    TargetMinimum=0.5
    Value=&RigUnit_GetCurveValue(Curve=CR Thigh Raised Front L).Value
   )
   Subtract = &RigVMFunction_MathDoubleSub(A=1, B=@Remap.Result)
   &RigUnit_ParentConstraint(
    AdvancedSettings=(InterpolationType=Average,RotationOrderForFilter=XZY)
    Child=(Type=Bone,Name="dyn_skirt_front_mid_MSTR")
    Filter=(TranslationFilter=(bX=True,bY=True,bZ=True),RotationFilter=(bX=true,bY=true,bZ=true),ScaleFilter=(bX=True,bY=True,bZ=True))
    Parents=((Item=(Type=Bone,Name="dyn_skirt_front_l_MSTR"),Weight=0.5),(Item=(Type=Bone,Name="dyn_skirt_front_r_MSTR"),Weight=0.5))
    Weight=1
    ExecutePin=@n2.Value
    Parents.0.Weight=@Remap.Result
    Parents.1.Weight=@Subtract.Result
   )
   > &RigUnit_ParentConstraint(
    AdvancedSettings=(InterpolationType=Average,RotationOrderForFilter=XZY)
    Child=(Type=Bone,Name="dyn_skirt_front_mid_01")
    Filter=(TranslationFilter=(bX=True,bY=True,bZ=True),RotationFilter=(bX=true,bY=true,bZ=true),ScaleFilter=(bX=True,bY=True,bZ=True))
    Parents=((Item=(Type=Bone,Name="dyn_skirt_front_l_01"),Weight=0.5),(Item=(Type=Bone,Name="dyn_skirt_front_r_01"),Weight=0.5))
    Weight=1
    Parents.0.Weight=@Remap.Result
    Parents.1.Weight=@Subtract.Result
   )
   > &RigUnit_ParentConstraint(
    AdvancedSettings=(InterpolationType=Average,RotationOrderForFilter=XZY)
    Child=(Type=Bone,Name="dyn_skirt_front_mid_02")
    Filter=(TranslationFilter=(bX=True,bY=True,bZ=True),RotationFilter=(bX=true,bY=true,bZ=true),ScaleFilter=(bX=True,bY=True,bZ=True))
    Parents=((Item=(Type=Bone,Name="dyn_skirt_front_l_02"),Weight=0.5),(Item=(Type=Bone,Name="dyn_skirt_front_r_02"),Weight=0.5))
    Weight=1
    Parents.0.Weight=@Remap.Result
    Parents.1.Weight=@Subtract.Result
   )
   > &RigUnit_ParentConstraint(
    AdvancedSettings=(InterpolationType=Average,RotationOrderForFilter=XZY)
    Child=(Type=Bone,Name="dyn_skirt_front_mid_03")
    Filter=(TranslationFilter=(bX=True,bY=True,bZ=True),RotationFilter=(bX=true,bY=true,bZ=true),ScaleFilter=(bX=True,bY=True,bZ=True))
    Parents=((Item=(Type=Bone,Name="dyn_skirt_front_l_03"),Weight=0.5),(Item=(Type=Bone,Name="dyn_skirt_front_r_03"),Weight=0.5))
    Weight=1
    Parents.1.Weight=@Subtract.Result
    Parents.0.Weight=@Remap.Result
   )
   > &RigUnit_ParentConstraint(
    AdvancedSettings=(InterpolationType=Average,RotationOrderForFilter=XZY)
    Child=(Type=Bone,Name="dyn_skirt_front_mid_04")
    Filter=(TranslationFilter=(bX=True,bY=True,bZ=True),RotationFilter=(bX=true,bY=true,bZ=true),ScaleFilter=(bX=True,bY=True,bZ=True))
    Parents=((Item=(Type=Bone,Name="dyn_skirt_front_l_04"),Weight=0.5),(Item=(Type=Bone,Name="dyn_skirt_front_r_04"),Weight=0.5))
    Weight=1
    Parents.1.Weight=@Subtract.Result
    Parents.0.Weight=@Remap.Result
   )
   > &RigUnit_ParentConstraint(
    AdvancedSettings=(InterpolationType=Average,RotationOrderForFilter=XZY)
    Child=(Type=Bone,Name="dyn_skirt_front_mid_05")
    Filter=(TranslationFilter=(bX=True,bY=True,bZ=True),RotationFilter=(bX=true,bY=true,bZ=true),ScaleFilter=(bX=True,bY=True,bZ=True))
    Parents=((Item=(Type=Bone,Name="dyn_skirt_front_l_05"),Weight=0.5),(Item=(Type=Bone,Name="dyn_skirt_front_r_05"),Weight=0.5))
    Weight=1
    Parents.1.Weight=@Subtract.Result
    Parents.0.Weight=@Remap.Result
   )
   > &RigUnit_ParentConstraint(
    AdvancedSettings=(InterpolationType=Average,RotationOrderForFilter=XZY)
    Child=(Type=Bone,Name="dyn_skirt_front_mid_06")
    Filter=(TranslationFilter=(bX=True,bY=True,bZ=True),RotationFilter=(bX=true,bY=true,bZ=true),ScaleFilter=(bX=True,bY=True,bZ=True))
    Parents=((Item=(Type=Bone,Name="dyn_skirt_front_l_06"),Weight=0.5),(Item=(Type=Bone,Name="dyn_skirt_front_r_06"),Weight=0.5))
    Weight=1
    Parents.1.Weight=@Subtract.Result
    Parents.0.Weight=@Remap.Result
   )
   n3 = &RigVMFunction_MathDoubleRemap(
    SourceMaximum=1
    SourceMinimum=0
    TargetMaximum=0.7
    TargetMinimum=0.5
    Value=&RigUnit_GetCurveValue(Curve=CR Thigh Raised Front L).Value
   )
   n4 = &RigVMFunction_MathDoubleSub(A=1, B=@n3.Result)
   > &RigUnit_ParentConstraint(
    AdvancedSettings=(InterpolationType=Average,RotationOrderForFilter=XZY)
    Child=(Type=Bone,Name="dyn_skirt_back_mid_MSTR")
    Filter=(TranslationFilter=(bX=True,bY=True,bZ=True),RotationFilter=(bX=true,bY=true,bZ=true),ScaleFilter=(bX=True,bY=True,bZ=True))
    Parents=((Item=(Type=Bone,Name="dyn_skirt_back_l_MSTR"),Weight=0.5),(Item=(Type=Bone,Name="dyn_skirt_back_r_MSTR"),Weight=0.5))
    Weight=1
    Parents.0.Weight=@n3.Result
    Parents.1.Weight=@n4.Result
   )
   > &RigUnit_ParentConstraint(
    AdvancedSettings=(InterpolationType=Average,RotationOrderForFilter=XZY)
    Child=(Type=Bone,Name="dyn_skirt_back_mid_01")
    Filter=(TranslationFilter=(bX=True,bY=True,bZ=True),RotationFilter=(bX=true,bY=true,bZ=true),ScaleFilter=(bX=True,bY=True,bZ=True))
    Parents=((Item=(Type=Bone,Name="dyn_skirt_back_l_01"),Weight=0.5),(Item=(Type=Bone,Name="dyn_skirt_back_r_01"),Weight=0.5))
    Weight=1
    Parents.0.Weight=@n3.Result
    Parents.1.Weight=@n4.Result
   )
   > &RigUnit_ParentConstraint(
    AdvancedSettings=(InterpolationType=Average,RotationOrderForFilter=XZY)
    Child=(Type=Bone,Name="dyn_skirt_back_mid_02")
    Filter=(TranslationFilter=(bX=True,bY=True,bZ=True),RotationFilter=(bX=true,bY=true,bZ=true),ScaleFilter=(bX=True,bY=True,bZ=True))
    Parents=((Item=(Type=Bone,Name="dyn_skirt_back_l_02"),Weight=0.5),(Item=(Type=Bone,Name="dyn_skirt_back_r_02"),Weight=0.5))
    Weight=1
    Parents.0.Weight=@n3.Result
    Parents.1.Weight=@n4.Result
   )
   > &RigUnit_ParentConstraint(
    AdvancedSettings=(InterpolationType=Average,RotationOrderForFilter=XZY)
    Child=(Type=Bone,Name="dyn_skirt_back_mid_03")
    Filter=(TranslationFilter=(bX=True,bY=True,bZ=True),RotationFilter=(bX=true,bY=true,bZ=true),ScaleFilter=(bX=True,bY=True,bZ=True))
    Parents=((Item=(Type=Bone,Name="dyn_skirt_back_l_03"),Weight=0.5),(Item=(Type=Bone,Name="dyn_skirt_back_r_03"),Weight=0.5))
    Weight=1
    Parents.1.Weight=@n4.Result
    Parents.0.Weight=@n3.Result
   )
   > &RigUnit_ParentConstraint(
    AdvancedSettings=(InterpolationType=Average,RotationOrderForFilter=XZY)
    Child=(Type=Bone,Name="dyn_skirt_back_mid_04")
    Filter=(TranslationFilter=(bX=True,bY=True,bZ=True),RotationFilter=(bX=true,bY=true,bZ=true),ScaleFilter=(bX=True,bY=True,bZ=True))
    Parents=((Item=(Type=Bone,Name="dyn_skirt_back_l_04"),Weight=0.5),(Item=(Type=Bone,Name="dyn_skirt_back_r_04"),Weight=0.5))
    Weight=1
    Parents.1.Weight=@n4.Result
    Parents.0.Weight=@n3.Result
   )
   > &RigUnit_ParentConstraint(
    AdvancedSettings=(InterpolationType=Average,RotationOrderForFilter=XZY)
    Child=(Type=Bone,Name="dyn_skirt_back_mid_05")
    Filter=(TranslationFilter=(bX=True,bY=True,bZ=True),RotationFilter=(bX=true,bY=true,bZ=true),ScaleFilter=(bX=True,bY=True,bZ=True))
    Parents=((Item=(Type=Bone,Name="dyn_skirt_back_l_05"),Weight=0.5),(Item=(Type=Bone,Name="dyn_skirt_back_r_05"),Weight=0.5))
    Weight=1
    Parents.1.Weight=@n4.Result
    Parents.0.Weight=@n3.Result
   )
   > &RigUnit_ParentConstraint(
    AdvancedSettings=(InterpolationType=Average,RotationOrderForFilter=XZY)
    Child=(Type=Bone,Name="dyn_skirt_back_mid_06")
    Filter=(TranslationFilter=(bX=True,bY=True,bZ=True),RotationFilter=(bX=true,bY=true,bZ=true),ScaleFilter=(bX=True,bY=True,bZ=True))
    Parents=((Item=(Type=Bone,Name="dyn_skirt_back_l_06"),Weight=0.5),(Item=(Type=Bone,Name="dyn_skirt_back_r_06"),Weight=0.5))
    Weight=1
    Parents.1.Weight=@n4.Result
    Parents.0.Weight=@n3.Result
   )
   n5 = &RigUnit_GetCurveValue(Curve=CR Thigh Raised Front R)
   > &RigUnit_ParentConstraint(
    AdvancedSettings=(InterpolationType=Average,RotationOrderForFilter=XZY)
    Child=(Type=Bone,Name="dyn_fold_back2_01")
    Parents=((Item=(Type=Bone,Name="dyn_skirt_back_r_MSTR"),Weight=1),(Item=(Type=Bone,Name="dyn_skirt_back_mid_MSTR"),Weight=1))
    Weight=@n5.Value
   )
   > &RigUnit_ParentConstraint(
    AdvancedSettings=(InterpolationType=Average,RotationOrderForFilter=XZY)
    Child=(Type=Bone,Name="dyn_fold_back1_01")
    Parents=((Item=(Type=Bone,Name="dyn_skirt_back_r_MSTR"),Weight=1),(Item=(Type=Bone,Name="dyn_skirt_back_mid_MSTR"),Weight=1))
    Weight=@n5.Value
   )
   n6 = &RigUnit_GetCurveValue(Curve=CR Thigh Raised Front R)
   > &RigUnit_ParentConstraint(
    AdvancedSettings=(InterpolationType=Average,RotationOrderForFilter=XZY)
    Child=(Type=Bone,Name="dyn_fold_front4_01")
    Parents=((Item=(Type=Bone,Name="dyn_skirt_front_r_MSTR"),Weight=1))
    Weight=@n6.Value
   )
   > &RigUnit_ParentConstraint(
    AdvancedSettings=(InterpolationType=Average,RotationOrderForFilter=XZY)
    Child=(Type=Bone,Name="dyn_fold_front5_01")
    Parents=((Item=(Type=Bone,Name="dyn_skirt_front_r_MSTR"),Weight=1),(Item=(Type=Bone,Name="dyn_skirt_front_mid_MSTR"),Weight=1))
    Weight=@n6.Value
   )
   n7 = RigVMRerouteNode(Value=())
   RigVMFunction_MathVectorDistance = &RigVMFunction_MathVectorDistance(
    A=&RigUnit_GetTransform(Item=(Type=Bone,Name="pelvis"), bInitial=False).Transform.Translation
    B=&RigUnit_GetTransform(Item=(Type=Bone,Name="root"), bInitial=False).Transform.Translation
   )
   &RigUnit_ModifyTransforms(
    ItemToModify="((Item=(Type=Bone,Name=\"dyn_skirt_front_l_01\"),Transform=(Rotation=(X=0,Y=0,Z=0,W=1),Translation=(X=0,Y=35,Z=0),Scale3D=(X=1,Y=1,Z=1))),(Item=(Type=Bone,Name=\"dyn_skirt_front_r_01\"),Transform=(Rotation=(X=0,Y=0,Z=0,W=1),Translation=(X=0,Y=-5,Z=0),Scale3D=( ...(truncated)"
    WeightMinimum=0
    ExecutePin=@n7.Value
    Weight=&RigVMFunction_MathDoubleSub(
     A=1
     B=&RigVMFunction_MathDoubleRemap(SourceMaximum=90, SourceMinimum=45, TargetMaximum=1, TargetMinimum=0, Value=@RigVMFunction_MathVectorDistance.Result).Result
    ).Result
   )
   RigUnit_BeginExecution = &RigUnit_BeginExecution()
   Constrain Mid Sim Chains = RigVMFunctionReferenceNode(ExecuteContext=@RigUnit_BeginExecution.ExecutePin)
   > Deform Post Sim Skirt Curves = RigVMFunctionReferenceNode()
  ]}
 RigVMModel Construction Event Graph {nodes [
   n1 = &RigUnit_PrepareForExecution()
   RigUnit_GetTransform = &RigUnit_GetTransform(Item=(Type=Bone,Name="thigh_l"), Space=LocalSpace, bInitial=False)
   RigVMVariableNode(ExecuteContext=@n1.ExecutePin, Value=@RigUnit_GetTransform.Transform.Rotation) {variable=prev thigh l quaternion}
   > RigVMVariableNode(Value=&RigUnit_GetTransform(Item=(Type=Bone,Name="thigh_r"), Space=LocalSpace, bInitial=False).Transform.Rotation) {variable=prev thigh r quaternion}
  ]}
]
hierarchy [
 root {loc=<0 0 0>, rot=<0 0 0 1>}
 pelvis {parent=root, loc=<0 -0.7071 93.1191>, rot=<0.0127 0.707 -0.0127 -0.707>}
 spine_01 {parent=pelvis, loc=<6.7494 -0.6388 0>, rot=<0 0 0.034 -0.9994>}
 spine_02 {parent=spine_01, loc=<7.2518 -0.035 0>, rot=<0 0 0.0238 -0.9997>}
 spine_03 {parent=spine_02, loc=<6.0104 0 0>, rot=<0 0 0.0149 -0.9999>}
 spine_04 {parent=spine_03, loc=<8.1024 0 0>, rot=<0 0 -0.0132 -0.9999>}
 spine_05 {parent=spine_04, loc=<10.0146 0 0>, rot=<0 0 -0.0015 -1>}
 neck_01 {parent=spine_05, loc=<11.3138 0.1529 0>, rot=<0 0 -0.2003 -0.9797>}
 neck_02 {parent=neck_01, loc=<4.5627 0 0>, rot=<0 0 0.0504 -0.9987>}
 head {parent=neck_02, loc=<5.3914 0 0>, rot=<0 0 0.115 -0.9934>}
 faceAttach {parent=head, loc=<2.2569 0.2713 0>, rot=<0.0029 -0.7071 -0.0029 -0.7071>}
 C_jaw {parent=faceAttach, loc=<-0.0134 -2.1613 -1.4291>, rot=<0 0 0 1>}
 L_lip_lower_outer {parent=C_jaw, loc=<1.6885 9.2004 -1.7608>, rot=<0.013 0.0445 0.3057 -0.951>}
 C_lip_lower_mid {parent=C_jaw, loc=<0.0223 10.0405 -1.7504>, rot=<0 0 0 -1>}
 R_lip_lower_outer {parent=C_jaw, loc=<-1.6617 9.2004 -1.7608>, rot=<0.013 -0.0445 -0.3057 -0.951>}
 teeth_lower {parent=C_jaw, loc=<0.0134 6.0835 -1.7637>, rot=<0 0 0 1>}
 tongue {parent=C_jaw, loc=<0.0134 3.8373 -2.8233>, rot=<-0.1568 0 0 -0.9876>}
 R_lip_upper_outer {parent=faceAttach, loc=<-1.6653 6.862 -1.9706>, rot=<0.0359 0.0728 -0.2627 -0.9615>}
 teeth_upper {parent=faceAttach, loc=<0 4.5028 -1.1408>, rot=<0 0 0 1>}
 L_eye {parent=faceAttach, loc=<2.1586 4.7722 4.3083>, rot=<0 0 0.9998 0.0175>}
 L_eye_lid_upper_mid {parent=L_eye, loc=<0 0 0>, rot=<0.0523 0 0 -0.9986>}
 L_eye_lid_lower_mid {parent=L_eye, loc=<0 0 0>, rot=<0.0329 0 0 -0.9995>}
 R_eye {parent=faceAttach, loc=<-2.1586 4.7721 4.3083>, rot=<0 0 0.9998 -0.0175>}
 R_eye_lid_upper_mid {parent=R_eye, loc=<0 0 0>, rot=<0.0523 0 0 -0.9986>}
 R_eye_lid_lower_mid {parent=R_eye, loc=<0 0 0>, rot=<0.0329 0 0 -0.9995>}
 R_lip_corner {parent=faceAttach, loc=<-2.5405 5.4696 -2.5881>, rot=<0.003 -0.008 0.0146 -0.9999>}
 R_cheek_inner {parent=faceAttach, loc=<-2.6217 5.5876 0.6525>, rot=<-0.0706 0.2697 -0.3184 -0.906>}
 R_brow_outer {parent=faceAttach, loc=<-3.7188 5.5117 6.0529>, rot=<0.0819 0.0868 -0.316 -0.9412>}
 R_brow_mid {parent=faceAttach, loc=<-2.5316 5.8798 7.2382>, rot=<-0.0001 0.0042 -0.2553 -0.9669>}
 C_nose_bridge {parent=faceAttach, loc=<0 8.4307 3.1188>, rot=<-0.2218 0 0 -0.9751>}
 L_lip_corner {parent=faceAttach, loc=<2.5405 5.4696 -2.5881>, rot=<0.003 0.008 -0.0146 -0.9999>}
 L_lip_upper_outer {parent=faceAttach, loc=<1.6653 6.862 -1.9706>, rot=<0.0359 -0.0728 0.2627 -0.9615>}
 C_lip_upper_mid {parent=faceAttach, loc=<0.0022 7.8658 -1.9421>, rot=<0.2225 0.0013 -0.0005 -0.9749>}
 L_cheek_inner {parent=faceAttach, loc=<2.6217 5.5876 0.6525>, rot=<-0.0706 -0.2697 0.3184 -0.906>}
 L_brow_outer {parent=faceAttach, loc=<3.9625 4.5207 6.4924>, rot=<-0.0287 -0.04 0.4491 -0.8921>}
 L_brow_mid {parent=faceAttach, loc=<2.5316 5.8798 7.2382>, rot=<-0.0001 -0.0042 0.2553 -0.9669>}
 C_brow_mid {parent=faceAttach, loc=<0 7.163 7.1397>, rot=<0 0 0 -1>}
 attach_cape {parent=spine_05, loc=<8.9532 -11.681 0.0524>, rot=<0.0616 -0.7044 -0.0616 -0.7044>}
 cape {parent=attach_cape, type=Null, loc=<0 0 0>, rot=<0 0 0 1>}
 attach_backpack {parent=spine_05, loc=<-0.3552 -13.6178 0.0435>, rot=<-0.0093 -0.707 0.0092 -0.707>}
 Backpack_BR {parent=attach_backpack, type=Null, loc=<0 -0.5774 0.0898>, rot=<-0.0771 0 0 0.997>}
 clavicle_l {parent=spine_05, loc=<4.3575 -0.1466 -2.3223>, rot=<0.0465 -0.7289 -0.009 -0.6829>}
 upperarm_l {parent=clavicle_l, loc=<15.5111 0 0>, rot=<0.0226 -0.3234 -0.0088 -0.9459>}
 lowerarm_l {parent=upperarm_l, loc=<26.6514 0 0>, rot=<0.0391 -0.0099 -0.2453 -0.9686>}
 hand_l {parent=lowerarm_l, loc=<22.3491 0 0>, rot=<-0.7081 -0.0507 0.0975 -0.6975>}
 thumb_01_l {parent=hand_l, loc=<1.3118 -0.3191 -2.459>, rot=<0.6546 -0.4042 0.0664 -0.6354>}
 thumb_02_l {parent=thumb_01_l, loc=<5.9336 -0.0448 0.0401>, rot=<-0.0037 0.0213 0.298 -0.9543>}
 thumb_03_l {parent=thumb_02_l, loc=<2.2223 0.3682 -0.0907>, rot=<0.0118 -0.013 0.0066 -0.9998>}
 index_metacarpal_l {parent=hand_l, loc=<1.6535 0.0183 -2.96>, rot=<0.0158 -0.0124 -0.0173 -0.9996>}
 index_01_l {parent=index_metacarpal_l, loc=<8.3008 0.0357 -0.0524>, rot=<-0.0044 0.0485 0.0924 -0.9945>}
 index_02_l {parent=index_01_l, loc=<3.3587 -0.3004 0.0452>, rot=<-0.002 0.0027 0.3225 -0.9466>}
 index_03_l {parent=index_02_l, loc=<2.5578 -0.123 -0.0252>, rot=<-0.002 0.0067 0.2558 -0.9667>}
 middle_metacarpal_l {parent=hand_l, loc=<2.1856 -0.1725 -0.5984>, rot=<0.006 0.0168 -0.0329 -0.9993>}
 middle_01_l {parent=middle_metacarpal_l, loc=<7.3275 -0.2447 0.0129>, rot=<-0.0208 0.0042 0.0858 -0.9961>}
 middle_02_l {parent=middle_01_l, loc=<4.0123 0.1401 0.1175>, rot=<-0.0218 0.0258 0.3512 -0.9357>}
 middle_03_l {parent=middle_02_l, loc=<2.6077 0.0534 0.1373>, rot=<0.0082 0.0188 0.1708 -0.9851>}
 ring_metacarpal_l {parent=hand_l, loc=<2.2074 -0.3625 1.3638>, rot=<-0.0284 0.0722 -0.0458 -0.9959>}
 ring_01_l {parent=ring_metacarpal_l, loc=<7.0295 -0.6197 -0.1446>, rot=<-0.0489 -0.0252 0.1892 -0.9804>}
 ring_02_l {parent=ring_01_l, loc=<3.6174 0.4647 -0.1027>, rot=<0 0 0.2806 -0.9598>}
 ring_03_l {parent=ring_02_l, loc=<2.4152 0.114 0.0199>, rot=<0 0 0.1189 -0.9929>}
 pinky_metacarpal_l {parent=hand_l, loc=<1.7616 -0.9753 3.3814>, rot=<-0.0397 0.0814 -0.069 -0.9935>}
 pinky_01_l {parent=pinky_metacarpal_l, loc=<7.1709 -0.4831 -0.1094>, rot=<-0.0654 -0.0035 0.2242 -0.9723>}
 pinky_02_l {parent=pinky_01_l, loc=<2.8274 0.0893 -0.1043>, rot=<0.0037 0.0011 0.2854 -0.9584>}
 pinky_03_l {parent=pinky_02_l, loc=<2.2046 0.0307 -0.0078>, rot=<-0.0002 0.0098 0.1445 -0.9895>}
 deform_wrist_dn_base_l {parent=hand_l, loc=<0 -2.2342 0>, rot=<0 0 0 -1>}
 deform_wrist_dn_l {parent=deform_wrist_dn_base_l, loc=<0 0 0>, rot=<0 0 0 1>}
 deform_wrist_up_base_l {parent=hand_l, loc=<0 2.2682 0>, rot=<0 0 0 -1>}
 deform_wrist_up_l {parent=deform_wrist_up_base_l, loc=<0 0 0>, rot=<0 0 0 1>}
 weapon_l {parent=hand_l, loc=<-0.1974 -3.4111 0.3811>, rot=<0 0 0 1>}
 hand_attach_l {parent=hand_l, loc=<8.5 -3 0>, rot=<0.7071 0.7071 0 0>}
 lowerarm_twist_02_l {parent=lowerarm_l, loc=<9.6199 0 0>, rot=<0 0 0 -1>}
 lowerarm_twist_01_l {parent=lowerarm_l, loc=<17.8549 0 0>, rot=<0 0 0 -1>}
 deform_elbow_in_base_l {parent=lowerarm_l, loc=<0.2112 5.259 0.8274>, rot=<0 0 0 -1>}
 deform_elbow_in_l {parent=deform_elbow_in_base_l, loc=<0 0 0>, rot=<0 0 0 1>}
 deform_elbow_out_base_l {parent=lowerarm_l, loc=<0 -4 0>, rot=<0 0 0 -1>}
 deform_elbow_out_l {parent=deform_elbow_out_base_l, loc=<0 0 0>, rot=<0 0 0 1>}
 upperarm_twist_01_l {parent=upperarm_l, loc=<0.0041 0 0>, rot=<0 0 0 -1>}
 deform_upperarm_bk_base_l {parent=upperarm_twist_01_l, loc=<1.0813 -7.2193 -1.9453>, rot=<0 0 0 1>}
 deform_upperarm_bk_l {parent=deform_upperarm_bk_base_l, loc=<0 0 0>, rot=<0 0 0 1>}
 deform_upperarm_fr_base_l {parent=upperarm_twist_01_l, loc=<-0.3691 7.0506 -2.765>, rot=<0 0 0 1>}
 deform_upperarm_fr_l {parent=deform_upperarm_fr_base_l, loc=<0 0 0>, rot=<0 0 0 1>}
 deform_upperarm_up_base_l {parent=upperarm_twist_01_l, loc=<-3.2216 -1.1391 3.3311>, rot=<0 0 0 1>}
 deform_upperarm_up_l {parent=deform_upperarm_up_base_l, loc=<0 0 0>, rot=<0 0 0 1>}
 upperarm_twist_02_l {parent=upperarm_l, loc=<13.7941 0 0>, rot=<0 0 0 -1>}
 dyn_upperarm_sleeve_l_MSTR {parent=upperarm_l, loc=<-0.0001 0 -0.0001>, rot=<0 0 0 -1>}
 dyn_upperarm_l_sleeve_twist_01_MSTR {parent=dyn_upperarm_sleeve_l_MSTR, loc=<10.0366 0 -0.1>, rot=<0 0 0 1>}
 dyn_upperarm_l_sleeve_twist_01 {parent=dyn_upperarm_l_sleeve_twist_01_MSTR, loc=<0 0 0>, rot=<0 0 0 1>}
 dyn_upperarm_l_sleeve_twist_01_end {parent=dyn_upperarm_l_sleeve_twist_01, loc=<0 0 -6>, rot=<0 0 0 1>}
 dyn_upperarm_l_sleeve_twist_02_MSTR {parent=dyn_upperarm_sleeve_l_MSTR, loc=<15.3484 0 -0.1>, rot=<0 0 0 -1>}
 dyn_upperarm_l_sleeve_twist_02 {parent=dyn_upperarm_l_sleeve_twist_02_MSTR, loc=<0 0 0>, rot=<0 0 0 -1>}
 dyn_upperarm_l_sleeve_twist_02_end {parent=dyn_upperarm_l_sleeve_twist_02, loc=<0 0 -7>, rot=<0 0 0 1>}
 dyn_upperarm_l_sleeve_twist_03_MSTR {parent=dyn_upperarm_sleeve_l_MSTR, loc=<20.9241 0 -0.1>, rot=<0 0 0 -1>}
 dyn_upperarm_l_sleeve_twist_03 {parent=dyn_upperarm_l_sleeve_twist_03_MSTR, loc=<0 0 0>, rot=<0 0 0 -1>}
 dyn_upperarm_l_sleeve_twist_03_end {parent=dyn_upperarm_l_sleeve_twist_03, loc=<0 0 -8>, rot=<0 0 0 1>}
 clavicle_r {parent=spine_05, loc=<4.3575 -0.1466 2.3223>, rot=<0.7289 0.0465 -0.6829 0.009>}
 upperarm_r {parent=clavicle_r, loc=<-15.5111 0 0>, rot=<0.0226 -0.3234 -0.0088 -0.9459>}
 lowerarm_r {parent=upperarm_r, loc=<-26.6514 0 0>, rot=<0.0391 -0.0099 -0.2453 -0.9686>}
 hand_r {parent=lowerarm_r, loc=<-22.3491 0 0>, rot=<-0.7081 -0.0507 0.0975 -0.6975>}
 thumb_01_r {parent=hand_r, loc=<-1.3118 0.3191 2.459>, rot=<0.6546 -0.4042 0.0664 -0.6354>}
 thumb_02_r {parent=thumb_01_r, loc=<-5.9336 0.0448 -0.0401>, rot=<-0.0037 0.0213 0.298 -0.9543>}
 thumb_03_r {parent=thumb_02_r, loc=<-2.2223 -0.3682 0.0907>, rot=<0.0118 -0.013 0.0066 -0.9998>}
 index_metacarpal_r {parent=hand_r, loc=<-1.6535 -0.0183 2.96>, rot=<0.0158 -0.0124 -0.0173 -0.9996>}
 index_01_r {parent=index_metacarpal_r, loc=<-8.3008 -0.0357 0.0524>, rot=<-0.0044 0.0485 0.0924 -0.9945>}
 index_02_r {parent=index_01_r, loc=<-3.3587 0.3004 -0.0452>, rot=<-0.002 0.0027 0.3225 -0.9466>}
 index_03_r {parent=index_02_r, loc=<-2.5578 0.123 0.0252>, rot=<-0.002 0.0067 0.2558 -0.9667>}
 middle_metacarpal_r {parent=hand_r, loc=<-2.1856 0.1725 0.5984>, rot=<0.006 0.0168 -0.0329 -0.9993>}
 middle_01_r {parent=middle_metacarpal_r, loc=<-7.3275 0.2447 -0.0129>, rot=<-0.0208 0.0042 0.0858 -0.9961>}
 middle_02_r {parent=middle_01_r, loc=<-4.0123 -0.1401 -0.1175>, rot=<-0.0218 0.0258 0.3512 -0.9357>}
 middle_03_r {parent=middle_02_r, loc=<-2.6077 -0.0534 -0.1373>, rot=<0.0082 0.0188 0.1708 -0.9851>}
 ring_metacarpal_r {parent=hand_r, loc=<-2.2074 0.3625 -1.3638>, rot=<-0.0284 0.0722 -0.0458 -0.9959>}
 ring_01_r {parent=ring_metacarpal_r, loc=<-7.0295 0.6197 0.1446>, rot=<-0.0489 -0.0252 0.1892 -0.9804>}
 ring_02_r {parent=ring_01_r, loc=<-3.6174 -0.4647 0.1027>, rot=<0 0 0.2806 -0.9598>}
 ring_03_r {parent=ring_02_r, loc=<-2.4152 -0.114 -0.0199>, rot=<0 0 0.1189 -0.9929>}
 pinky_metacarpal_r {parent=hand_r, loc=<-1.7616 0.9753 -3.3814>, rot=<-0.0397 0.0814 -0.069 -0.9935>}
 pinky_01_r {parent=pinky_metacarpal_r, loc=<-7.1709 0.4831 0.1094>, rot=<-0.0654 -0.0035 0.2242 -0.9723>}
 pinky_02_r {parent=pinky_01_r, loc=<-2.8274 -0.0893 0.1043>, rot=<0.0037 0.0011 0.2854 -0.9584>}
 pinky_03_r {parent=pinky_02_r, loc=<-2.2046 -0.0307 0.0078>, rot=<-0.0002 0.0098 0.1445 -0.9895>}
 deform_wrist_dn_base_r {parent=hand_r, loc=<-0.0004 2.2345 -0.0001>, rot=<0 0 0 -1>}
 deform_wrist_dn_r {parent=deform_wrist_dn_base_r, loc=<0 0 0>, rot=<0 0 0 1>}
 deform_wrist_up_base_r {parent=hand_r, loc=<0.0003 -2.2684 0.0001>, rot=<0 0 0 -1>}
 deform_wrist_up_r {parent=deform_wrist_up_base_r, loc=<0 0 0>, rot=<0 0 0 1>}
 weapon_r {parent=hand_r, loc=<0.1974 3.4111 -0.3811>, rot=<0 0 0 1>}
 hand_attach_r {parent=hand_r, loc=<-8.5001 2.9999 0>, rot=<0 0 -0.7071 -0.7071>}
 lowerarm_twist_02_r {parent=lowerarm_r, loc=<-9.6199 0 0>, rot=<0 0 0 -1>}
 lowerarm_twist_01_r {parent=lowerarm_r, loc=<-17.8549 0 0>, rot=<0 0 0 -1>}
 deform_elbow_in_base_r {parent=lowerarm_r, loc=<-0.2111 -5.2591 -0.8275>, rot=<0 0 0 -1>}
 deform_elbow_in_r {parent=deform_elbow_in_base_r, loc=<-0.0001 0 0.0002>, rot=<0 0 0 1>}
 deform_elbow_out_base_r {parent=lowerarm_r, loc=<0 4 -0.0001>, rot=<0 0 0 -1>}
 deform_elbow_out_r {parent=deform_elbow_out_base_r, loc=<0 0 0>, rot=<0 0 0 1>}
 upperarm_twist_01_r {parent=upperarm_r, loc=<-0.0041 0 0>, rot=<0 0 0 -1>}
 deform_upperarm_bk_base_r {parent=upperarm_twist_01_r, loc=<-1.0814 7.2192 1.9455>, rot=<0 0 0 1>}
 deform_upperarm_bk_r {parent=deform_upperarm_bk_base_r, loc=<0 0 0>, rot=<0 0 0 1>}
 deform_upperarm_fr_base_r {parent=upperarm_twist_01_r, loc=<0.3689 -7.0506 2.7652>, rot=<0 0 0 1>}
 deform_upperarm_fr_r {parent=deform_upperarm_fr_base_r, loc=<0 0 0>, rot=<0 0 0 1>}
 deform_upperarm_up_base_r {parent=upperarm_twist_01_r, loc=<3.2214 1.1391 -3.3308>, rot=<0 0 0 1>}
 deform_upperarm_up_r {parent=deform_upperarm_up_base_r, loc=<0 0 0>, rot=<0 0 0 -1>}
 upperarm_twist_02_r {parent=upperarm_r, loc=<-13.7941 0 0>, rot=<0 0 0 -1>}
 dyn_upperarm_sleeve_r_MSTR {parent=upperarm_r, loc=<0 0 0.0002>, rot=<0 0 0 -1>}
 dyn_upperarm_r_sleeve_twist_01_MSTR {parent=dyn_upperarm_sleeve_r_MSTR, loc=<-10.0368 0 0.1002>, rot=<0 0 0 -1>}
 dyn_upperarm_r_sleeve_twist_01 {parent=dyn_upperarm_r_sleeve_twist_01_MSTR, loc=<0 0 0>, rot=<0 0 0 -1>}
 dyn_upperarm_r_sleeve_twist_01_end {parent=dyn_upperarm_r_sleeve_twist_01, loc=<0.0004 0 5.9995>, rot=<0 0 0 -1>}
 dyn_upperarm_r_sleeve_twist_02_MSTR {parent=dyn_upperarm_sleeve_r_MSTR, loc=<-15.348 0 0.0995>, rot=<0 0 0 -1>}
 dyn_upperarm_r_sleeve_twist_02 {parent=dyn_upperarm_r_sleeve_twist_02_MSTR, loc=<0 0 0>, rot=<0 0 0 -1>}
 dyn_upperarm_r_sleeve_twist_02_end {parent=dyn_upperarm_r_sleeve_twist_02, loc=<-0.0003 0 7.0003>, rot=<0 0 0 -1>}
 dyn_upperarm_r_sleeve_twist_03_MSTR {parent=dyn_upperarm_sleeve_r_MSTR, loc=<-20.9243 0 0.1002>, rot=<0 0 0 -1>}
 dyn_upperarm_r_sleeve_twist_03 {parent=dyn_upperarm_r_sleeve_twist_03_MSTR, loc=<0 0 0>, rot=<0 0 0 -1>}
 dyn_upperarm_r_sleeve_twist_03_end {parent=dyn_upperarm_r_sleeve_twist_03, loc=<0.0004 0 7.9996>, rot=<0 0 0 -1>}
 deform_brooch {parent=clavicle_r, loc=<-12.811 -8.7106 -3.8223>, rot=<-0.5826 0.2566 0.7124 -0.2953>}
 deform_pec_aim_base_l {parent=spine_04, loc=<7.8732 12.6388 -2.4123>, rot=<0.0102 -0.4602 0.3979 -0.7936>}
 deform_pec_aim_l {parent=deform_pec_aim_base_l, loc=<0 0 0>, rot=<0 0 0 1>}
 deform_pec_base_l {parent=deform_pec_aim_l, loc=<10.766 0 0>, rot=<0 0 0 1>}
 deform_pec_l {parent=deform_pec_base_l, loc=<0.055 0 0>, rot=<0 0 0 1>}
 deform_pec_aim_base_r {parent=spine_04, loc=<7.8727 12.6388 2.4123>, rot=<-0.4602 -0.0102 0.7936 0.3979>}
 deform_pec_aim_r {parent=deform_pec_aim_base_r, loc=<0 0 0>, rot=<0 0 0 1>}
 deform_pec_base_r {parent=deform_pec_aim_r, loc=<-10.7662 -0.0006 -0.0007>, rot=<0 0 0 1>}
 deform_pec_r {parent=deform_pec_base_r, loc=<0 0 0>, rot=<0 0 0 1>}
 dyn_fold_front1_MSTR {parent=spine_04, loc=<7.4931 12.5839 12.9878>, rot=<-0.1537 -0.1614 0.9711 0.0854>}
 dyn_fold_front1_01 {parent=dyn_fold_front1_MSTR, loc=<0 0 0>, rot=<0 0 0 1>}
 dyn_fold_front1_02 {parent=dyn_fold_front1_01, loc=<7.723 0 0>, rot=<0 0 0 1>}
 dyn_fold_front1_03 {parent=dyn_fold_front1_02, loc=<6.029 0 0>, rot=<0 0 0 1>}
 dyn_fold_front2_MSTR {parent=spine_04, loc=<4.7076 10.8629 17.0542>, rot=<0.0762 -0.2261 0.9711 -0.0068>}
 dyn_fold_front2_01 {parent=dyn_fold_front2_MSTR, loc=<0 0 0>, rot=<0 0 0 1>}
 dyn_fold_front2_02 {parent=dyn_fold_front2_01, loc=<6.3042 0 0>, rot=<0 0 0 1>}
 dyn_fold_front2_03 {parent=dyn_fold_front2_02, loc=<6.5675 0 0>, rot=<0 0 0 1>}
 dyn_skirt_MSTR {parent=spine_01, loc=<0 0 0>, rot=<0 0 0 1>}
 dyn_skirt_front_mid_MSTR {parent=dyn_skirt_MSTR, loc=<-1.4158 14.6109 0>, rot=<0 0 1 0.0093>}
 dyn_skirt_front_mid_01 {parent=dyn_skirt_front_mid_MSTR, loc=<48.1235 0 0>, rot=<0 0 0 1>}
 dyn_skirt_front_mid_02 {parent=dyn_skirt_front_mid_01, loc=<0 0 0>, rot=<0 0 0 1>}
 dyn_skirt_front_mid_03 {parent=dyn_skirt_front_mid_02, loc=<10.424 0 0>, rot=<0 0 0 1>}
 dyn_skirt_front_mid_04 {parent=dyn_skirt_front_mid_03, loc=<6.9951 0 0>, rot=<0 0 0 1>}
 dyn_skirt_front_mid_05 {parent=dyn_skirt_front_mid_04, loc=<7.0087 0 0>, rot=<0 0 0 1>}
 dyn_skirt_front_mid_06 {parent=dyn_skirt_front_mid_05, loc=<6.9649 0 0>, rot=<0 0 0 1>}
 dyn_skirt_side_l_MSTR {parent=dyn_skirt_MSTR, loc=<-1.2346 -2.4178 -16.4314>, rot=<0.0575 -0.7145 -0.6964 -0.0335>}
 dyn_skirt_side_l_01 {parent=dyn_skirt_side_l_MSTR, loc=<45.0635 0 0>, rot=<0 0 0 1>}
 dyn_skirt_side_l_02 {parent=dyn_skirt_side_l_01, loc=<0 0 0>, rot=<0 0 0 1>}
 dyn_skirt_side_l_03 {parent=dyn_skirt_side_l_02, loc=<13.9465 -0.3318 -0.0357>, rot=<-0.0077 -0.0008 0.0076 -0.9999>}
 dyn_skirt_side_l_04 {parent=dyn_skirt_side_l_03, loc=<7.0201 0 0>, rot=<0 0 0 1>}
 dyn_skirt_side_l_05 {parent=dyn_skirt_side_l_04, loc=<7.0305 0 0>, rot=<0 0 0 1>}
 dyn_skirt_side_l_06 {parent=dyn_skirt_side_l_05, loc=<7.0107 0 0>, rot=<0 0 0 1>}
 dyn_skirt_back_l_MSTR {parent=dyn_skirt_MSTR, loc=<-0.7173 -11.3926 -10.69>, rot=<0.0556 -0.9978 -0.002 -0.0362>}
 dyn_skirt_back_l_base {parent=dyn_skirt_back_l_MSTR, loc=<27.2315 0 0>, rot=<0 0 0 1>}
 dyn_skirt_back_l_01 {parent=dyn_skirt_back_l_base, loc=<14.1228 0 0>, rot=<0 0 0 1>}
 dyn_skirt_back_l_02 {parent=dyn_skirt_back_l_01, loc=<0 0 0>, rot=<0 0 0 1>}
 dyn_skirt_back_l_03 {parent=dyn_skirt_back_l_02, loc=<17.3477 0 0>, rot=<0 0 0 1>}
 dyn_skirt_back_l_04 {parent=dyn_skirt_back_l_03, loc=<7.0051 0 0>, rot=<0 0 0 1>}
 dyn_skirt_back_l_05 {parent=dyn_skirt_back_l_04, loc=<7.0406 0 0>, rot=<0 0 0 1>}
 dyn_skirt_back_l_06 {parent=dyn_skirt_back_l_05, loc=<7.0025 0 0>, rot=<0 0 0 1>}
 dyn_skirt_back_mid_MSTR {parent=dyn_skirt_MSTR, loc=<-0.7364 -14.5993 0>, rot=<0.0649 -0.9979 0 0>}
 dyn_skirt_back_mid_base {parent=dyn_skirt_back_mid_MSTR, loc=<27.135 0 0>, rot=<0 0 0 1>}
 dyn_skirt_back_mid_01 {parent=dyn_skirt_back_mid_base, loc=<14.0044 0 0>, rot=<0 0 0 1>}
 dyn_skirt_back_mid_02 {parent=dyn_skirt_back_mid_01, loc=<0 0 0>, rot=<0 0 0 1>}
 dyn_skirt_back_mid_03 {parent=dyn_skirt_back_mid_02, loc=<17.4005 0 0>, rot=<0 0 0 1>}
 dyn_skirt_back_mid_04 {parent=dyn_skirt_back_mid_03, loc=<6.9456 0 0>, rot=<0 0 0 1>}
 dyn_skirt_back_mid_05 {parent=dyn_skirt_back_mid_04, loc=<7.001 0 0>, rot=<0 0 0 1>}
 dyn_skirt_back_mid_06 {parent=dyn_skirt_back_mid_05, loc=<6.9758 0 0>, rot=<0 0 0 1>}
 dyn_skirt_back_r_MSTR {parent=dyn_skirt_MSTR, loc=<-0.7172 -11.3925 10.69>, rot=<0.9978 0.0556 -0.0362 0.002>}
 dyn_skirt_back_r_base {parent=dyn_skirt_back_r_MSTR, loc=<-27.2315 0 0>, rot=<0 0 0 1>}
 dyn_skirt_back_r_01 {parent=dyn_skirt_back_r_base, loc=<-14.1227 0 0>, rot=<0 0 0 1>}
 dyn_skirt_back_r_02 {parent=dyn_skirt_back_r_01, loc=<0 0 0>, rot=<0 0 0 1>}
 dyn_skirt_back_r_03 {parent=dyn_skirt_back_r_02, loc=<-17.3478 0 0>, rot=<0 0 0 1>}
 dyn_skirt_back_r_04 {parent=dyn_skirt_back_r_03, loc=<-7.0051 0 0>, rot=<0 0 0 1>}
 dyn_skirt_back_r_05 {parent=dyn_skirt_back_r_04, loc=<-7.0406 0 0.0001>, rot=<0 0 0 1>}
 dyn_skirt_back_r_06 {parent=dyn_skirt_back_r_05, loc=<-7.0025 0 -0.0001>, rot=<0 0 0 1>}
 dyn_skirt_back_r_06_FxSocket {parent=dyn_skirt_back_r_06, type=Null, loc=<0 0 0>, rot=<0 -1 0 0>}
 dyn_skirt_side_r_MSTR {parent=dyn_skirt_MSTR, loc=<-1.2346 -2.4178 16.4314>, rot=<-0.7145 -0.0575 0.0335 -0.6964>}
 dyn_skirt_side_r_01 {parent=dyn_skirt_side_r_MSTR, loc=<-45.0635 0 0>, rot=<0 0 0 1>}
 dyn_skirt_side_r_02 {parent=dyn_skirt_side_r_01, loc=<0 0 0>, rot=<0 0 0 1>}
 dyn_skirt_side_r_03 {parent=dyn_skirt_side_r_02, loc=<-13.9465 0.3319 0.0357>, rot=<-0.0077 -0.0008 0.0076 -0.9999>}
 dyn_skirt_side_r_04 {parent=dyn_skirt_side_r_03, loc=<-7.0201 -0.0001 0>, rot=<0 0 0 1>}
 dyn_skirt_side_r_05 {parent=dyn_skirt_side_r_04, loc=<-7.0305 0.0001 0>, rot=<0 0 0 1>}
 dyn_skirt_side_r_06 {parent=dyn_skirt_side_r_05, loc=<-7.0108 0 0>, rot=<0 0 0 1>}
 dyn_skirt_side_r_06_FxSocket {parent=dyn_skirt_side_r_06, type=Null, loc=<0 0 0>, rot=<0 -1 0 0>}
 dyn_skirt_front_l_MSTR {parent=dyn_skirt_MSTR, loc=<-1.2526 10.3527 -12.013>, rot=<-0.0218 0.1163 0.993 0.0041>}
 dyn_skirt_front_l_01 {parent=dyn_skirt_front_l_MSTR, loc=<48.2489 0 0>, rot=<0 0 0 1>}
 dyn_skirt_front_l_02 {parent=dyn_skirt_front_l_01, loc=<0 0 0>, rot=<0 0 0 1>}
 dyn_skirt_front_l_03 {parent=dyn_skirt_front_l_02, loc=<10.3272 0 0>, rot=<0 0 0 1>}
 dyn_skirt_front_l_04 {parent=dyn_skirt_front_l_03, loc=<6.9811 0 0>, rot=<0 0 0 1>}
 dyn_skirt_front_l_05 {parent=dyn_skirt_front_l_04, loc=<7.0006 0 0>, rot=<0 0 0 1>}
 dyn_skirt_front_l_06 {parent=dyn_skirt_front_l_05, loc=<6.9768 0 0>, rot=<0 0 0 1>}
 dyn_skirt_front_r_MSTR {parent=dyn_skirt_MSTR, loc=<-1.2527 10.3527 12.013>, rot=<-0.1163 -0.0218 0.0041 -0.993>}
 dyn_skirt_front_r_01 {parent=dyn_skirt_front_r_MSTR, loc=<-48.2489 0 0>, rot=<0 0 0 1>}
 dyn_skirt_front_r_02 {parent=dyn_skirt_front_r_01, loc=<0 0 0>, rot=<0 0 0 1>}
 dyn_skirt_front_r_03 {parent=dyn_skirt_front_r_02, loc=<-10.3272 0.0001 0>, rot=<0 0 0 1>}
 dyn_skirt_front_r_04 {parent=dyn_skirt_front_r_03, loc=<-6.9811 -0.0001 0>, rot=<0 0 0 1>}
 dyn_skirt_front_r_05 {parent=dyn_skirt_front_r_04, loc=<-7.0006 0 0>, rot=<0 0 0 1>}
 dyn_skirt_front_r_06 {parent=dyn_skirt_front_r_05, loc=<-6.9768 0 0>, rot=<0 0 0 1>}
 dyn_skirt_front_r_06_FxSocket {parent=dyn_skirt_front_r_06, type=Null, loc=<0 0 0>, rot=<0 -1 0 0>}
 dyn_skirt_SIMSPACE {parent=spine_01, loc=<0 0 0>, rot=<0 0 0 1>}
 thigh_l {parent=pelvis, loc=<-2.5693 -0.0109 -7.7501>, rot=<0.0012 0.0572 0.0155 -0.9982>}
 calf_l {parent=thigh_l, loc=<-39.0484 0 0>, rot=<0.0006 -0.0161 -0.0391 -0.9991>}
 foot_l {parent=calf_l, loc=<-42.561 0.1077 0.4292>, rot=<0.0418 -0.0455 0.0343 -0.9975>}
 ball_l {parent=foot_l, loc=<-6.1599 13.8264 0.6806>, rot=<0.0001 -0.0001 -0.7063 -0.7079>}
 calf_twist_02_l {parent=calf_l, loc=<-18.8064 -0.0967 0.0397>, rot=<0.0237 0.0095 -0.0035 -0.9997>}
 calf_twist_01_l {parent=calf_l, loc=<-35.813 -0.0967 0.0397>, rot=<0.0237 0.0095 -0.0035 -0.9997>}
 deform_knee_in_base_l {parent=calf_l, loc=<1.298 -5.2957 -0.0496>, rot=<0 0 0 -1>}
 deform_knee_in_l {parent=deform_knee_in_base_l, loc=<0 0 0>, rot=<0 0 0 1>}
 deform_knee_out_base_l {parent=calf_l, loc=<0 4 0>, rot=<0 0 0 -1>}
 deform_knee_out_l {parent=deform_knee_out_base_l, loc=<0 0 0>, rot=<0 0 0 1>}
 thigh_twist_01_l {parent=thigh_l, loc=<-21.6303 0 0>, rot=<0.0013 0 0 -1>}
 deform_groin_base_l {parent=thigh_l, loc=<-2.1908 8.1142 -1.7442>, rot=<0 0 0 -1>}
 deform_groin_l {parent=deform_groin_base_l, loc=<0 0 0>, rot=<0 0 0 1>}
 thigh_r {parent=pelvis, loc=<-2.5693 -0.0109 7.7501>, rot=<0.0572 -0.0012 0.9982 0.0155>}
 calf_r {parent=thigh_r, loc=<39.0484 0 0>, rot=<0.0006 -0.0161 -0.0391 -0.9991>}
 foot_r {parent=calf_r, loc=<42.561 -0.1077 -0.4292>, rot=<0.0418 -0.0455 0.0343 -0.9975>}
 ball_r {parent=foot_r, loc=<6.1599 -13.8264 -0.6806>, rot=<0.0001 -0.0001 -0.7063 -0.7079>}
 calf_twist_02_r {parent=calf_r, loc=<18.8064 0.0967 -0.0397>, rot=<0.0237 0.0095 -0.0035 -0.9997>}
 calf_twist_01_r {parent=calf_r, loc=<35.813 0.0967 -0.0397>, rot=<0.0237 0.0095 -0.0035 -0.9997>}
 deform_knee_in_base_r {parent=calf_r, loc=<-1.298 5.2957 0.0496>, rot=<0 0 0 -1>}
 deform_knee_in_r {parent=deform_knee_in_base_r, loc=<0 0 0>, rot=<0 0 0 1>}
 deform_knee_out_base_r {parent=calf_r, loc=<0 -4 0>, rot=<0 0 0 -1>}
 deform_knee_out_r {parent=deform_knee_out_base_r, loc=<0 0 0>, rot=<0 0 0 1>}
 thigh_twist_01_r {parent=thigh_r, loc=<21.6303 0 0>, rot=<0.0013 0 0 -1>}
 deform_groin_base_r {parent=thigh_r, loc=<2.1908 -8.1142 1.7442>, rot=<0 0 0 -1>}
 deform_groin_r {parent=deform_groin_base_r, loc=<0 0 0>, rot=<0 0 0 1>}
 deform_thigh_notwist_l {parent=pelvis, loc=<-2.5785 -0.0168 -7.6736>, rot=<-0.0807 0.0001 0 -0.9967>}
 deform_thigh_notwist_r {parent=pelvis, loc=<-2.5386 -0.0122 7.6605>, rot=<0.0572 -0.0012 0.9982 0.0155>}
 deform_glute_base_l {parent=pelvis, loc=<-8.8064 -8.8222 -8.4287>, rot=<0.0275 0.0501 -0.4472 -0.8926>}
 deform_glute_l {parent=deform_glute_base_l, loc=<0 0 0>, rot=<0 0 0 1>}
 deform_glute_base_r {parent=pelvis, loc=<-8.8062 -8.8222 8.4287>, rot=<0.0501 -0.0275 0.8926 -0.4472>}
 deform_glute_r {parent=deform_glute_base_r, loc=<0 0 0>, rot=<0 0 0 1>}
 dyn_skirt_ground_collision {parent=root, loc=<0 0 -8.2344>, rot=<0 0 0 1>}
 ik_hand_root {parent=root, loc=<0 0 0>, rot=<0 0 0 1>}
 ik_hand_gun {parent=ik_hand_root, loc=<-51.9727 4.0883 102.6935>, rot=<0.6084 0.1793 0.4613 -0.6205>}
 ik_hand_r {parent=ik_hand_gun, loc=<0 0 0>, rot=<0 0 0 1>}
 ik_hand_l {parent=ik_hand_gun, loc=<53.0285 82.1781 35.2051>, rot=<-0.5102 -0.7906 -0.3387 0>}
 attach {parent=root, loc=<0 0 0>, rot=<0 0 0 1>}
 ik_foot_root {parent=root, loc=<0 0 0>, rot=<0 0 0 1>}
 ik_foot_l {parent=ik_foot_root, loc=<15.2608 -4.4176 9.4432>, rot=<0.0357 0.7045 0.0284 -0.7082>}
 ik_foot_r {parent=ik_foot_root, loc=<-15.2608 -4.4176 9.4432>, rot=<0.7082 0.0284 -0.7045 0.0357>}
 root_fp {parent=root, loc=<0 0 0>, rot=<0 0 0 1>}
 attach_fp {parent=root_fp, loc=<0 0 0>, rot=<0 0 0 1>}
 camera_root {parent=root_fp, loc=<0 0 0>, rot=<0 0 0 1>}
 camera {parent=camera_root, loc=<0 0 0>, rot=<0 0 0 1>}
 attach_emote_01 {parent=root, loc=<0 0 0>, rot=<0 0 0 1>}
 attach_emote_02 {parent=root, loc=<0 0 0>, rot=<0 0 0 1>}
 attach_emote_03 {parent=root, loc=<0 0 0>, rot=<0 0 0 1>}
 warp_point_01 {parent=root, loc=<0 0 0>, rot=<0 0 0 1>}
 warp_point_02 {parent=root, loc=<0 0 0>, rot=<0 0 0 1>}
 warp_point_03 {parent=root, loc=<0 0 0>, rot=<0 0 0 1>}
 CR Foot Near Pelvis L {type=Curve, loc=<0 0 0>, rot=<0 0 0 1>}
 CR Foot Near Pelvis R {type=Curve, loc=<0 0 0>, rot=<0 0 0 1>}
 CR Rear Knee Bend L {type=Curve, loc=<0 0 0>, rot=<0 0 0 1>}
 CR Rear Knee Bend R {type=Curve, loc=<0 0 0>, rot=<0 0 0 1>}
 CR Thigh L Across R {type=Curve, loc=<0 0 0>, rot=<0 0 0 1>}
 CR Thigh R Across L {type=Curve, loc=<0 0 0>, rot=<0 0 0 1>}
 CR Thigh Raised Front L {type=Curve, loc=<0 0 0>, rot=<0 0 0 1>}
 CR Thigh Raised Front R {type=Curve, loc=<0 0 0>, rot=<0 0 0 1>}
 MaterialCurveA {type=Curve, loc=<0 0 0>, rot=<0 0 0 1>}
]
variables [
 Is Baseball Sliding {type=bool, default=False}
 Is Knee Sliding {type=bool, default=False}
 back skirt l {
  type="TArray<FRigElementKey>"
  default="((Type=Bone,Name=\"dyn_skirt_back_l_MSTR\"),(Type=Bone,Name=\"dyn_skirt_back_l_base\"),(Type=Bone,Name=\"dyn_skirt_back_l_01\"),(Type=Bone,Name=\"dyn_skirt_back_l_02\"),(Type=Bone,Name=\"dyn_skirt_back_l_03\"),(Type=Bone,Name=\"dyn_skirt_back_l_04\"),(Type=Bone,Name=\" ...(truncated)"
 }
 back skirt mid {
  type="TArray<FRigElementKey>"
  default="((Type=Bone,Name=\"dyn_skirt_back_mid_MSTR\"),(Type=Bone,Name=\"dyn_skirt_back_mid_base\"),(Type=Bone,Name=\"dyn_skirt_back_mid_01\"),(Type=Bone,Name=\"dyn_skirt_back_mid_02\"),(Type=Bone,Name=\"dyn_skirt_back_mid_03\"),(Type=Bone,Name=\"dyn_skirt_back_mid_04\"),(Type ...(truncated)"
 }
 back skirt r {
  type="TArray<FRigElementKey>"
  default="((Type=Bone,Name=\"dyn_skirt_back_r_MSTR\"),(Type=Bone,Name=\"dyn_skirt_back_r_base\"),(Type=Bone,Name=\"dyn_skirt_back_r_01\"),(Type=Bone,Name=\"dyn_skirt_back_r_02\"),(Type=Bone,Name=\"dyn_skirt_back_r_03\"),(Type=Bone,Name=\"dyn_skirt_back_r_04\"),(Type=Bone,Name=\" ...(truncated)"
 }
 front skirt l {
  type="TArray<FRigElementKey>"
  default=((Type=Bone,Name="dyn_skirt_front_l_01"),(Type=Bone,Name="dyn_skirt_front_l_02"),(Type=Bone,Name="dyn_skirt_front_l_03"),(Type=Bone,Name="dyn_skirt_front_l_04"),(Type=Bone,Name="dyn_skirt_front_l_05"),(Type=Bone,Name="dyn_skirt_front_l_06"))
 }
 front skirt mid {
  type="TArray<FRigElementKey>"
  default="((Type=Bone,Name=\"dyn_skirt_front_mid_MSTR\"),(Type=Bone,Name=\"dyn_skirt_front_mid_01\"),(Type=Bone,Name=\"dyn_skirt_front_mid_02\"),(Type=Bone,Name=\"dyn_skirt_front_mid_03\"),(Type=Bone,Name=\"dyn_skirt_front_mid_04\"),(Type=Bone,Name=\"dyn_skirt_front_mid_05\"),( ...(truncated)"
 }
 front skirt r {
  type="TArray<FRigElementKey>"
  default=((Type=Bone,Name="dyn_skirt_front_r_01"),(Type=Bone,Name="dyn_skirt_front_r_02"),(Type=Bone,Name="dyn_skirt_front_r_03"),(Type=Bone,Name="dyn_skirt_front_r_04"),(Type=Bone,Name="dyn_skirt_front_r_05"),(Type=Bone,Name="dyn_skirt_front_r_06"))
 }
 prev thigh l quaternion {type=FQuat, default=(X=0,Y=0,Z=0,W=1)}
 prev thigh r quaternion {type=FQuat, default=(X=0,Y=0,Z=0,W=1)}
]
