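# Physics asset for the left sleeve of the M_MED_FlickLedge male medium player cosmetic body, defining simulation shapes for dynamic upper-arm sleeve twist bones.
# Two parallel chains, twist_02 and twist_03, each link a small MSTR capsule to a mid capsule and an end capsule.
# Angular limits are tight at each MSTR joint and looser at the end joints, with twist_02's end allowing wider swing than twist_03's.
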PhysicsAsset /BRCosmetics/Characters/Player/Male/Medium/Bodies/M_MED_FlickLedge/Meshes/M_MED_FlickLedge_SleeveL_Physics
bodies {
 dyn_upperarm_l_sleeve_twist_02 [{shapeType=Capsule, length=4, radius=4, shapeName=dyn_lowerarm_l_sleeve_twist_02_capsule, center=<0 0 -4>, rotation=<rot 90 0 0>}]
 dyn_upperarm_l_sleeve_twist_02_MSTR [{shapeType=Capsule, length=0.505, radius=0.505, shapeName=dyn_upperarm_l_sleeve_twist_02_MSTR_capsule, rotation=<rot 0 -90 0>}]
 dyn_upperarm_l_sleeve_twist_02_end [{shapeType=Capsule, length=4, radius=4, shapeName=dyn_lowerarm_l_sleeve_twist_02_end_capsule, center=<0 0 -4>, rotation=<rot 90 0 0>}]
 dyn_upperarm_l_sleeve_twist_03 [{shapeType=Capsule, length=4, radius=4, shapeName=dyn_lowerarm_l_sleeve_twist_03_capsule, center=<0 0 -4>, rotation=<rot 90 0 0>}]
 dyn_upperarm_l_sleeve_twist_03_MSTR [{shapeType=Capsule, length=0.505, radius=0.505, shapeName=dyn_upperarm_l_sleeve_twist_03_MSTR_capsule, rotation=<rot 0 -90 0>}]
 dyn_upperarm_l_sleeve_twist_03_end [{shapeType=Capsule, length=4, radius=4, shapeName=dyn_lowerarm_l_sleeve_twist_03_end_capsule, center=<0 0 -4>, rotation=<rot 90 0 0>}]
}
constraints [
 {
  bone1Name=dyn_upperarm_l_sleeve_twist_02
  bone2Name=dyn_upperarm_l_sleeve_twist_02_MSTR
  swing1LimitDegrees=2
  swing2LimitDegrees=2
  twistLimitDegrees=20
 }
 {
  bone1Name=dyn_upperarm_l_sleeve_twist_02_end
  bone2Name=dyn_upperarm_l_sleeve_twist_02
  swing1LimitDegrees=12
  swing2LimitDegrees=32
  twistLimitDegrees=50
 }
 {
  bone1Name=dyn_upperarm_l_sleeve_twist_03
  bone2Name=dyn_upperarm_l_sleeve_twist_03_MSTR
  swing1LimitDegrees=2
  swing2LimitDegrees=2
  twistLimitDegrees=20
 }
 {
  bone1Name=dyn_upperarm_l_sleeve_twist_03_end
  bone2Name=dyn_upperarm_l_sleeve_twist_03
  swing1LimitDegrees=12
  swing2LimitDegrees=16
  twistLimitDegrees=50
 }
]
