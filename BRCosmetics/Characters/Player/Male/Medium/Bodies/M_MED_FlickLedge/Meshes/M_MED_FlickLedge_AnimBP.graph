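# Anim Blueprint for the FlickLedge male body cosmetic, adding secondary motion to skirt, sleeves and cloth folds.
# Copied parent pose runs through a Control Rig, RigidBody sims and two front-fold AnimDynamics chains; other fold chains are unwired.
# A post-sim rig follows either calf-constrained skirt bones for the custom emote or speed-driven back-skirt lift outside skydiving.
# EventGraph updates gravity, physics settings, floor-collision and emote flags.
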AnimBlueprint /BRCosmetics/Characters/Player/Male/Medium/Bodies/M_MED_FlickLedge/Meshes/M_MED_FlickLedge_AnimBP
target_skeleton=&/M_MED_FlickLedge_Skeleton
graphs [
 AnimGraph {nodes [
   n1 = "Anim Dynamics - Chain: dyn_fold_front3_01 -> dyn_fold_front3_03"!(ComponentPose, GravityOverride:Vector="0,0,-980", ExternalForce="0,0,0") -> (Pose) {
    alpha=0.9
    angularDampingOverride=0.9999
    angularSpringConstant=34
    bChain=true
    bOverrideAngularDamping=true
    bUseGravityOverride=true
    bUsePlanarLimit=false
    linearSpringConstant=100
    simulationSpace=RootRelative
    boundBone {boneName=dyn_fold_front3_01}
    chainEnd {boneName=dyn_fold_front3_03}
    componentAppliedLinearAccClamp=<500 500 500>
    gravityOverride {z=-980}
    physicsBodyDefinitions [
     {
      boundBone {boneName=dyn_fold_front3_01}
      boxExtents=<5 30 30>
      constraintSetup {angularLimitsMax=<0 2 2>, angularLimitsMin=<0 -2 -2>, angularTarget=<0 0 0>}
      localJointOffset=<1 0 0>
     }
     {
      boundBone {boneName=dyn_fold_front3_02}
      boxExtents=<5 30 30>
      constraintSetup {angularLimitsMax=<0 5 5>, angularLimitsMin=<0 -5 -5>, angularTarget=<0 0 0>}
      localJointOffset=<1 0 0>
     }
     {
      boundBone {boneName=dyn_fold_front3_03}
      boxExtents=<5 30 30>
      constraintSetup {angularLimitsMax=<0 5 5>, angularLimitsMin=<0 -5 -5>, angularTarget=<0 0 0>}
      localJointOffset=<1 0 0>
     }
    ]
    planarLimits [{drivingBone {boneName=spine_04}, planeTransform {location=<0 12 0>, rotation=<rot 0 -5 70>}}]
    relativeSpaceBone {boneName=spine_04}
    simSpaceSettings {maxAngularAcceleration=40, maxAngularVelocity=8, simSpaceAngularAlpha=0.75}
   }
   n2 = "Anim Dynamics - Chain: dyn_fold_front4_01 -> dyn_fold_front4_03"!(ComponentPose=@n1, GravityOverride:Vector="0,0,-980", ExternalForce="0,0,0") -> (Pose) {
    alpha=0.9
    angularDampingOverride=0.9999
    angularSpringConstant=34
    bChain=true
    bOverrideAngularDamping=true
    bUseGravityOverride=true
    bUsePlanarLimit=false
    linearSpringConstant=100
    simulationSpace=RootRelative
    boundBone {boneName=dyn_fold_front4_01}
    chainEnd {boneName=dyn_fold_front4_03}
    componentAppliedLinearAccClamp=<500 500 500>
    gravityOverride {z=-980}
    physicsBodyDefinitions [
     {
      boundBone {boneName=dyn_fold_front4_01}
      boxExtents=<5 30 30>
      constraintSetup {angularLimitsMax=<0 2 2>, angularLimitsMin=<0 -2 -2>, angularTarget=<0 0 0>}
      localJointOffset=<1 0 0>
     }
     {
      boundBone {boneName=dyn_fold_front4_02}
      boxExtents=<5 30 30>
      constraintSetup {angularLimitsMax=<0 5 5>, angularLimitsMin=<0 -5 -5>, angularTarget=<0 0 0>}
      localJointOffset=<1 0 0>
     }
     {
      boundBone {boneName=dyn_fold_front4_03}
      boxExtents=<5 30 30>
      constraintSetup {angularLimitsMax=<0 5 5>, angularLimitsMin=<0 -5 -5>, angularTarget=<0 0 0>}
      localJointOffset=<1 0 0>
     }
    ]
    planarLimits [{drivingBone {boneName=thigh_r}, planeTransform {location=<0 -13 0>, rotation=<rot 0 5 -55>}}]
    relativeSpaceBone {boneName=spine_04}
    simSpaceSettings {maxAngularAcceleration=40, maxAngularVelocity=8, simSpaceAngularAlpha=0.75}
   }
   n3 = "Anim Dynamics - Chain: dyn_fold_front5_01 -> dyn_fold_front5_03"!(ComponentPose=@n2, GravityOverride:Vector="0,0,-980", ExternalForce="0,0,0") -> (Pose) {
    alpha=0.9
    angularDampingOverride=0.9999
    angularSpringConstant=34
    bChain=true
    bOverrideAngularDamping=true
    bUseGravityOverride=true
    bUsePlanarLimit=false
    linearSpringConstant=100
    simulationSpace=RootRelative
    boundBone {boneName=dyn_fold_front5_01}
    chainEnd {boneName=dyn_fold_front5_03}
    componentAppliedLinearAccClamp=<500 500 500>
    gravityOverride {z=-980}
    physicsBodyDefinitions [
     {
      boundBone {boneName=dyn_fold_front5_01}
      boxExtents=<5 30 30>
      constraintSetup {angularLimitsMax=<0 2 2>, angularLimitsMin=<0 -2 -2>, angularTarget=<0 0 0>}
      localJointOffset=<1 0 0>
     }
     {
      boundBone {boneName=dyn_fold_front5_02}
      boxExtents=<5 30 30>
      constraintSetup {angularLimitsMax=<0 5 5>, angularLimitsMin=<0 -5 -5>, angularTarget=<0 0 0>}
      localJointOffset=<1 0 0>
     }
     {
      boundBone {boneName=dyn_fold_front5_03}
      boxExtents=<5 30 30>
      constraintSetup {angularLimitsMax=<0 5 5>, angularLimitsMin=<0 -5 -5>, angularTarget=<0 0 0>}
      localJointOffset=<1 0 0>
     }
    ]
    planarLimits [{drivingBone {boneName=thigh_r}, planeTransform {location=<0 -12 0>, rotation=<rot 0 0 -65>}}]
    relativeSpaceBone {boneName=spine_04}
    simSpaceSettings {maxAngularAcceleration=40, maxAngularVelocity=8, simSpaceAngularAlpha=0.75}
   }
   n4 = "Anim Dynamics - Chain: dyn_fold_back1_01 -> dyn_fold_back1_03"!(ComponentPose=@n3, GravityOverride:Vector="0,0,-980", ExternalForce="0,0,0") -> (Pose) {
    alpha=0.9
    angularDampingOverride=0.9999
    angularSpringConstant=250
    bChain=true
    bOverrideAngularDamping=true
    bUseGravityOverride=true
    bUsePlanarLimit=false
    linearSpringConstant=100
    simulationSpace=RootRelative
    boundBone {boneName=dyn_fold_back1_01}
    chainEnd {boneName=dyn_fold_back1_03}
    componentAppliedLinearAccClamp=<50 50 50>
    gravityOverride {z=-980}
    physicsBodyDefinitions [
     {
      boundBone {boneName=dyn_fold_back1_01}
      boxExtents=<1 30 30>
      constraintSetup {angularLimitsMax=<2 2 2>, angularLimitsMin=<-2 -2 -2>, angularTarget=<0 0 0>}
      localJointOffset=<1 0 0>
     }
     {
      boundBone {boneName=dyn_fold_back1_02}
      boxExtents=<1 30 30>
      constraintSetup {angularLimitsMax=<5 5 5>, angularLimitsMin=<-5 -5 -5>, angularTarget=<0 0 0>}
      localJointOffset=<1 0 0>
     }
     {
      boundBone {boneName=dyn_fold_back1_03}
      boxExtents=<1 30 30>
      constraintSetup {angularLimitsMax=<10 10 10>, angularLimitsMin=<-10 -10 -10>, angularTarget=<0 0 0>}
      localJointOffset=<1 0 0>
     }
    ]
    planarLimits [{drivingBone {boneName=dyn_skirt_back_r_MSTR}, planeTransform {location=<0 2 0>, rotation=<rot 0 0 140>}}]
    relativeSpaceBone {boneName=spine_04}
    simSpaceSettings {maxAngularAcceleration=40, maxAngularVelocity=8, simSpaceAngularAlpha=1}
   }
   "Anim Dynamics - Chain: dyn_fold_back2_01 -> dyn_fold_back2_03"!(ComponentPose=@n4, GravityOverride:Vector="0,0,-980", ExternalForce="0,0,0") -> (Pose) {
    alpha=0.9
    angularDampingOverride=0.9999
    angularSpringConstant=250
    bChain=true
    bOverrideAngularDamping=true
    bUseGravityOverride=true
    bUsePlanarLimit=false
    linearSpringConstant=100
    simulationSpace=RootRelative
    boundBone {boneName=dyn_fold_back2_01}
    chainEnd {boneName=dyn_fold_back2_03}
    componentAppliedLinearAccClamp=<50 50 50>
    gravityOverride {z=-980}
    physicsBodyDefinitions [
     {
      boundBone {boneName=dyn_fold_back2_01}
      boxExtents=<1 30 30>
      constraintSetup {angularLimitsMax=<2 2 2>, angularLimitsMin=<-2 -2 -2>, angularTarget=<0 0 0>}
      localJointOffset=<1 0 0>
     }
     {
      boundBone {boneName=dyn_fold_back2_02}
      boxExtents=<1 30 30>
      constraintSetup {angularLimitsMax=<5 10 10>, angularLimitsMin=<-5 -10 -10>, angularTarget=<0 0 0>}
      localJointOffset=<1 0 0>
     }
     {
      boundBone {boneName=dyn_fold_back2_03}
      boxExtents=<1 30 30>
      constraintSetup {angularLimitsMax=<10 10 10>, angularLimitsMin=<-10 -10 -10>, angularTarget=<0 0 0>}
      localJointOffset=<1 0 0>
     }
    ]
    planarLimits [{drivingBone {boneName=dyn_skirt_back_r_MSTR}, planeTransform {location=<0 2 0>, rotation=<rot 0 0 120>}}]
    relativeSpaceBone {boneName=spine_04}
    simSpaceSettings {maxAngularAcceleration=40, maxAngularVelocity=8, simSpaceAngularAlpha=1}
   }
   Local To Component()
   n5 = Get PawnSpeed2D()
   n6 = "Transform (Modify) Bone - Bone: dyn_skirt_back_l_MSTR"(
    ComponentPose=Local To Component(LocalPose=Use cached pose 'Post Dynamics'())
    Alpha=@n5
    Translation="0,0,0"
    Rotation="0,-20,0"
    Scale="1,1,1"
   ) {
    rotationMode=BMM_Additive
    rotationSpace=BCS_BoneSpace
    alphaScaleBiasClamp {bInterpResult=true, bMapRange=true, inRange {max=900, min=440}}
    boneToModify {boneName=dyn_skirt_back_l_MSTR}
    rotation {yaw=-20}
   }
   n7 = "Transform (Modify) Bone - Bone: dyn_skirt_back_mid_MSTR"(ComponentPose=@n6, Alpha=@n5, Translation="0,0,0", Rotation="0,-20,0", Scale="1,1,1") {
    rotationMode=BMM_Additive
    rotationSpace=BCS_BoneSpace
    alphaScaleBiasClamp {bInterpResult=true, bMapRange=true, inRange {max=900, min=440}}
    boneToModify {boneName=dyn_skirt_back_mid_MSTR}
    rotation {yaw=-20}
   }
   n8 = "Transform (Modify) Bone - Bone: dyn_skirt_back_r_MSTR"(ComponentPose=@n7, Alpha=@n5, Translation="0,0,0", Rotation="0,-20,0", Scale="1,1,1") {
    rotationMode=BMM_Additive
    rotationSpace=BCS_BoneSpace
    alphaScaleBiasClamp {bInterpResult=true, bMapRange=true, inRange {max=900, min=440}}
    boneToModify {boneName=dyn_skirt_back_r_MSTR}
    rotation {yaw=-20}
   }
   n9 = Component To Local(ComponentPose=@n8)
   n10 = Use cached pose 'Post Dynamics'()
   n11 = Get bIsSkyDiving()
   n12 = "Constraint - dyn_skirt_front_l_02 (P)"(ComponentPose=Local To Component(LocalPose=Use cached pose 'Post Dynamics'()), Alpha=1, ConstraintWeights_0=1) {boneToModify {boneName=dyn_skirt_front_l_02}, constraintSetup [{targetBone {boneName=calf_l}}], constraintWeights [1]}
   n13 = "Constraint - dyn_skirt_front_r_02 (P)"(ComponentPose=@n12, Alpha=1, ConstraintWeights_0=1) {boneToModify {boneName=dyn_skirt_front_r_02}, constraintSetup [{targetBone {boneName=calf_r}}], constraintWeights [1]}
   n14 = "Constraint - dyn_skirt_front_l_03 (P)"(ComponentPose=@n13, Alpha=1, ConstraintWeights_0=1) {boneToModify {boneName=dyn_skirt_front_l_03}, constraintSetup [{targetBone {boneName=calf_l}}], constraintWeights [1]}
   n15 = "Constraint - dyn_skirt_front_r_03 (P)"(ComponentPose=@n14, Alpha=1, ConstraintWeights_0=1) {boneToModify {boneName=dyn_skirt_front_r_03}, constraintSetup [{targetBone {boneName=calf_r}}], constraintWeights [1]}
   n16 = "Constraint - dyn_skirt_front_l_04 (P)"(ComponentPose=@n15, Alpha=1, ConstraintWeights_0=1) {boneToModify {boneName=dyn_skirt_front_l_04}, constraintSetup [{targetBone {boneName=calf_l}}], constraintWeights [1]}
   n17 = "Constraint - dyn_skirt_front_r_04 (P)"(ComponentPose=@n16, Alpha=1, ConstraintWeights_0=1) {boneToModify {boneName=dyn_skirt_front_r_04}, constraintSetup [{targetBone {boneName=calf_r}}], constraintWeights [1]}
   n18 = Control Rig(Source=Blend Poses by bool(
     BlendPose_0=Component To Local(ComponentPose=@n17)
     BlendPose_1=Blend Poses by bool(BlendPose_0=@n10, BlendPose_1=@n9, bActiveValue=@n11)
     bActiveValue=Get CustomEmoteIsPlaying() -> (CustomEmoteIsPlaying:bool)
    )) {controlRigAssetReference {blueprintRigClass=&/M_MED_FlickLedge_PostSimCtrlRig.M_MED_FlickLedge_PostSimCtrlRig_C}}
   Output Pose(Result=@n18) {layerGroup=None}
   n19 = Control Rig!(
    Source=Copy Pose From Mesh() {bCopyCurves=true, bUseAttachedParent=true}
    Alpha
    Is Baseball Sliding:bool=Get bIsBaseballSliding()
    Is Knee Sliding:bool=Get bIsKneeSliding()
    isTacSprint:bool=Get bIsTacticalSprinting()
    Crouching:bool=Get bIsCrouching()
   ) -> (Pose) {controlRigAssetReference {blueprintRigClass=&/M_MED_FlickLedge_CtrlRig.M_MED_FlickLedge_CtrlRig_C}}
   n20 = "Transform (Modify) Bone - Bone: dyn_skirt_ground_collision"!(ComponentPose=Local To Component(LocalPose=@n19), bAlphaBoolEnabled:bool=True, Translation="0,0,-500", Rotation="0,0,0", Scale="1,1,1") -> (Pose) {
    alphaInputType=Bool
    translationMode=BMM_Additive
    translationSpace=BCS_BoneSpace
    alphaBoolBlend {blendInTime=0.2, blendOutTime=0.2}
    boneToModify {boneName=dyn_skirt_ground_collision}
    translation {z=-500}
   }
   n21 = RigidBody!(
    ComponentPose=@n20
    Alpha
    OverrideWorldGravity:Vector="0,0,-980"
    ExternalForce
    SimSpaceSettings:Sim Space Settings Structure=(WorldAlpha=1,VelocityScaleZ=1,DampingAlpha=0,MaxLinearVelocity=800,MaxAngularVelocity=12,MaxLinearAcceleration=4000,MaxAngularAcceleration=60,ExternalLinearDragV=(X=0.1,Y=0.1,Z=0.1),ExternalLinearVelocity=(X=0,Y=0,Z=0),ExternalAngularVelocity=(X=0,Y=0,Z=0))
   ) -> (Pose) {
    bClampLinearTranslationLimitToRefPose=true
    bOverrideWorldGravity=true
    overridePhysicsAsset=&/M_MED_FlickLedge_Physics
    simulationSpace=BaseBoneSpace
    baseBoneRef {boneName=dyn_skirt_SIMSPACE}
    overrideWorldGravity {z=-980}
    simSpaceSettings {
     dampingAlpha=0
     maxAngularAcceleration=60
     maxAngularVelocity=12
     maxLinearAcceleration=4000
     maxLinearVelocity=800
     worldAlpha=1
     externalLinearDragV=<0.1 0.1 0.1>
    }
   }
   n22 = RigidBody(
    ComponentPose=@n21
    OverrideWorldGravity="0,0,-980"
    SimSpaceSettings=(WorldAlpha=1,VelocityScaleZ=1,DampingAlpha=0,MaxLinearVelocity=800,MaxAngularVelocity=12,MaxLinearAcceleration=4000,MaxAngularAcceleration=60,ExternalLinearDragV=(X=0.1,Y=0.1,Z=0.1),ExternalLinearVelocity=(X=0,Y=0,Z=0),ExternalAngularVelocity=(X=0,Y=0,Z=0))
   ) {
    bClampLinearTranslationLimitToRefPose=true
    bOverrideWorldGravity=true
    overridePhysicsAsset=&/M_MED_FlickLedge_SleeveL_Physics
    simulationSpace=BaseBoneSpace
    baseBoneRef {boneName=upperarm_l}
    overrideWorldGravity {z=-980}
    simSpaceSettings {
     dampingAlpha=0
     maxAngularAcceleration=60
     maxAngularVelocity=12
     maxLinearAcceleration=4000
     maxLinearVelocity=800
     worldAlpha=1
     externalLinearDragV=<0.1 0.1 0.1>
    }
   }
   n23 = RigidBody(
    ComponentPose=@n22
    OverrideWorldGravity="0,0,-980"
    SimSpaceSettings=(WorldAlpha=1,VelocityScaleZ=1,DampingAlpha=0,MaxLinearVelocity=800,MaxAngularVelocity=12,MaxLinearAcceleration=4000,MaxAngularAcceleration=60,ExternalLinearDragV=(X=0.1,Y=0.1,Z=0.1),ExternalLinearVelocity=(X=0,Y=0,Z=0),ExternalAngularVelocity=(X=0,Y=0,Z=0))
   ) {
    bClampLinearTranslationLimitToRefPose=true
    bOverrideWorldGravity=true
    overridePhysicsAsset=&/M_MED_FlickLedge_SleeveR_Physics
    simulationSpace=BaseBoneSpace
    baseBoneRef {boneName=upperarm_r}
    overrideWorldGravity {z=-980}
    simSpaceSettings {
     dampingAlpha=0
     maxAngularAcceleration=60
     maxAngularVelocity=12
     maxLinearAcceleration=4000
     maxLinearVelocity=800
     worldAlpha=1
     externalLinearDragV=<0.1 0.1 0.1>
    }
   }
   n24 = "Anim Dynamics - Chain: dyn_fold_front1_01 -> dyn_fold_front1_03"!(ComponentPose=@n23, GravityOverride:Vector="0,0,-980", ExternalForce="0,0,0") -> (Pose) {
    alpha=0.9
    angularDampingOverride=0.9999
    angularSpringConstant=34
    bChain=true
    bOverrideAngularDamping=true
    bUseGravityOverride=true
    bUsePlanarLimit=false
    linearSpringConstant=100
    simulationSpace=RootRelative
    boundBone {boneName=dyn_fold_front1_01}
    chainEnd {boneName=dyn_fold_front1_03}
    componentAppliedLinearAccClamp=<500 500 500>
    gravityOverride {z=-980}
    physicsBodyDefinitions [
     {
      boundBone {boneName=dyn_fold_front1_01}
      boxExtents=<5 30 30>
      constraintSetup {angularLimitsMax=<0 1 1>, angularLimitsMin=<0 -1 -1>, angularTarget=<0 0 0>}
      localJointOffset=<1 0 0>
     }
     {
      boundBone {boneName=dyn_fold_front1_02}
      boxExtents=<5 30 30>
      constraintSetup {angularLimitsMax=<0 5 5>, angularLimitsMin=<0 -5 -5>, angularTarget=<1 0 0>}
      localJointOffset=<1 0 0>
     }
     {
      boundBone {boneName=dyn_fold_front1_03}
      boxExtents=<5 30 30>
      constraintSetup {angularLimitsMax=<0 5 5>, angularLimitsMin=<0 -5 -5>, angularTarget=<1 0 0>}
      localJointOffset=<1 0 0>
     }
    ]
    planarLimits [{drivingBone {boneName=spine_04}, planeTransform {location=<0 15 0>, rotation=<rot 0 -5 65>}}]
    relativeSpaceBone {boneName=spine_04}
    simSpaceSettings {maxAngularAcceleration=40, maxAngularVelocity=8, simSpaceAngularAlpha=0.75}
   }
   n25 = "Anim Dynamics - Chain: dyn_fold_front2_01 -> dyn_fold_front2_03"!(ComponentPose=@n24, GravityOverride:Vector="0,0,-980", ExternalForce="0,0,0") -> (Pose) {
    alpha=0.9
    angularDampingOverride=0.9999
    angularSpringConstant=34
    bChain=true
    bOverrideAngularDamping=true
    bUseGravityOverride=true
    bUsePlanarLimit=false
    linearSpringConstant=100
    simulationSpace=RootRelative
    boundBone {boneName=dyn_fold_front2_01}
    chainEnd {boneName=dyn_fold_front2_03}
    componentAppliedLinearAccClamp=<500 500 500>
    gravityOverride {z=-980}
    physicsBodyDefinitions [
     {
      boundBone {boneName=dyn_fold_front2_01}
      boxExtents=<5 30 30>
      constraintSetup {angularLimitsMax=<0 2 2>, angularLimitsMin=<0 -2 -2>, angularTarget=<0 0 0>}
      localJointOffset=<1 0 0>
     }
     {
      boundBone {boneName=dyn_fold_front2_02}
      boxExtents=<5 30 30>
      constraintSetup {angularLimitsMax=<0 5 5>, angularLimitsMin=<0 -5 -5>, angularTarget=<0 0 0>}
      localJointOffset=<1 0 0>
     }
     {
      boundBone {boneName=dyn_fold_front2_03}
      boxExtents=<5 30 30>
      constraintSetup {angularLimitsMax=<0 5 5>, angularLimitsMin=<0 -5 -5>, angularTarget=<0 0 0>}
      localJointOffset=<1 0 0>
     }
    ]
    planarLimits [{drivingBone {boneName=spine_04}, planeTransform {location=<0 15 0>, rotation=<rot 0 -5 60>}}]
    relativeSpaceBone {boneName=spine_04}
    simSpaceSettings {maxAngularAcceleration=40, maxAngularVelocity=8, simSpaceAngularAlpha=0.75}
   }
   Save cached pose 'Post Dynamics'(Pose=Component To Local(ComponentPose=@n25)) {cacheName=Post Dynamics}
  ]}
 EventGraph {nodes [
   Event BlueprintUpdateAnimation()
   > n1 = Sequence!(execute) -> (then_0, then_1, then_2:Exec)
   n2 = Branch(execute=@n1.then_1, Condition=Get bIsCrouching())
   > Set MoveFloorCollision(MoveFloorCollision:bool=false) -> (Output_Get:bool)
   Set MoveFloorCollision(execute=@n2.else, MoveFloorCollision=true)
   Set CustomEmoteIsPlaying(
    execute=@n1.then_2
    CustomEmoteIsPlaying:bool=Montage_IsPlaying(
     self=GetOwnerAnimInstance!(self:Custom Character Part Anim Instance Object Reference) -> (ReturnValue:Anim Instance Object Reference)
     Montage=/BRCosmetics/Animation/Game/MainPlayer/Emotes/FlickLedge/CMM/Emote_FlickLedge_CMM_M
    )
   ) -> (Output_Get:bool)
   n3 = "Gravity Override State Function (Parameter Asset) (Nativized)"(
    execute=@n1.then_0
    CharacterPartAnimInstance=Self-Reference()
    GameTimeInSeconds=GetGameTimeInSeconds()
    Parameters=&/M_MED_FlickLedge_Parameters
   )
   > Set Gravity Override(Gravity Override=@n3)
   > Set RBAN Settings Out(RBAN Settings Out=@n3.OutRigidBodyAnimNodeParameters)
   > n4 = Sequence!(execute) -> (then_0, then_1)
   > Branch(Condition=Get bIsSliding())
   > Set RBAN Settings Out(RBAN Settings Out=Get RBAN Settings Sliding() -> (RBAN Settings Sliding:Fort Rigid Body Anim Node Parameters Structure))
   Branch(execute=@n4.then_1, Condition=AND Boolean(A=Get bIsInFrontEnd(), B=Get bIsPlayingEmote()))
   > Set RBAN Settings Out()
   > Set Gravity Override(Gravity Override="0,0,-980")
  ]}
]
variables [
 CustomEmoteIsPlaying {type=boolean}
 Gravity Override {type=Vector}
 Gravity Override_AnimDynamics {type=Vector}
 MoveFloorCollision {type=boolean}
 MovingForward {type=boolean}
 RBAN Settings Out {type=FortRigidBodyAnimNodeParameters}
 RBAN Settings Sliding {
  type=FortRigidBodyAnimNodeParameters
  default {simSpaceSettings {
    dampingAlpha=0
    maxAngularAcceleration=10
    maxAngularVelocity=4
    maxLinearAcceleration=1000
    maxLinearVelocity=200
    velocityScaleZ=0.4
    worldAlpha=0.4
   }}
 }
]
