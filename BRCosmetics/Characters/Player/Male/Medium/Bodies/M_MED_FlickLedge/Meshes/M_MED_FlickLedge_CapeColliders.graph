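# Physics asset of cape colliders for the M_MED_FlickLedge male character body, with collision shapes only and no constraints.
# Capsules cover the pelvis, spine, clavicles and neck, with a sphere on the head, while arm, leg and foot bodies have no shapes.
# Large capsules on the left, middle and right back skirt bones (MSTR, 01, 04) add rear skirt collision, with left and right sides mirrored.
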PhysicsAsset /BRCosmetics/Characters/Player/Male/Medium/Bodies/M_MED_FlickLedge/Meshes/M_MED_FlickLedge_CapeColliders
bodies {
 calf_l []
 calf_r []
 clavicle_l [{shapeType=Capsule, length=9.0192, radius=9.8871, center=<9.4266 -1.1806 -2.4977>, rotation=<rot 69.3297 1.4476 176.9462>}]
 clavicle_r [{shapeType=Capsule, length=9.0192, radius=9.8871, center=<-9.4266 1.1806 2.4977>, rotation=<rot 69.3297 1.4476 -3.0538>}]
 dyn_skirt_back_l_01 [{shapeType=Capsule, length=17.0762, radius=18.4253, shapeName=dyn_skirt_back_l_01_capsule, center=<12.6836 8.5229 0>}]
 dyn_skirt_back_l_04 [{shapeType=Capsule, length=18.7153, radius=21.5892, shapeName=dyn_skirt_back_l_04_capsule, center=<11.1948 5.6343 0>}]
 dyn_skirt_back_l_MSTR [{
   shapeType=Capsule
   length=36.3844
   radius=17.1856
   shapeName=dyn_skirt_back_l_MSTR_capsule
   center=<30.8498 6.7331 -1.9969>
   rotation=<rot -88.5118 0.0001 -0.0001>
  }]
 dyn_skirt_back_mid_01 [{shapeType=Capsule, length=50.7452, radius=22.9875, shapeName=dyn_skirt_back_mid_01_capsule, center=<12.6836 13.334 0>}]
 dyn_skirt_back_mid_04 [{shapeType=Capsule, length=52.5903, radius=21.1425, shapeName=dyn_skirt_back_mid_04_capsule, center=<11.1948 10.9696 0>}]
 dyn_skirt_back_mid_MSTR [{
   shapeType=Capsule
   length=35.1883
   radius=17.9599
   shapeName=dyn_skirt_back_mid_MSTR_capsule
   center=<28.3057 8.6116 -0.4044>
   rotation=<rot -88.4348 179.9999 -179.9999>
  }]
 dyn_skirt_back_r_01 [{
   shapeType=Capsule
   length=17.0762
   radius=18.4253
   shapeName=dyn_skirt_back_r_01_capsule
   center=<-12.6836 -8.5229 0>
   rotation=<rot 0 0 -180>
  }]
 dyn_skirt_back_r_04 [{
   shapeType=Capsule
   length=18.7153
   radius=21.5892
   shapeName=dyn_skirt_back_r_04_capsule
   center=<-11.1948 -5.6343 0>
   rotation=<rot 0 0 -180>
  }]
 dyn_skirt_back_r_MSTR [{
   shapeType=Capsule
   length=36.3844
   radius=17.1856
   shapeName=dyn_skirt_back_r_MSTR_capsule
   center=<-30.8498 -6.7331 1.9969>
   rotation=<rot -88.5118 0.0001 179.9999>
  }]
 foot_l []
 foot_r []
 head [{shapeType=Sphere, radius=17.3125, center=<11.028 -0.174 0>}]
 lowerarm_l []
 lowerarm_r []
 neck_02 [{shapeType=Capsule, length=13.4375, radius=7.3175, center=<-0.2642 -0.3408 0>, rotation=<rot 90 -171.1193 -175.3978>}]
 pelvis [{shapeType=Capsule, length=19.3524, radius=16.965, center=<-5.1352 -6.6528 0>}]
 spine_01 [{shapeType=Capsule, length=7.9438, radius=13.5586, center=<-5.7576 -4.3707 0>}]
 spine_02 [
  {shapeType=Capsule, length=16.2398, radius=17.9579, center=<-5.96 -4.39 -1.8505>}
  {shapeType=Capsule, length=29.3956, radius=11.5807, center=<2.5024 0.4992 3.9825>, rotation=<rot 90 0.0049 0.0049>}
  {shapeType=Capsule, length=29.875, radius=11.1013, center=<2.4704 0.4991 -3.9825>, rotation=<rot 90 -0.0396 -0.0396>}
 ]
 spine_03 [{shapeType=Capsule, length=15.6141, radius=16.7458, center=<0.1445 -2.4311 0.0001>}]
 spine_04 [{shapeType=Capsule, length=15.4806, radius=14.1325, center=<2.766 -0.5276 0.0002>}]
 spine_05 [
  {shapeType=Capsule, length=8.0226, radius=15.695, center=<-2.2795 1.8109 0>}
  {shapeType=Capsule, length=13.9518, radius=13.9425, center=<-0.2916 0.4452 0>}
 ]
 thigh_l []
 thigh_r []
 upperarm_l []
 upperarm_r []
}
constraints []
