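# Post-process-style AnimBlueprint for the FlickLedge character's face accessory mesh, adding secondary motion to its middle, left and right hair bones.
# It copies pose and curves from the attached parent, then runs a chain of root-relative Anim Dynamics nodes with angular springs, tight limits and reduced gravity.
# Additive bone-pitch nodes sit unconnected to the output.
# Velocity-direction logic for LeftMovement and RightMovement is unwired from the update event, and neither variable is read.
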AnimBlueprint /BRCosmetics/Characters/Player/Male/Medium/Bodies/M_MED_FlickLedge/Meshes/Parts/M_MED_FlickLedge_FaceAcc_AnimBP
target_skeleton=&/M_MED_FlickLedge_FaceAcc_Skeleton
graphs [
 AnimGraph {nodes [
   n1 = "Transform (Modify) Bone - Bone: dyn_hair_mid_01"!(ComponentPose, Alpha=1, Rotation="-30,0,0", Scale="1,1,1") -> (Pose) {
    rotationMode=BMM_Additive
    rotationSpace=BCS_BoneSpace
    alphaScaleBiasClamp {bInterpResult=true, bMapRange=true, inRange {max=90}}
    boneToModify {boneName=dyn_hair_mid_01}
    rotation {pitch=-30}
   }
   n2 = "Transform (Modify) Bone - Bone: dyn_hair_l_01"!(ComponentPose=@n1, Alpha=1, Rotation="-30,0,0", Scale="1,1,1") -> (Pose) {
    rotationMode=BMM_Additive
    rotationSpace=BCS_BoneSpace
    alphaScaleBiasClamp {bInterpResult=true, bMapRange=true, inRange {max=90}}
    boneToModify {boneName=dyn_hair_l_01}
    rotation {pitch=-30}
   }
   n3 = "Transform (Modify) Bone - Bone: dyn_hair_r_01"!(ComponentPose=@n2, Alpha=1, Rotation="-30,0,0", Scale="1,1,1") -> (Pose) {
    rotationMode=BMM_Additive
    rotationSpace=BCS_BoneSpace
    alphaScaleBiasClamp {bInterpResult=true, bMapRange=true, inRange {max=90}}
    boneToModify {boneName=dyn_hair_r_01}
    rotation {pitch=-30}
   }
   n4 = "Transform (Modify) Bone - Bone: dyn_hair_mid_01"(ComponentPose=@n3, Alpha=1, Rotation="30,0,0", Scale="1,1,1") {
    rotationMode=BMM_Additive
    rotationSpace=BCS_BoneSpace
    alphaScaleBiasClamp {bInterpResult=true, bMapRange=true, inRange {max=0, min=-90}, outRange {max=0, min=1}}
    boneToModify {boneName=dyn_hair_mid_01}
    rotation {pitch=30}
   }
   n5 = "Transform (Modify) Bone - Bone: dyn_hair_l_01"(ComponentPose=@n4, Alpha=1, Rotation="30,0,0", Scale="1,1,1") {
    rotationMode=BMM_Additive
    rotationSpace=BCS_BoneSpace
    alphaScaleBiasClamp {bInterpResult=true, bMapRange=true, inRange {max=0, min=-90}, outRange {max=0, min=1}}
    boneToModify {boneName=dyn_hair_l_01}
    rotation {pitch=30}
   }
   "Transform (Modify) Bone - Bone: dyn_hair_r_01"(ComponentPose=@n5, Alpha=1, Rotation="30,0,0", Scale="1,1,1") {
    rotationMode=BMM_Additive
    rotationSpace=BCS_BoneSpace
    alphaScaleBiasClamp {bInterpResult=true, bMapRange=true, inRange {max=0, min=-90}, outRange {max=0, min=1}}
    boneToModify {boneName=dyn_hair_r_01}
    rotation {pitch=30}
   }
   n6 = Anim Dynamics - Bone: dyn_hair_mid_01(
    ComponentPose=Local To Component(LocalPose=Copy Pose From Mesh() {bCopyCurves=true, bUseAttachedParent=true})
    Alpha=1
    ExternalForce="0,0,0"
   ) {
    angularDampingOverride=0.9
    angularSpringConstant=5000
    bAngularSpring=true
    bOverrideAngularDamping=true
    bUseGravityOverride=true
    gravityScale=0.35
    simulationSpace=RootRelative
    boundBone {boneName=dyn_hair_mid_01}
    gravityOverride {z=-50}
    physicsBodyDefinitions [{
      boundBone {boneName=dyn_hair_mid_01}
      boxExtents=<20 30 30>
      constraintSetup {angularLimitsMax=<15 15 15>, angularLimitsMin=<-15 -15 -15>, angularTarget=<1 0 0>}
      localJointOffset=<4 0 0>
     }]
    simSpaceSettings {maxAngularAcceleration=100, maxAngularVelocity=1400, simSpaceAngularAlpha=1}
   }
   n7 = Anim Dynamics - Bone: dyn_hair_mid_02(ComponentPose=@n6, Alpha=1, ExternalForce="0,0,0") {
    angularDampingOverride=0.9
    angularSpringConstant=5000
    bAngularSpring=true
    bOverrideAngularDamping=true
    bUseGravityOverride=true
    gravityScale=0.35
    simulationSpace=RootRelative
    boundBone {boneName=dyn_hair_mid_02}
    gravityOverride {z=-50}
    physicsBodyDefinitions [{
      boundBone {boneName=dyn_hair_mid_02}
      boxExtents=<20 30 30>
      constraintSetup {angularLimitsMax=<15 15 15>, angularLimitsMin=<-15 -15 -15>, angularTarget=<1 0 0>}
      localJointOffset=<4 0 0>
     }]
    simSpaceSettings {maxAngularAcceleration=100, maxAngularVelocity=1400, simSpaceAngularAlpha=1}
   }
   n8 = Anim Dynamics - Bone: dyn_hair_mid_03(ComponentPose=@n7, Alpha=1, ExternalForce="0,0,0") {
    angularDampingOverride=0.9
    angularSpringConstant=5000
    bAngularSpring=true
    bOverrideAngularDamping=true
    bUseGravityOverride=true
    gravityScale=0.35
    simulationSpace=RootRelative
    boundBone {boneName=dyn_hair_mid_03}
    gravityOverride {z=-50}
    physicsBodyDefinitions [{
      boundBone {boneName=dyn_hair_mid_03}
      boxExtents=<20 30 30>
      constraintSetup {angularLimitsMax=<15 15 15>, angularLimitsMin=<-15 -15 -15>, angularTarget=<1 0 0>}
      localJointOffset=<4 0 0>
     }]
    simSpaceSettings {maxAngularAcceleration=100, maxAngularVelocity=1400, simSpaceAngularAlpha=1}
   }
   n9 = Anim Dynamics - Bone: dyn_hair_l_01(ComponentPose=@n8, Alpha=1, ExternalForce="0,0,0") {
    angularDampingOverride=0.9
    angularSpringConstant=5000
    bAngularSpring=true
    bOverrideAngularDamping=true
    bUseGravityOverride=true
    gravityScale=0.35
    simulationSpace=RootRelative
    boundBone {boneName=dyn_hair_l_01}
    gravityOverride {z=-50}
    physicsBodyDefinitions [{
      boundBone {boneName=dyn_hair_l_01}
      boxExtents=<20 30 30>
      constraintSetup {angularLimitsMax=<15 15 15>, angularLimitsMin=<-15 -15 -15>, angularTarget=<1 0 0>}
      localJointOffset=<4 0 0>
     }]
    simSpaceSettings {maxAngularAcceleration=100, maxAngularVelocity=1400, simSpaceAngularAlpha=1}
   }
   n10 = Anim Dynamics - Bone: dyn_hair_l_02(ComponentPose=@n9, Alpha=1, ExternalForce="0,0,0") {
    angularDampingOverride=0.9
    angularSpringConstant=5000
    bAngularSpring=true
    bOverrideAngularDamping=true
    bUseGravityOverride=true
    gravityScale=0.35
    simulationSpace=RootRelative
    boundBone {boneName=dyn_hair_l_02}
    gravityOverride {z=-50}
    physicsBodyDefinitions [{
      boundBone {boneName=dyn_hair_l_02}
      boxExtents=<20 30 30>
      constraintSetup {angularLimitsMax=<15 15 15>, angularLimitsMin=<-15 -15 -15>, angularTarget=<1 0 0>}
      localJointOffset=<4 0 0>
     }]
    simSpaceSettings {maxAngularAcceleration=100, maxAngularVelocity=1400, simSpaceAngularAlpha=1}
   }
   n11 = Anim Dynamics - Bone: dyn_hair_r_01(ComponentPose=@n10, Alpha=1, ExternalForce="0,0,0") {
    angularDampingOverride=0.9
    angularSpringConstant=5000
    bAngularSpring=true
    bOverrideAngularDamping=true
    bUseGravityOverride=true
    gravityScale=0.35
    simulationSpace=RootRelative
    boundBone {boneName=dyn_hair_r_01}
    gravityOverride {z=-50}
    physicsBodyDefinitions [{
      boundBone {boneName=dyn_hair_r_01}
      boxExtents=<20 30 30>
      constraintSetup {angularLimitsMax=<15 15 15>, angularLimitsMin=<-15 -15 -15>, angularTarget=<1 0 0>}
      localJointOffset=<-4 0 0>
     }]
    simSpaceSettings {maxAngularAcceleration=100, maxAngularVelocity=1400, simSpaceAngularAlpha=1}
   }
   n12 = Anim Dynamics - Bone: dyn_hair_r_02(ComponentPose=@n11, Alpha=1, ExternalForce="0,0,0") {
    angularDampingOverride=0.9
    angularSpringConstant=5000
    bAngularSpring=true
    bOverrideAngularDamping=true
    bUseGravityOverride=true
    gravityScale=0.35
    simulationSpace=RootRelative
    boundBone {boneName=dyn_hair_r_02}
    gravityOverride {z=-50}
    physicsBodyDefinitions [{
      boundBone {boneName=dyn_hair_r_02}
      boxExtents=<20 30 30>
      constraintSetup {angularLimitsMax=<15 15 15>, angularLimitsMin=<-15 -15 -15>, angularTarget=<1 0 0>}
      localJointOffset=<-4 0 0>
     }]
    simSpaceSettings {maxAngularAcceleration=100, maxAngularVelocity=1400, simSpaceAngularAlpha=1}
   }
   Output Pose(Result=Component To Local(ComponentPose=@n12)) {layerGroup=None}
  ]}
 EventGraph {nodes [
   n1 = GetVelocity!(self:Actor Object Reference=TryGetPawnOwner()) -> (ReturnValue:Vector)
   n2 = CalculateDirection!(
    self:Anim Instance Object Reference
    Velocity:"Vector (by ref)"=vector * vector(A=@n1, B:Vector="1,0,0") -> (ReturnValue=~)
    BaseRotation:"Rotator (by ref)"="0, 0, 0"
   ) -> (ReturnValue:float)
   Set RightMovement(RightMovement:double="Clamp (Float)"(Value=@n2, Min=0, Max=90)) -> (Output_Get:double)
   > Set LeftMovement(LeftMovement:double="Clamp (Float)"(Value=@n2, Min=-90, Max=0)) -> (Output_Get:double)
   "To String (Float)"(InDouble=Vector Length XY!(A:Vector=@n1) -> (ReturnValue:double))
   Event BlueprintUpdateAnimation()
  ]}
]
variables [
 LeftMovement {type=number}
 RightMovement {type=number}
]
